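# Animation Blueprint for the sportbike vehicle skeleton, animating idle, locomotion, braking, reversing, jumps and swimming.
# A cached FullBody state machine feeds a PostProcessing linked layer.
# The ThrottleBar machine overrides throttle and engine bones for revving, throttling and ground tricks, then MotorcycleAdditives applies on top with inverted alpha.
# A FinalPoseOverride linked layer and inertialization finish the pose, smoothing inertialized transitions such as jump entry.
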AnimBlueprint /Sportbike/Animation/Proto_Sportbike_AnimBP
target_skeleton=/Sportbike/Vehicle/Meshes/Moto_Sport_Veh_Skeleton
graphs [
 AnimGraph {nodes [
   n1 = Use cached pose 'FullBodyPose'()
   AnimGraphNode_StateMachine_1 = ThrottleBar() {is=title AnimGraphNode_StateMachine, bAllowConduitEntryStates=true}
   AnimGraphNode_StateMachine_2 = MotorcycleAdditives() {is=title AnimGraphNode_StateMachine}
   n2 = Apply Additive(
    Base=Layered blend per bone(
     BasePose=PostProcessing!(InPosePostProcessing:Pose Link Structure=@n1) -> (Pose:Pose Link Structure) {is=title AnimGraphNode_LinkedAnimLayer, layer=PostProcessing}
     BlendPoses_0=@AnimGraphNode_StateMachine_1
    ) {layerSetup [{branchFilters [{boneName=throttle}, {boneName=engine}]}]}
    Additive=@AnimGraphNode_StateMachine_2
   ) {alphaBoolBlend {blendOption=Linear}, alphaScaleBiasClamp {bias=1, scale=-1}}
   Output Pose(Result=Inertialization!(Source:Pose Link Structure=FinalPoseOverride!(InPoseFinalPoseOverride:Pose Link Structure=@n2) -> (Pose:Pose Link Structure) {is=title AnimGraphNode_LinkedAnimLayer, layer=FinalPoseOverride}) -> (Pose:Pose Link Structure) {is=title AnimGraphNode_Inertialization}) {layerGroup=None}
   AnimGraphNode_StateMachine_0 = FullBody() {is=title AnimGraphNode_StateMachine, bAllowConduitEntryStates=true}
   Save cached pose 'FullBodyPose'(Pose=@AnimGraphNode_StateMachine_0) {cacheName=FullBodyPose, cachePoseName=FullBodyPose}
  ]}
 FullBody {
  entry=Idle
  path=AnimGraph.AnimGraphNode_StateMachine_0.FullBody
  ids=id
  nodes [
   ToLocomotion() {is=title AnimStateAliasNode, also {aliases [Idle,JumpLand,MoveStop,ReverseStop]}}
   ToIdle() {is=title AnimStateAliasNode, also {aliases [JumpLand,MoveStop,ReverseStop]}}
   ToJump() {is=title AnimStateAliasNode, also {aliases [Braking,BrakingOutro,Idle,JumpLand,Locomotion,MoveStop,Reverse,ReverseStop]}}
   ByPassToMovingJumpLand() {is=title AnimStateAliasNode, also {aliases [NormalJump]}}
   AnimStateConduitNode_1 = LocoSelector() {is=title AnimStateConduitNode}
   AnimStateConduitNode_2 = JumpSelector() {is=title AnimStateConduitNode}
   AnimStateNode_0 = Idle()
   AnimStateNode_1 = Locomotion() {is=title AnimStateNode, bAlwaysResetOnEntry=true, stateEntered {guid=#1, notifyName=LocomotionEnter}}
   AnimStateNode_10 = MoveStop() {is=title AnimStateNode}
   AnimStateNode_11 = NormalJump() {is=title AnimStateNode}
   AnimStateNode_13 = JumpLand() {is=title AnimStateNode, stateEntered {guid=#2, notifyName=JumpLandEnter}}
   AnimStateNode_14 = ReverseStop() {is=title AnimStateNode}
   AnimStateNode_2 = Swim() {is=title AnimStateNode}
   AnimStateNode_3 = Braking() {is=title AnimStateNode}
   AnimStateNode_4 = BrakingOutro() {is=title AnimStateNode}
   AnimStateNode_9 = Reverse() {is=title AnimStateNode}
   AnimStateTransitionNode_0 = AnimStateTransitionNode() {priorityOrder=2}
   AnimStateTransitionNode_1 = AnimStateTransitionNode()
   AnimStateTransitionNode_12 = AnimStateTransitionNode()
   AnimStateTransitionNode_18 = AnimStateTransitionNode() {priorityOrder=100}
   AnimStateTransitionNode_19 = AnimStateTransitionNode() {crossfadeDuration=0.35, priorityOrder=0}
   AnimStateTransitionNode_2 = AnimStateTransitionNode()
   AnimStateTransitionNode_20 = AnimStateTransitionNode()
   AnimStateTransitionNode_21 = AnimStateTransitionNode() {bAutomaticRuleBasedOnSequencePlayerInState=true, crossfadeDuration=0.4}
   AnimStateTransitionNode_22 = AnimStateTransitionNode()
   AnimStateTransitionNode_23 = AnimStateTransitionNode() {crossfadeDuration=0.067, logicType=TLT_Inertialization, priorityOrder=100}
   AnimStateTransitionNode_25 = AnimStateTransitionNode() {crossfadeDuration=0.033}
   AnimStateTransitionNode_27 = AnimStateTransitionNode()
   AnimStateTransitionNode_28 = AnimStateTransitionNode()
   AnimStateTransitionNode_3 = AnimStateTransitionNode()
   AnimStateTransitionNode_4 = AnimStateTransitionNode()
   AnimStateTransitionNode_5 = AnimStateTransitionNode()
   AnimStateTransitionNode_6 = AnimStateTransitionNode() {bAutomaticRuleBasedOnSequencePlayerInState=true, crossfadeDuration=0.3, priorityOrder=2}
   AnimStateTransitionNode_7 = AnimStateTransitionNode()
   AnimStateTransitionNode_8 = AnimStateTransitionNode()
  ]
  links [
   {
    dst=BrakingOutro
    rule=AnimGraph.AnimGraphNode_StateMachine_0.FullBody.AnimStateTransitionNode_5.Transition
    src=Braking
    via=AnimStateTransitionNode_5
   }
   {
    dst=LocoSelector
    rule=AnimGraph.AnimGraphNode_StateMachine_0.FullBody.AnimStateTransitionNode_7.Transition
    src=Braking
    via=AnimStateTransitionNode_7
   }
   {
    dst=Locomotion
    rule=AnimGraph.AnimGraphNode_StateMachine_0.FullBody.AnimStateTransitionNode_6.Transition
    src=BrakingOutro
    via=AnimStateTransitionNode_6
   }
   {
    dst=Locomotion
    rule=AnimGraph.AnimGraphNode_StateMachine_0.FullBody.AnimStateTransitionNode_8.Transition
    src=BrakingOutro
    via=AnimStateTransitionNode_8
   }
   {
    dst=Locomotion
    rule=AnimGraph.AnimGraphNode_StateMachine_0.FullBody.AnimStateTransitionNode_27.Transition
    src=ByPassToMovingJumpLand
    via=AnimStateTransitionNode_27
   }
   {
    dst=NormalJump
    rule=AnimGraph.AnimGraphNode_StateMachine_0.FullBody.AnimStateTransitionNode_23.Transition
    src=JumpSelector
    via=AnimStateTransitionNode_23
   }
   {
    dst=Braking
    rule=AnimGraph.AnimGraphNode_StateMachine_0.FullBody.AnimStateTransitionNode_3.Transition
    src=LocoSelector
    via=AnimStateTransitionNode_3
   }
   {
    dst=Locomotion
    rule=AnimGraph.AnimGraphNode_StateMachine_0.FullBody.AnimStateTransitionNode_18.Transition
    src=LocoSelector
    via=AnimStateTransitionNode_18
   }
   {
    dst=Reverse
    rule=AnimGraph.AnimGraphNode_StateMachine_0.FullBody.AnimStateTransitionNode_19.Transition
    src=LocoSelector
    via=AnimStateTransitionNode_19
   }
   {
    dst=Swim
    rule=AnimGraph.AnimGraphNode_StateMachine_0.FullBody.AnimStateTransitionNode_0.Transition
    src=LocoSelector
    via=AnimStateTransitionNode_0
   }
   {
    dst=LocoSelector
    rule=AnimGraph.AnimGraphNode_StateMachine_0.FullBody.AnimStateTransitionNode_4.Transition
    src=Locomotion
    via=AnimStateTransitionNode_4
   }
   {
    dst=MoveStop
    rule=AnimGraph.AnimGraphNode_StateMachine_0.FullBody.AnimStateTransitionNode_20.Transition
    src=Locomotion
    via=AnimStateTransitionNode_20
   }
   {
    dst=JumpLand
    rule=AnimGraph.AnimGraphNode_StateMachine_0.FullBody.AnimStateTransitionNode_25.Transition
    src=NormalJump
    via=AnimStateTransitionNode_25
   }
   {
    dst=LocoSelector
    rule=AnimGraph.AnimGraphNode_StateMachine_0.FullBody.AnimStateTransitionNode_2.Transition
    src=Reverse
    via=AnimStateTransitionNode_2
   }
   {
    dst=ReverseStop
    rule=AnimGraph.AnimGraphNode_StateMachine_0.FullBody.AnimStateTransitionNode_28.Transition
    src=Reverse
    via=AnimStateTransitionNode_28
   }
   {
    dst=LocoSelector
    rule=AnimGraph.AnimGraphNode_StateMachine_0.FullBody.AnimStateTransitionNode_1.Transition
    src=Swim
    via=AnimStateTransitionNode_1
   }
   {
    dst=Idle
    rule=AnimGraph.AnimGraphNode_StateMachine_0.FullBody.AnimStateTransitionNode_21.Transition
    src=ToIdle
    via=AnimStateTransitionNode_21
   }
   {
    dst=JumpSelector
    rule=AnimGraph.AnimGraphNode_StateMachine_0.FullBody.AnimStateTransitionNode_22.Transition
    src=ToJump
    via=AnimStateTransitionNode_22
   }
   {
    dst=LocoSelector
    rule=AnimGraph.AnimGraphNode_StateMachine_0.FullBody.AnimStateTransitionNode_12.Transition
    src=ToLocomotion
    via=AnimStateTransitionNode_12
   }
  ]
 }
 LocoSelector {path=AnimGraph.AnimGraphNode_StateMachine_0.FullBody.AnimStateConduitNode_1.LocoSelector, nodes [Result(bCanEnterTransition=true)]}
 JumpSelector {path=AnimGraph.AnimGraphNode_StateMachine_0.FullBody.AnimStateConduitNode_2.JumpSelector, nodes [Result(bCanEnterTransition=true)]}
 Idle {
  path=AnimGraph.AnimGraphNode_StateMachine_0.FullBody.AnimStateNode_0.Idle
  nodes [Output Animation Pose(Result=Sequence Player 'Sportbike_Idle_Stopped_Veh'() {sequence=/Sportbike/Animation/Sportbike/Sportbike_Idle_Stopped_Veh})]
 }
 Locomotion {
  path=AnimGraph.AnimGraphNode_StateMachine_0.FullBody.AnimStateNode_1.Locomotion
  nodes [
   AnimGraphNode_StateMachine_0 = Movement() {is=title AnimGraphNode_StateMachine, bAllowConduitEntryStates=true}
   Output Animation Pose(Result=@AnimGraphNode_StateMachine_0)
  ]
 }
 Movement {
  entry=Conduit
  path=AnimGraph.AnimGraphNode_StateMachine_0.FullBody.AnimStateNode_1.Locomotion.AnimGraphNode_StateMachine_0.Movement
  ids=id
  nodes [
   AnimStateConduitNode_0 = Conduit()
   AnimStateNode_0 = MovingStart() {is=title AnimStateNode}
   AnimStateNode_1 = MovingLoop() {is=title AnimStateNode}
   AnimStateTransitionNode_15 = AnimStateTransitionNode() {priorityOrder=100}
   AnimStateTransitionNode_16 = AnimStateTransitionNode() {crossfadeDuration=0.5}
   AnimStateTransitionNode_17 = AnimStateTransitionNode() {bAutomaticRuleBasedOnSequencePlayerInState=true, crossfadeDuration=0.5}
   AnimStateTransitionNode_18 = AnimStateTransitionNode() {blendMode=Linear, priorityOrder=0}
  ]
  links [
   {
    dst=MovingLoop
    rule=AnimGraph.AnimGraphNode_StateMachine_0.FullBody.AnimStateNode_1.Locomotion.AnimGraphNode_StateMachine_0.Movement.AnimStateTransitionNode_16.Transition
    src=Conduit
    via=AnimStateTransitionNode_16
   }
   {
    dst=MovingStart
    rule=AnimGraph.AnimGraphNode_StateMachine_0.FullBody.AnimStateNode_1.Locomotion.AnimGraphNode_StateMachine_0.Movement.AnimStateTransitionNode_15.Transition
    src=Conduit
    via=AnimStateTransitionNode_15
   }
   {
    dst=MovingLoop
    rule=AnimGraph.AnimGraphNode_StateMachine_0.FullBody.AnimStateNode_1.Locomotion.AnimGraphNode_StateMachine_0.Movement.AnimStateTransitionNode_17.Transition
    src=MovingStart
    via=AnimStateTransitionNode_17
   }
   {
    dst=MovingLoop
    rule=AnimGraph.AnimGraphNode_StateMachine_0.FullBody.AnimStateNode_1.Locomotion.AnimGraphNode_StateMachine_0.Movement.AnimStateTransitionNode_18.Transition
    src=MovingStart
    via=AnimStateTransitionNode_18
   }
  ]
 }
 Conduit {
  path=AnimGraph.AnimGraphNode_StateMachine_0.FullBody.AnimStateNode_1.Locomotion.AnimGraphNode_StateMachine_0.Movement.AnimStateConduitNode_0.Conduit
  nodes [Result(bCanEnterTransition=true)]
 }
 MovingStart {
  path=AnimGraph.AnimGraphNode_StateMachine_0.FullBody.AnimStateNode_1.Locomotion.AnimGraphNode_StateMachine_0.Movement.AnimStateNode_0.MovingStart
  nodes [Output Animation Pose(Result=Sequence Player 'Sportbike_MoveStart_Veh'() {bLoopAnimation=false, sequence=/Sportbike/Animation/Sportbike/Sportbike_MoveStart_Veh})]
 }
 MovingLoop {
  path=AnimGraph.AnimGraphNode_StateMachine_0.FullBody.AnimStateNode_1.Locomotion.AnimGraphNode_StateMachine_0.Movement.AnimStateNode_1.MovingLoop
  nodes [Output Animation Pose(Result=Sequence Player 'Sportbike_Loco_Veh'() {sequence=/Sportbike/Animation/Sportbike/Sportbike_Loco_Veh})]
 }
 Transition {
  path=AnimGraph.AnimGraphNode_StateMachine_0.FullBody.AnimStateNode_1.Locomotion.AnimGraphNode_StateMachine_0.Movement.AnimStateTransitionNode_15.Transition
  nodes [Result(bCanEnterTransition=true)]
 }
 Transition {
  path=AnimGraph.AnimGraphNode_StateMachine_0.FullBody.AnimStateNode_1.Locomotion.AnimGraphNode_StateMachine_0.Movement.AnimStateTransitionNode_16.Transition
  nodes [Result(bCanEnterTransition=Get bStateTransitionRuleToMovingLoop() -> (bStateTransitionRuleToMovingLoop:bool))]
 }
 Transition {
  path=AnimGraph.AnimGraphNode_StateMachine_0.FullBody.AnimStateNode_1.Locomotion.AnimGraphNode_StateMachine_0.Movement.AnimStateTransitionNode_17.Transition
  nodes [Result()]
 }
 Transition {
  path=AnimGraph.AnimGraphNode_StateMachine_0.FullBody.AnimStateNode_1.Locomotion.AnimGraphNode_StateMachine_0.Movement.AnimStateTransitionNode_18.Transition
  nodes [Result(bCanEnterTransition=Get bEarlyOutFromMovingStart() -> (bEarlyOutFromMovingStart:bool))]
 }
 MoveStop {
  path=AnimGraph.AnimGraphNode_StateMachine_0.FullBody.AnimStateNode_10.MoveStop
  nodes [Output Animation Pose(Result=Sequence Player 'Sportbike_MoveStop_Veh'() {bLoopAnimation=false, sequence=/Sportbike/Animation/Sportbike/Sportbike_MoveStop_Veh})]
 }
 NormalJump {
  path=AnimGraph.AnimGraphNode_StateMachine_0.FullBody.AnimStateNode_11.NormalJump
  nodes [
   AnimGraphNode_StateMachine_0 = NormalJumpTransition() {is=title AnimGraphNode_StateMachine, bAllowConduitEntryStates=true}
   Output Animation Pose(Result=@AnimGraphNode_StateMachine_0)
  ]
 }
 NormalJumpTransition {
  entry=Conduit
  path=AnimGraph.AnimGraphNode_StateMachine_0.FullBody.AnimStateNode_11.NormalJump.AnimGraphNode_StateMachine_0.NormalJumpTransition
  ids=id
  nodes [
   AnimStateConduitNode_0 = Conduit()
   AnimStateNode_0 = NormalJumpStart() {is=title AnimStateNode}
   AnimStateNode_1 = NormalJumpLoop() {is=title AnimStateNode, stateEntered {guid=#3, notifyName=NormalJumpLoopEnter}}
   AnimStateTransitionNode_0 = AnimStateTransitionNode() {priorityOrder=100}
   AnimStateTransitionNode_1 = AnimStateTransitionNode() {bAutomaticRuleBasedOnSequencePlayerInState=true}
   AnimStateTransitionNode_2 = AnimStateTransitionNode() {priorityOrder=2}
  ]
  links [
   {
    dst=NormalJumpLoop
    rule=AnimGraph.AnimGraphNode_StateMachine_0.FullBody.AnimStateNode_11.NormalJump.AnimGraphNode_StateMachine_0.NormalJumpTransition.AnimStateTransitionNode_2.Transition
    src=Conduit
    via=AnimStateTransitionNode_2
   }
   {
    dst=NormalJumpStart
    rule=AnimGraph.AnimGraphNode_StateMachine_0.FullBody.AnimStateNode_11.NormalJump.AnimGraphNode_StateMachine_0.NormalJumpTransition.AnimStateTransitionNode_0.Transition
    src=Conduit
    via=AnimStateTransitionNode_0
   }
   {
    dst=NormalJumpLoop
    rule=AnimGraph.AnimGraphNode_StateMachine_0.FullBody.AnimStateNode_11.NormalJump.AnimGraphNode_StateMachine_0.NormalJumpTransition.AnimStateTransitionNode_1.Transition
    src=NormalJumpStart
    via=AnimStateTransitionNode_1
   }
  ]
 }
 Conduit {
  path=AnimGraph.AnimGraphNode_StateMachine_0.FullBody.AnimStateNode_11.NormalJump.AnimGraphNode_StateMachine_0.NormalJumpTransition.AnimStateConduitNode_0.Conduit
  nodes [Result(bCanEnterTransition=true)]
 }
 NormalJumpStart {
  path=AnimGraph.AnimGraphNode_StateMachine_0.FullBody.AnimStateNode_11.NormalJump.AnimGraphNode_StateMachine_0.NormalJumpTransition.AnimStateNode_0.NormalJumpStart
  nodes [Output Animation Pose(Result=Sequence Player 'Sportbike_Jump_Start_Veh'() {bLoopAnimation=false, sequence=/Sportbike/Animation/Sportbike/Sportbike_Jump_Start_Veh})]
 }
 NormalJumpLoop {
  path=AnimGraph.AnimGraphNode_StateMachine_0.FullBody.AnimStateNode_11.NormalJump.AnimGraphNode_StateMachine_0.NormalJumpTransition.AnimStateNode_1.NormalJumpLoop
  nodes [Output Animation Pose(Result=Sequence Player 'Sportbike_Jump_Loop_Veh'() {sequence=/Sportbike/Animation/Sportbike/Sportbike_Jump_Loop_Veh})]
 }
 Transition {
  path=AnimGraph.AnimGraphNode_StateMachine_0.FullBody.AnimStateNode_11.NormalJump.AnimGraphNode_StateMachine_0.NormalJumpTransition.AnimStateTransitionNode_0.Transition
  nodes [Result(bCanEnterTransition=true)]
 }
 Transition {
  path=AnimGraph.AnimGraphNode_StateMachine_0.FullBody.AnimStateNode_11.NormalJump.AnimGraphNode_StateMachine_0.NormalJumpTransition.AnimStateTransitionNode_1.Transition
  nodes [Result()]
 }
 Transition {
  path=AnimGraph.AnimGraphNode_StateMachine_0.FullBody.AnimStateNode_11.NormalJump.AnimGraphNode_StateMachine_0.NormalJumpTransition.AnimStateTransitionNode_2.Transition
  nodes [Result(bCanEnterTransition=Get bIsFromJumpTrick() -> (bIsFromJumpTrick:bool))]
 }
 JumpLand {
  path=AnimGraph.AnimGraphNode_StateMachine_0.FullBody.AnimStateNode_13.JumpLand
  nodes [Output Animation Pose(Result=Sequence Player 'Sportbike_Jump_Land_Veh'() {bLoopAnimation=false, sequence=/Sportbike/Animation/Sportbike/Sportbike_Jump_Land_Veh})]
 }
 ReverseStop {
  path=AnimGraph.AnimGraphNode_StateMachine_0.FullBody.AnimStateNode_14.ReverseStop
  nodes [Output Animation Pose(Result=Sequence Player 'Sportbike_Reverse_Stop_Veh'() {bLoopAnimation=false, sequence=/Sportbike/Animation/Sportbike/Sportbike_Reverse_Stop_Veh})]
 }
 Swim {
  path=AnimGraph.AnimGraphNode_StateMachine_0.FullBody.AnimStateNode_2.Swim
  nodes [Output Animation Pose(Result=Sequence Player 'Sportbike_Loco_Veh'() {sequence=&Sportbike_Loco_Veh})]
 }
 Braking {
  path=AnimGraph.AnimGraphNode_StateMachine_0.FullBody.AnimStateNode_3.Braking
  nodes [
   AnimGraphNode_StateMachine_0 = BrakingTransition() {is=title AnimGraphNode_StateMachine}
   Output Animation Pose(Result=Layered blend per bone(
     BasePose=@AnimGraphNode_StateMachine_0
     BlendPoses_0=Sequence Player 'Sportbike_Loco_Veh'() {sequence=&Sportbike_Loco_Veh}
    ) {layerSetup [{branchFilters [{boneName=brake_lever}]}]})
  ]
 }
 BrakingTransition {
  entry=BrakingStart
  path=AnimGraph.AnimGraphNode_StateMachine_0.FullBody.AnimStateNode_3.Braking.AnimGraphNode_StateMachine_0.BrakingTransition
  ids=id
  nodes [
   AnimStateNode_0 = BrakingStart() {is=title AnimStateNode}
   AnimStateNode_1 = BrakingLoop() {is=title AnimStateNode}
   AnimStateTransitionNode_6 = AnimStateTransitionNode() {bAutomaticRuleBasedOnSequencePlayerInState=true}
  ]
  links [{
    dst=BrakingLoop
    rule=AnimGraph.AnimGraphNode_StateMachine_0.FullBody.AnimStateNode_3.Braking.AnimGraphNode_StateMachine_0.BrakingTransition.AnimStateTransitionNode_6.Transition
    src=BrakingStart
    via=AnimStateTransitionNode_6
   }]
 }
 BrakingStart {
  path=AnimGraph.AnimGraphNode_StateMachine_0.FullBody.AnimStateNode_3.Braking.AnimGraphNode_StateMachine_0.BrakingTransition.AnimStateNode_0.BrakingStart
  nodes [Output Animation Pose(Result=Sequence Player 'Sportbike_Brake_Start_Veh'() {bLoopAnimation=false, sequence=/Sportbike/Animation/Sportbike/Sportbike_Brake_Start_Veh})]
 }
 BrakingLoop {
  path=AnimGraph.AnimGraphNode_StateMachine_0.FullBody.AnimStateNode_3.Braking.AnimGraphNode_StateMachine_0.BrakingTransition.AnimStateNode_1.BrakingLoop
  nodes [Output Animation Pose(Result=Sequence Player 'Sportbike_Brake_Pose_Veh'() {sequence=/Sportbike/Animation/Sportbike/Sportbike_Brake_Pose_Veh})]
 }
 Transition {
  path=AnimGraph.AnimGraphNode_StateMachine_0.FullBody.AnimStateNode_3.Braking.AnimGraphNode_StateMachine_0.BrakingTransition.AnimStateTransitionNode_6.Transition
  nodes [Result()]
 }
 BrakingOutro {
  path=AnimGraph.AnimGraphNode_StateMachine_0.FullBody.AnimStateNode_4.BrakingOutro
  nodes [Output Animation Pose(Result=Sequence Player 'Sportbike_Brake_End_Veh'() {bLoopAnimation=false, sequence=/Sportbike/Animation/Sportbike/Sportbike_Brake_End_Veh})]
 }
 Reverse {
  path=AnimGraph.AnimGraphNode_StateMachine_0.FullBody.AnimStateNode_9.Reverse
  nodes [
   AnimGraphNode_StateMachine_0 = ReverseTransition() {is=title AnimGraphNode_StateMachine, bAllowConduitEntryStates=true}
   Output Animation Pose(Result=@AnimGraphNode_StateMachine_0)
  ]
 }
 ReverseTransition {
  entry=Conduit
  path=AnimGraph.AnimGraphNode_StateMachine_0.FullBody.AnimStateNode_9.Reverse.AnimGraphNode_StateMachine_0.ReverseTransition
  ids=id
  nodes [
   AnimStateConduitNode_0 = Conduit()
   AnimStateNode_0 = ReverseStart() {is=title AnimStateNode}
   AnimStateNode_1 = ReverseLoop() {is=title AnimStateNode}
   AnimStateTransitionNode_1 = AnimStateTransitionNode() {priorityOrder=100}
   AnimStateTransitionNode_2 = AnimStateTransitionNode() {bAutomaticRuleBasedOnSequencePlayerInState=true}
   AnimStateTransitionNode_3 = AnimStateTransitionNode() {crossfadeDuration=0.35}
  ]
  links [
   {
    dst=ReverseLoop
    rule=AnimGraph.AnimGraphNode_StateMachine_0.FullBody.AnimStateNode_9.Reverse.AnimGraphNode_StateMachine_0.ReverseTransition.AnimStateTransitionNode_3.Transition
    src=Conduit
    via=AnimStateTransitionNode_3
   }
   {
    dst=ReverseStart
    rule=AnimGraph.AnimGraphNode_StateMachine_0.FullBody.AnimStateNode_9.Reverse.AnimGraphNode_StateMachine_0.ReverseTransition.AnimStateTransitionNode_1.Transition
    src=Conduit
    via=AnimStateTransitionNode_1
   }
   {
    dst=ReverseLoop
    rule=AnimGraph.AnimGraphNode_StateMachine_0.FullBody.AnimStateNode_9.Reverse.AnimGraphNode_StateMachine_0.ReverseTransition.AnimStateTransitionNode_2.Transition
    src=ReverseStart
    via=AnimStateTransitionNode_2
   }
  ]
 }
 Conduit {
  path=AnimGraph.AnimGraphNode_StateMachine_0.FullBody.AnimStateNode_9.Reverse.AnimGraphNode_StateMachine_0.ReverseTransition.AnimStateConduitNode_0.Conduit
  nodes [Result(bCanEnterTransition=true)]
 }
 ReverseStart {
  path=AnimGraph.AnimGraphNode_StateMachine_0.FullBody.AnimStateNode_9.Reverse.AnimGraphNode_StateMachine_0.ReverseTransition.AnimStateNode_0.ReverseStart
  nodes [Output Animation Pose(Result=Sequence Player 'Sportbike_Reverse_Start_Veh'() {bLoopAnimation=false, sequence=/Sportbike/Animation/Sportbike/Sportbike_Reverse_Start_Veh})]
 }
 ReverseLoop {
  path=AnimGraph.AnimGraphNode_StateMachine_0.FullBody.AnimStateNode_9.Reverse.AnimGraphNode_StateMachine_0.ReverseTransition.AnimStateNode_1.ReverseLoop
  nodes [Output Animation Pose(Result=Blend Poses by bool(
     BlendPose_0=Sequence Player 'Sportbike_Loco_Veh'() {sequence=&Sportbike_Loco_Veh}
     BlendPose_1=Sequence Player 'Sportbike_ReverseLoop_Veh'() {sequence=/Sportbike/Animation/Sportbike/Sportbike_ReverseLoop_Veh}
     BlendTime_0=0.6
     BlendTime_1=0.2
    ) {blendType=Linear})]
 }
 Transition {
  path=AnimGraph.AnimGraphNode_StateMachine_0.FullBody.AnimStateNode_9.Reverse.AnimGraphNode_StateMachine_0.ReverseTransition.AnimStateTransitionNode_1.Transition
  nodes [Result(bCanEnterTransition=true)]
 }
 Transition {
  path=AnimGraph.AnimGraphNode_StateMachine_0.FullBody.AnimStateNode_9.Reverse.AnimGraphNode_StateMachine_0.ReverseTransition.AnimStateTransitionNode_2.Transition
  nodes [Result()]
 }
 Transition {
  path=AnimGraph.AnimGraphNode_StateMachine_0.FullBody.AnimStateNode_9.Reverse.AnimGraphNode_StateMachine_0.ReverseTransition.AnimStateTransitionNode_3.Transition
  nodes [Result(bCanEnterTransition=Get bStateTransitionRuleToReverseLoop() -> (bStateTransitionRuleToReverseLoop:bool))]
 }
 Transition {
  path=AnimGraph.AnimGraphNode_StateMachine_0.FullBody.AnimStateTransitionNode_0.Transition
  nodes [Result(bCanEnterTransition=Get bIsSurfaceWater() -> (bIsSurfaceWater:bool))]
 }
 Transition {
  path=AnimGraph.AnimGraphNode_StateMachine_0.FullBody.AnimStateTransitionNode_1.Transition
  nodes [Result(bCanEnterTransition=NOT Boolean(A=Get bIsSurfaceWater()))]
 }
 Transition {
  path=AnimGraph.AnimGraphNode_StateMachine_0.FullBody.AnimStateTransitionNode_12.Transition
  nodes [Result(bCanEnterTransition=Get bIsMoving() -> (bIsMoving:bool))]
 }
 Transition {path=AnimGraph.AnimGraphNode_StateMachine_0.FullBody.AnimStateTransitionNode_18.Transition, nodes [Result(bCanEnterTransition=true)]}
 Transition {
  path=AnimGraph.AnimGraphNode_StateMachine_0.FullBody.AnimStateTransitionNode_19.Transition
  nodes [Result(bCanEnterTransition=Get bShouldReverse() -> (bShouldReverse:bool))]
 }
 Transition {
  path=AnimGraph.AnimGraphNode_StateMachine_0.FullBody.AnimStateTransitionNode_2.Transition
  nodes [Result(bCanEnterTransition=Get bIsSurfaceWater())]
 }
 Transition {
  path=AnimGraph.AnimGraphNode_StateMachine_0.FullBody.AnimStateTransitionNode_20.Transition
  nodes [Result(bCanEnterTransition=Get bShouldStop() -> (bShouldStop:bool))]
 }
 Transition {path=AnimGraph.AnimGraphNode_StateMachine_0.FullBody.AnimStateTransitionNode_21.Transition, nodes [Result()]}
 Transition {
  path=AnimGraph.AnimGraphNode_StateMachine_0.FullBody.AnimStateTransitionNode_22.Transition
  nodes [Result(bCanEnterTransition=Get bShouldJump() -> (bShouldJump:bool))]
 }
 Transition {path=AnimGraph.AnimGraphNode_StateMachine_0.FullBody.AnimStateTransitionNode_23.Transition, nodes [Result(bCanEnterTransition=true)]}
 Transition {
  path=AnimGraph.AnimGraphNode_StateMachine_0.FullBody.AnimStateTransitionNode_25.Transition
  nodes [Result(bCanEnterTransition=Get bIsLandingInPlace() -> (bIsLandingInPlace:bool))]
 }
 Transition {
  path=AnimGraph.AnimGraphNode_StateMachine_0.FullBody.AnimStateTransitionNode_27.Transition
  nodes [Result(bCanEnterTransition=Get bStateRuleFallingToMoving() -> (bStateRuleFallingToMoving:bool))]
 }
 Transition {
  path=AnimGraph.AnimGraphNode_StateMachine_0.FullBody.AnimStateTransitionNode_28.Transition
  nodes [Result(bCanEnterTransition=Get bShouldReverseStop() -> (bShouldReverseStop:bool))]
 }
 Transition {
  path=AnimGraph.AnimGraphNode_StateMachine_0.FullBody.AnimStateTransitionNode_3.Transition
  nodes [Result(bCanEnterTransition=Get bShouldBrake() -> (bShouldBrake:bool))]
 }
 Transition {
  path=AnimGraph.AnimGraphNode_StateMachine_0.FullBody.AnimStateTransitionNode_4.Transition
  nodes [Result(bCanEnterTransition=Get bShouldExitFromLocomotion() -> (bShouldExitFromLocomotion:bool))]
 }
 Transition {
  path=AnimGraph.AnimGraphNode_StateMachine_0.FullBody.AnimStateTransitionNode_5.Transition
  nodes [Result(bCanEnterTransition=NOT Boolean(A=Get bIsBraking() -> (bIsBraking:bool)))]
 }
 Transition {path=AnimGraph.AnimGraphNode_StateMachine_0.FullBody.AnimStateTransitionNode_6.Transition, nodes [Result()]}
 Transition {
  path=AnimGraph.AnimGraphNode_StateMachine_0.FullBody.AnimStateTransitionNode_7.Transition
  nodes [Result(bCanEnterTransition=Get bStoppie() -> (bStoppie:bool))]
 }
 Transition {
  path=AnimGraph.AnimGraphNode_StateMachine_0.FullBody.AnimStateTransitionNode_8.Transition
  nodes [Result(bCanEnterTransition=Get bEarlyOutBrakingOutro() -> (bEarlyOutBrakingOutro:bool))]
 }
 ThrottleBar {
  entry=FullBodyPose
  path=AnimGraph.AnimGraphNode_StateMachine_1.ThrottleBar
  ids=id
  nodes [
   ToGroundTrick() {is=title AnimStateAliasNode, also {aliases [FullBodyPose, GroundTrickOutro, Revving, RevvingOutro, Throttle]}}
   AnimStateNode_0 = GroundTrick() {is=title AnimStateNode}
   AnimStateNode_1 = GroundTrickOutro() {is=title AnimStateNode}
   AnimStateNode_12 = Revving() {
    is=title AnimStateNode
    bAlwaysResetOnEntry=true
    stateEntered {guid=#4}
    stateFullyBlended {guid=#5}
    stateLeft {guid=#6}
   }
   AnimStateNode_13 = FullBodyPose() {
    is=title AnimStateNode
    stateEntered {guid=#7}
    stateFullyBlended {guid=#8}
    stateLeft {guid=#9}
   }
   AnimStateNode_14 = RevvingOutro() {
    is=title AnimStateNode
    stateEntered {guid=#10}
    stateFullyBlended {guid=#11}
    stateLeft {guid=#12}
   }
   AnimStateNode_15 = Throttle() {
    is=title AnimStateNode
    bAlwaysResetOnEntry=true
    stateEntered {guid=#13}
    stateFullyBlended {guid=#14}
    stateLeft {guid=#15}
   }
   AnimStateTransitionNode_0 = AnimStateTransitionNode() {crossfadeDuration=0.067}
   AnimStateTransitionNode_1 = AnimStateTransitionNode() {priorityOrder=2}
   AnimStateTransitionNode_10 = AnimStateTransitionNode()
   AnimStateTransitionNode_11 = AnimStateTransitionNode() {crossfadeDuration=0.033}
   AnimStateTransitionNode_14 = AnimStateTransitionNode() {
    crossfadeDuration=0.067
    transitionEnd {guid=#16}
    transitionInterrupt {guid=#17}
    transitionStart {guid=#18}
   }
   AnimStateTransitionNode_15 = AnimStateTransitionNode() {
    crossfadeDuration=0.1
    transitionEnd {guid=#19}
    transitionInterrupt {guid=#20}
    transitionStart {guid=#21}
   }
   AnimStateTransitionNode_21 = AnimStateTransitionNode()
   AnimStateTransitionNode_22 = AnimStateTransitionNode() {bAutomaticRuleBasedOnSequencePlayerInState=true, priorityOrder=2}
   AnimStateTransitionNode_6 = AnimStateTransitionNode() {crossfadeDuration=0.3}
   AnimStateTransitionNode_7 = AnimStateTransitionNode() {priorityOrder=2}
   AnimStateTransitionNode_8 = AnimStateTransitionNode() {bAutomaticRuleBasedOnSequencePlayerInState=true, crossfadeDuration=0.4, priorityOrder=2}
  ]
  links [
   {
    dst=Revving
    rule=AnimGraph.AnimGraphNode_StateMachine_1.ThrottleBar.AnimStateTransitionNode_0.Transition
    src=FullBodyPose
    via=AnimStateTransitionNode_0
   }
   {
    dst=Throttle
    rule=AnimGraph.AnimGraphNode_StateMachine_1.ThrottleBar.AnimStateTransitionNode_1.Transition
    src=FullBodyPose
    via=AnimStateTransitionNode_1
   }
   {
    dst=FullBodyPose
    rule=AnimGraph.AnimGraphNode_StateMachine_1.ThrottleBar.AnimStateTransitionNode_11.Transition
    src=GroundTrick
    via=AnimStateTransitionNode_11
   }
   {
    dst=GroundTrickOutro
    rule=AnimGraph.AnimGraphNode_StateMachine_1.ThrottleBar.AnimStateTransitionNode_7.Transition
    src=GroundTrick
    via=AnimStateTransitionNode_7
   }
   {
    dst=FullBodyPose
    rule=AnimGraph.AnimGraphNode_StateMachine_1.ThrottleBar.AnimStateTransitionNode_10.Transition
    src=GroundTrickOutro
    via=AnimStateTransitionNode_10
   }
   {
    dst=FullBodyPose
    rule=AnimGraph.AnimGraphNode_StateMachine_1.ThrottleBar.AnimStateTransitionNode_8.Transition
    src=GroundTrickOutro
    via=AnimStateTransitionNode_8
   }
   {
    dst=RevvingOutro
    rule=AnimGraph.AnimGraphNode_StateMachine_1.ThrottleBar.AnimStateTransitionNode_14.Transition
    src=Revving
    via=AnimStateTransitionNode_14
   }
   {
    dst=FullBodyPose
    rule=AnimGraph.AnimGraphNode_StateMachine_1.ThrottleBar.AnimStateTransitionNode_21.Transition
    src=RevvingOutro
    via=AnimStateTransitionNode_21
   }
   {
    dst=FullBodyPose
    rule=AnimGraph.AnimGraphNode_StateMachine_1.ThrottleBar.AnimStateTransitionNode_22.Transition
    src=RevvingOutro
    via=AnimStateTransitionNode_22
   }
   {
    dst=RevvingOutro
    rule=AnimGraph.AnimGraphNode_StateMachine_1.ThrottleBar.AnimStateTransitionNode_15.Transition
    src=Throttle
    via=AnimStateTransitionNode_15
   }
   {
    dst=GroundTrick
    rule=AnimGraph.AnimGraphNode_StateMachine_1.ThrottleBar.AnimStateTransitionNode_6.Transition
    src=ToGroundTrick
    via=AnimStateTransitionNode_6
   }
  ]
 }
 GroundTrick {
  path=AnimGraph.AnimGraphNode_StateMachine_1.ThrottleBar.AnimStateNode_0.GroundTrick
  nodes [
   AnimGraphNode_StateMachine_0 = GroundTrickTransition() {is=title AnimGraphNode_StateMachine}
   Output Animation Pose(Result=@AnimGraphNode_StateMachine_0)
  ]
 }
 GroundTrickTransition {
  entry=GroundTrickStart
  path=AnimGraph.AnimGraphNode_StateMachine_1.ThrottleBar.AnimStateNode_0.GroundTrick.AnimGraphNode_StateMachine_0.GroundTrickTransition
  ids=id
  nodes [
   AnimStateNode_0 = GroundTrickStart() {is=title AnimStateNode}
   AnimStateNode_1 = GroundTrickLoop() {is=title AnimStateNode}
   AnimStateTransitionNode_9 = AnimStateTransitionNode() {bAutomaticRuleBasedOnSequencePlayerInState=true}
  ]
  links [{
    dst=GroundTrickLoop
    rule=AnimGraph.AnimGraphNode_StateMachine_1.ThrottleBar.AnimStateNode_0.GroundTrick.AnimGraphNode_StateMachine_0.GroundTrickTransition.AnimStateTransitionNode_9.Transition
    src=GroundTrickStart
    via=AnimStateTransitionNode_9
   }]
 }
 GroundTrickStart {
  path=AnimGraph.AnimGraphNode_StateMachine_1.ThrottleBar.AnimStateNode_0.GroundTrick.AnimGraphNode_StateMachine_0.GroundTrickTransition.AnimStateNode_0.GroundTrickStart
  nodes [Output Animation Pose(Result=Sequence Player 'SportBike_GroundTrick_Intro'() {bLoopAnimation=false, sequence=/Sportbike/Animation/Sportbike/SportBike_GroundTrick_Intro})]
 }
 GroundTrickLoop {
  path=AnimGraph.AnimGraphNode_StateMachine_1.ThrottleBar.AnimStateNode_0.GroundTrick.AnimGraphNode_StateMachine_0.GroundTrickTransition.AnimStateNode_1.GroundTrickLoop
  nodes [Output Animation Pose(Result=Sequence Player 'SportBike_GroundTrick_Loop'() {sequence=/Sportbike/Animation/Sportbike/SportBike_GroundTrick_Loop})]
 }
 Transition {
  path=AnimGraph.AnimGraphNode_StateMachine_1.ThrottleBar.AnimStateNode_0.GroundTrick.AnimGraphNode_StateMachine_0.GroundTrickTransition.AnimStateTransitionNode_9.Transition
  nodes [Result()]
 }
 GroundTrickOutro {
  path=AnimGraph.AnimGraphNode_StateMachine_1.ThrottleBar.AnimStateNode_1.GroundTrickOutro
  nodes [Output Animation Pose(Result=Sequence Player 'SportBike_GroundTrick_Outro'() {bLoopAnimation=false, sequence=/Sportbike/Animation/Sportbike/SportBike_GroundTrick_Outro})]
 }
 Revving {
  path=AnimGraph.AnimGraphNode_StateMachine_1.ThrottleBar.AnimStateNode_12.Revving
  nodes [
   AnimGraphNode_StateMachine_0 = RevvingTransition() {is=title AnimGraphNode_StateMachine}
   Output Animation Pose(Result=@AnimGraphNode_StateMachine_0)
  ]
 }
 RevvingTransition {
  entry=RevvingIntro
  path=AnimGraph.AnimGraphNode_StateMachine_1.ThrottleBar.AnimStateNode_12.Revving.AnimGraphNode_StateMachine_0.RevvingTransition
  ids=id
  nodes [
   AnimStateNode_0 = RevvingIntro() {is=title AnimStateNode}
   AnimStateNode_2 = RevvingLoop() {is=title AnimStateNode}
   AnimStateTransitionNode_7 = AnimStateTransitionNode() {bAutomaticRuleBasedOnSequencePlayerInState=true, crossfadeDuration=0.033}
  ]
  links [{
    dst=RevvingLoop
    rule=AnimGraph.AnimGraphNode_StateMachine_1.ThrottleBar.AnimStateNode_12.Revving.AnimGraphNode_StateMachine_0.RevvingTransition.AnimStateTransitionNode_7.Transition
    src=RevvingIntro
    via=AnimStateTransitionNode_7
   }]
 }
 RevvingIntro {
  path=AnimGraph.AnimGraphNode_StateMachine_1.ThrottleBar.AnimStateNode_12.Revving.AnimGraphNode_StateMachine_0.RevvingTransition.AnimStateNode_0.RevvingIntro
  nodes [Output Animation Pose(Result=Sequence Player 'Sportbike_Revving_Intro_Veh'() {bLoopAnimation=false, sequence=/Sportbike/Animation/Sportbike/Sportbike_Revving_Intro_Veh})]
 }
 RevvingLoop {
  path=AnimGraph.AnimGraphNode_StateMachine_1.ThrottleBar.AnimStateNode_12.Revving.AnimGraphNode_StateMachine_0.RevvingTransition.AnimStateNode_2.RevvingLoop
  nodes [Output Animation Pose(Result=Sequence Player 'Sportbike_Revving_Loop_Veh'() {sequence=/Sportbike/Animation/Sportbike/Sportbike_Revving_Loop_Veh})]
 }
 Transition {
  path=AnimGraph.AnimGraphNode_StateMachine_1.ThrottleBar.AnimStateNode_12.Revving.AnimGraphNode_StateMachine_0.RevvingTransition.AnimStateTransitionNode_7.Transition
  nodes [Result()]
 }
 FullBodyPose {
  path=AnimGraph.AnimGraphNode_StateMachine_1.ThrottleBar.AnimStateNode_13.FullBodyPose
  nodes [Output Animation Pose(Result=Use cached pose 'FullBodyPose'())]
 }
 RevvingOutro {
  path=AnimGraph.AnimGraphNode_StateMachine_1.ThrottleBar.AnimStateNode_14.RevvingOutro
  nodes [Output Animation Pose(Result=Sequence Player 'Sportbike_Revving_Outro_Veh'() {bLoopAnimation=false, sequence=/Sportbike/Animation/Sportbike/Sportbike_Revving_Outro_Veh})]
 }
 Throttle {
  path=AnimGraph.AnimGraphNode_StateMachine_1.ThrottleBar.AnimStateNode_15.Throttle
  nodes [
   AnimGraphNode_StateMachine_0 = ThrottleTransition() {is=title AnimGraphNode_StateMachine, bAllowConduitEntryStates=true}
   Output Animation Pose(Result=@AnimGraphNode_StateMachine_0)
  ]
 }
 ThrottleTransition {
  entry=Conduit
  path=AnimGraph.AnimGraphNode_StateMachine_1.ThrottleBar.AnimStateNode_15.Throttle.AnimGraphNode_StateMachine_0.ThrottleTransition
  ids=id
  nodes [
   AnimStateConduitNode_0 = Conduit()
   AnimStateNode_0 = ThrottleIntro() {is=title AnimStateNode}
   AnimStateNode_2 = ThrottleLoop() {is=title AnimStateNode}
   AnimStateTransitionNode_10 = AnimStateTransitionNode()
   AnimStateTransitionNode_7 = AnimStateTransitionNode() {bAutomaticRuleBasedOnSequencePlayerInState=true}
   AnimStateTransitionNode_9 = AnimStateTransitionNode() {priorityOrder=100}
  ]
  links [
   {
    dst=ThrottleIntro
    rule=AnimGraph.AnimGraphNode_StateMachine_1.ThrottleBar.AnimStateNode_15.Throttle.AnimGraphNode_StateMachine_0.ThrottleTransition.AnimStateTransitionNode_9.Transition
    src=Conduit
    via=AnimStateTransitionNode_9
   }
   {
    dst=ThrottleLoop
    rule=AnimGraph.AnimGraphNode_StateMachine_1.ThrottleBar.AnimStateNode_15.Throttle.AnimGraphNode_StateMachine_0.ThrottleTransition.AnimStateTransitionNode_10.Transition
    src=Conduit
    via=AnimStateTransitionNode_10
   }
   {
    dst=ThrottleLoop
    rule=AnimGraph.AnimGraphNode_StateMachine_1.ThrottleBar.AnimStateNode_15.Throttle.AnimGraphNode_StateMachine_0.ThrottleTransition.AnimStateTransitionNode_7.Transition
    src=ThrottleIntro
    via=AnimStateTransitionNode_7
   }
  ]
 }
 Conduit {
  path=AnimGraph.AnimGraphNode_StateMachine_1.ThrottleBar.AnimStateNode_15.Throttle.AnimGraphNode_StateMachine_0.ThrottleTransition.AnimStateConduitNode_0.Conduit
  nodes [Result(bCanEnterTransition=true)]
 }
 ThrottleIntro {
  path=AnimGraph.AnimGraphNode_StateMachine_1.ThrottleBar.AnimStateNode_15.Throttle.AnimGraphNode_StateMachine_0.ThrottleTransition.AnimStateNode_0.ThrottleIntro
  nodes [Output Animation Pose(Result=Sequence Player 'Sportbike_Throttle_Intro_Veh'() {bLoopAnimation=false, sequence=/Sportbike/Animation/Sportbike/Sportbike_Throttle_Intro_Veh})]
 }
 ThrottleLoop {
  path=AnimGraph.AnimGraphNode_StateMachine_1.ThrottleBar.AnimStateNode_15.Throttle.AnimGraphNode_StateMachine_0.ThrottleTransition.AnimStateNode_2.ThrottleLoop
  nodes [Output Animation Pose(Result=Sequence Player 'Sportbike_Throttle_Loop_Veh'() {bLoopAnimation=false, sequence=/Sportbike/Animation/Sportbike/Sportbike_Throttle_Loop_Veh})]
 }
 Transition {
  path=AnimGraph.AnimGraphNode_StateMachine_1.ThrottleBar.AnimStateNode_15.Throttle.AnimGraphNode_StateMachine_0.ThrottleTransition.AnimStateTransitionNode_10.Transition
  nodes [Result(bCanEnterTransition=Get bStateTransitionRuleToMovingLoop())]
 }
 Transition {
  path=AnimGraph.AnimGraphNode_StateMachine_1.ThrottleBar.AnimStateNode_15.Throttle.AnimGraphNode_StateMachine_0.ThrottleTransition.AnimStateTransitionNode_7.Transition
  nodes [Result()]
 }
 Transition {
  path=AnimGraph.AnimGraphNode_StateMachine_1.ThrottleBar.AnimStateNode_15.Throttle.AnimGraphNode_StateMachine_0.ThrottleTransition.AnimStateTransitionNode_9.Transition
  nodes [Result(bCanEnterTransition=true)]
 }
 Transition {
  path=AnimGraph.AnimGraphNode_StateMachine_1.ThrottleBar.AnimStateTransitionNode_0.Transition
  nodes [Result(bCanEnterTransition=Get bIsRevving() -> (bIsRevving:bool))]
 }
 Transition {
  path=AnimGraph.AnimGraphNode_StateMachine_1.ThrottleBar.AnimStateTransitionNode_1.Transition
  nodes [Result(bCanEnterTransition=Get bIsThrottling() -> (bIsThrottling:bool))]
 }
 Transition {
  path=AnimGraph.AnimGraphNode_StateMachine_1.ThrottleBar.AnimStateTransitionNode_10.Transition
  nodes [Result(bCanEnterTransition=Get ShouldExitFromLocomotion() -> (ShouldExitFromLocomotion:bool))]
 }
 Transition {
  path=AnimGraph.AnimGraphNode_StateMachine_1.ThrottleBar.AnimStateTransitionNode_11.Transition
  nodes [Result(bCanEnterTransition=Get SportBikeWheelieEarlyOut() -> (SportBikeWheelieEarlyOut:bool))]
 }
 Transition {
  path=AnimGraph.AnimGraphNode_StateMachine_1.ThrottleBar.AnimStateTransitionNode_14.Transition
  nodes [Result(bCanEnterTransition=NOT Boolean(A=Get bIsRevving()))]
 }
 Transition {
  path=AnimGraph.AnimGraphNode_StateMachine_1.ThrottleBar.AnimStateTransitionNode_15.Transition
  nodes [Result(bCanEnterTransition=Get bShouldStopThrottling() -> (bShouldStopThrottling:bool))]
 }
 Transition {
  path=AnimGraph.AnimGraphNode_StateMachine_1.ThrottleBar.AnimStateTransitionNode_21.Transition
  nodes [Result(bCanEnterTransition=Get bShouldPlayThrottleAnim() -> (bShouldPlayThrottleAnim:bool))]
 }
 Transition {path=AnimGraph.AnimGraphNode_StateMachine_1.ThrottleBar.AnimStateTransitionNode_22.Transition, nodes [Result()]}
 Transition {
  path=AnimGraph.AnimGraphNode_StateMachine_1.ThrottleBar.AnimStateTransitionNode_6.Transition
  nodes [Result(bCanEnterTransition=Get ShouldPlayGroundTrick() -> (ShouldPlayGroundTrick:bool))]
 }
 Transition {
  path=AnimGraph.AnimGraphNode_StateMachine_1.ThrottleBar.AnimStateTransitionNode_7.Transition
  nodes [Result(bCanEnterTransition=Get StateRuleFromWheelieToGroundTrickOutro() -> (StateRuleFromWheelieToGroundTrickOutro:bool))]
 }
 Transition {path=AnimGraph.AnimGraphNode_StateMachine_1.ThrottleBar.AnimStateTransitionNode_8.Transition, nodes [Result()]}
 MotorcycleAdditives {
  entry=Idle
  path=AnimGraph.AnimGraphNode_StateMachine_2.MotorcycleAdditives
  ids=id
  nodes [
   AnimStateNode_0 = Idle()
   AnimStateNode_1 = GearShifting() {is=title AnimStateNode, bAlwaysResetOnEntry=true}
   AnimStateTransitionNode_11 = AnimStateTransitionNode() {crossfadeDuration=0.1}
   AnimStateTransitionNode_8 = AnimStateTransitionNode() {crossfadeDuration=0.1}
   AnimStateTransitionNode_9 = AnimStateTransitionNode() {bAutomaticRuleBasedOnSequencePlayerInState=true, priorityOrder=2}
  ]
  links [
   {
    dst=Idle
    rule=AnimGraph.AnimGraphNode_StateMachine_2.MotorcycleAdditives.AnimStateTransitionNode_11.Transition
    src=GearShifting
    via=AnimStateTransitionNode_11
   }
   {
    dst=Idle
    rule=AnimGraph.AnimGraphNode_StateMachine_2.MotorcycleAdditives.AnimStateTransitionNode_9.Transition
    src=GearShifting
    via=AnimStateTransitionNode_9
   }
   {
    dst=GearShifting
    rule=AnimGraph.AnimGraphNode_StateMachine_2.MotorcycleAdditives.AnimStateTransitionNode_8.Transition
    src=Idle
    via=AnimStateTransitionNode_8
   }
  ]
 }
 Idle {
  path=AnimGraph.AnimGraphNode_StateMachine_2.MotorcycleAdditives.AnimStateNode_0.Idle
  nodes [Output Animation Pose(Result="Sequence Player (additive) 'Sportbike_Idle_Engine_Veh'"() {sequence=/Sportbike/Animation/Sportbike/Sportbike_Idle_Engine_Veh})]
 }
 GearShifting {
  path=AnimGraph.AnimGraphNode_StateMachine_2.MotorcycleAdditives.AnimStateNode_1.GearShifting
  nodes [Output Animation Pose(Result="Sequence Player (additive) 'SportBike_Driver_GearShiftJolt_CMM_Moto_Sport_Veh'"() {bLoopAnimation=false, sequence=/Sportbike/Animation/Sportbike/SportBike_Driver_GearShiftJolt_CMM_Moto_Sport_Veh})]
 }
 Transition {
  path=AnimGraph.AnimGraphNode_StateMachine_2.MotorcycleAdditives.AnimStateTransitionNode_11.Transition
  nodes [Result(bCanEnterTransition=Get IsChangingGear() -> (IsChangingGear:bool))]
 }
 Transition {
  path=AnimGraph.AnimGraphNode_StateMachine_2.MotorcycleAdditives.AnimStateTransitionNode_8.Transition
  nodes [Result(bCanEnterTransition=Get IsChangingGear())]
 }
 Transition {path=AnimGraph.AnimGraphNode_StateMachine_2.MotorcycleAdditives.AnimStateTransitionNode_9.Transition, nodes [Result()]}
 EventGraph {nodes [Event BlueprintUpdateAnimation()]}
 FinalPoseOverride {nodes [Output Pose(Result=Blend Poses by bool(
     BlendPose_0=Sequence Player 'Sportbike_Parked_Veh'() {sequence=/Sportbike/Animation/Sportbike/Sportbike_Parked_Veh}
     BlendPose_1=Input Pose() {name=InPoseFinalPoseOverride}
     BlendTime_0=0.3
    ) {blendType=Linear}) {layerGroup=None}]}
 PostProcessing {nodes [
   n1 = "Transform (Modify) Bone - Bone: front_differential"(
    ComponentPose=Local To Component(LocalPose=Input Pose() {name=InPosePostProcessing})
    Alpha=1
    Translation=Get AxlePositionF() -> (AxlePositionF:Vector)
    Rotation="0,0,0"
    Scale="1,1,1"
   ) {
    rotationSpace=BCS_BoneSpace
    translationMode=BMM_Replace
    translationSpace=BCS_BoneSpace
    alphaBoolBlend {blendOption=Linear}
    boneToModify {boneName=front_differential}
   }
   n2 = "Transform (Modify) Bone - Bone: rear_differential"(ComponentPose=@n1, Alpha=1, Translation="0,0,0", Rotation=Get SwingArmRotation() -> (SwingArmRotation:Rotator), Scale="1,1,1") {
    rotationMode=BMM_Replace
    rotationSpace=BCS_BoneSpace
    translationSpace=BCS_BoneSpace
    alphaBoolBlend {blendOption=Linear}
    boneToModify {boneName=rear_differential}
   }
   n3 = "Transform (Modify) Bone - Bone: wheel_steering_fr"(ComponentPose=@n2, Alpha=1, Translation="0,0,0", Rotation=Get LimitedWheelRotationFR() -> (LimitedWheelRotationFR:Rotator), Scale="1,1,1") {
    rotationMode=BMM_Replace
    rotationSpace=BCS_BoneSpace
    translationSpace=BCS_BoneSpace
    alphaBoolBlend {blendOption=Linear}
    boneToModify {boneName=wheel_steering_fr}
   }
   n4 = "Transform (Modify) Bone - Bone: wheel_tilted_fr"(ComponentPose=@n3, Alpha=1, Translation="0,0,0", Rotation=Get WheelSpinFR() -> (WheelSpinFR:Rotator), Scale="1,1,1") {
    rotationMode=BMM_Replace
    rotationSpace=BCS_BoneSpace
    translationSpace=BCS_BoneSpace
    alphaBoolBlend {blendOption=Linear}
    boneToModify {boneName=wheel_tilted_fr}
   }
   n5 = "Transform (Modify) Bone - Bone: wheel_tilted_bk"(ComponentPose=@n4, Alpha=1, Translation="0,0,0", Rotation=Get WheelSpinBR() -> (WheelSpinBR:Rotator), Scale="1,1,1") {
    rotationMode=BMM_Replace
    rotationSpace=BCS_BoneSpace
    translationSpace=BCS_BoneSpace
    alphaBoolBlend {blendOption=Linear}
    boneToModify {boneName=wheel_tilted_bk}
   }
   Output Pose(Result=Component To Local(ComponentPose=@n5)) {layerGroup=None}
   Get WheelRotationFR() -> (WheelRotationFR:Rotator)
  ]}
]
variables []
