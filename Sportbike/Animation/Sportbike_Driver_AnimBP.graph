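# Animation Blueprint for a Fortnite player driving the Sportbike, supplying the vehicle full-body override layer.
# A riding state machine (locomotion, jumps, boost, tricks) blends with an upper-body machine, then gets riding additives, slope warping and hand IK via Leg IK nodes.
# A combat branch with aim offsets and additives swaps in through an inertialized blend, followed by pelvis adjustment.
# State notifies set jump flags; the audio controller's gear-change event sets the gear.
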AnimBlueprint /Sportbike/Animation/Sportbike_Driver_AnimBP
target_skeleton=&Fortnite_M_Avg_Player_Skeleton
graphs [
 AdjustmentGroup {nodes [
   n1 = "Transform (Modify) Bone - Bone: pelvis"!(ComponentPose=Local To Component(LocalPose=Input Pose() {name=InPoseAdjustment}), Alpha=0.25, Rotation="0,0,0") -> (Pose) {
    rotationMode=BMM_Additive
    rotationSpace=BCS_WorldSpace
    translationSpace=BCS_WorldSpace
    alphaBoolBlend {blendOption=Linear}
    boneToModify {boneName=pelvis}
   }
   Output Pose(Result=Component To Local(ComponentPose=@n1)) {layerGroup=None}
  ]}
 AnimGraph {nodes [Output Pose() {layerGroup=None}]}
 CachedPose {nodes [
   n1 = Input Pose() {name=InPoseCached}
   n2 = Sequence Player 'MotoDirt_Driver_Loco_Pose_CMM'() {sequence=/Motorcycle/Animation/Driver/MotoDirt_Driver_Loco_Pose_CMM}
   Output Pose(Result=Blend Poses by bool(
     BlendPose_0=Sequence Player 'MotoDirt_Driver_Loco_Pose_CMM'() {sequence=&MotoDirt_Driver_Loco_Pose_CMM}
     BlendPose_1=Apply Additive(Base=@n2, Additive=@n1) {alphaBoolBlend {blendOption=Linear}}
     BlendTime_0=0.2
     BlendTime_1=0.2
    ) {blendType=Linear}) {layerGroup=None}
  ]}
 CombatAdditiveGroup {nodes [
   n1 = Input Pose() {name=InPoseCombatAdditive}
   AnimGraphNode_StateMachine_2 = CombatUpperBodyAdditive() {is=title AnimGraphNode_StateMachine}
   AnimGraphNode_StateMachine_4 = MeleeOffset() {is=title AnimGraphNode_StateMachine}
   Output Pose(Result=Apply Additive!(
     Base=Apply Additive!(Base=@n1, Additive=@AnimGraphNode_StateMachine_2, Alpha:float) -> (Pose) {alphaBoolBlend {blendOption=Linear}}
     Additive=@AnimGraphNode_StateMachine_4
     bAlphaBoolEnabled:bool=True
    ) -> (Pose) {alphaInputType=Bool, alphaBoolBlend {blendInTime=0.25, blendOption=Linear, blendOutTime=0.25}}) {layerGroup=None}
  ]}
 CombatUpperBodyAdditive {
  entry=IdleAdd
  path=CombatAdditiveGroup.AnimGraphNode_StateMachine_2.CombatUpperBodyAdditive
  ids=id
  nodes [
   AnimStateNode_0 = IdleAdd() {is=title AnimStateNode}
   AnimStateNode_1 = RidingAdd() {is=title AnimStateNode}
   AnimStateTransitionNode_0 = AnimStateTransitionNode()
   AnimStateTransitionNode_1 = AnimStateTransitionNode()
  ]
  links [
   {
    dst=RidingAdd
    rule=CombatAdditiveGroup.AnimGraphNode_StateMachine_2.CombatUpperBodyAdditive.AnimStateTransitionNode_0.Transition
    src=IdleAdd
    via=AnimStateTransitionNode_0
   }
   {
    dst=IdleAdd
    rule=CombatAdditiveGroup.AnimGraphNode_StateMachine_2.CombatUpperBodyAdditive.AnimStateTransitionNode_1.Transition
    src=RidingAdd
    via=AnimStateTransitionNode_1
   }
  ]
 }
 IdleAdd {
  path=CombatAdditiveGroup.AnimGraphNode_StateMachine_2.CombatUpperBodyAdditive.AnimStateNode_0.IdleAdd
  nodes [Output Animation Pose(Result="Sequence Player (additive) 'MotoDirt_Driver_Combat_Idle'"() {sequence=/Motorcycle/Animation/Driver/MotoDirt_Driver_Combat_Idle})]
 }
 RidingAdd {
  path=CombatAdditiveGroup.AnimGraphNode_StateMachine_2.CombatUpperBodyAdditive.AnimStateNode_1.RidingAdd
  nodes [Output Animation Pose(Result=Blendspace Player 'MotoDirt_Driver_CombatMoving_BS'!(X=0, PlayRate:float=1) -> (Pose) {blendSpace=/Motorcycle/Animation/Driver/BlendSpace/MotoDirt_Driver_CombatMoving_BS})]
 }
 Transition {
  path=CombatAdditiveGroup.AnimGraphNode_StateMachine_2.CombatUpperBodyAdditive.AnimStateTransitionNode_0.Transition
  nodes [Result(bCanEnterTransition=Get bIsMoving() -> (bIsMoving:bool))]
 }
 Transition {
  path=CombatAdditiveGroup.AnimGraphNode_StateMachine_2.CombatUpperBodyAdditive.AnimStateTransitionNode_1.Transition
  nodes [Result(bCanEnterTransition=NOT Boolean(A=Get bIsMoving()))]
 }
 MeleeOffset {
  entry=MeleeUpperBodyOffset
  path=CombatAdditiveGroup.AnimGraphNode_StateMachine_4.MeleeOffset
  ids=id
  nodes [AnimStateNode_0 = MeleeUpperBodyOffset() {is=title AnimStateNode}]
 }
 MeleeUpperBodyOffset {
  path=CombatAdditiveGroup.AnimGraphNode_StateMachine_4.MeleeOffset.AnimStateNode_0.MeleeUpperBodyOffset
  nodes [Output Animation Pose(Result=Blend Poses by bool(
     BlendPose_0=Blendspace Player 'MotoDirt_MeleeOffset_1H2Swing_BS'!(X=0) -> (Pose) {blendSpace=/Motorcycle/Animation/Driver/MeleeOffset/MotoDirt_MeleeOffset_1H2Swing_BS}
     BlendPose_1=Additive Identity Pose()
    ) {blendType=Linear})]
 }
 CombatAimOffsetGroup {nodes [
   n1 = Input Pose() {name=InPoseCombatAimOffset}
   n2 = "Sequence Player (additive) 'Sportbike_Driver_Combat_CorrectivePelvis'"() {sequence=/Sportbike/Animation/Driver/Sportbike_Driver_Combat_CorrectivePelvis}
   n3 = AimOffset Player 'MotoDirt_Driver_AO_NoHip_Yaw'!(
    X:float=0
    BasePose:Pose Link Structure=AimOffset Player 'MotoDirt_Driver_AO_Yaw'!(
     X:float=0
     BasePose:Pose Link Structure=Apply Additive(Base=@n1, Additive=@n2) {alphaBoolBlend {blendOption=Linear}}
     bAlphaBoolEnabled:bool=False
    ) -> (Pose:Pose Link Structure) {alphaInputType=Bool, blendSpace=/Motorcycle/Animation/Driver/AO/MotoDirt_Driver_AO_Yaw, alphaBoolBlend {blendOption=Linear}}
    bAlphaBoolEnabled:bool=False
   ) -> (Pose:Pose Link Structure) {alphaInputType=Bool, blendSpace=/Motorcycle/Animation/Driver/AO/MotoDirt_Driver_AO_NoHip_Yaw, alphaBoolBlend {blendOption=Linear}}
   Output Pose(Result=AimOffset Player 'MotoDirt_Driver_Roll_AO'!(
     X:float=0
     BasePose:Pose Link Structure=AimOffset Player 'MotoDirt_Driver_Pitch_AO'!(X:float=0, BasePose:Pose Link Structure=@n3, Alpha:float=1) -> (Pose:Pose Link Structure) {blendSpace=/Motorcycle/Animation/Driver/AO/MotoDirt_Driver_Pitch_AO, alphaBoolBlend {blendOption=Linear}}
     Alpha:float=1
    ) -> (Pose:Pose Link Structure) {blendSpace=/Motorcycle/Animation/Driver/AO/MotoDirt_Driver_Roll_AO, alphaBoolBlend {blendOption=Linear}}) {layerGroup=None}
  ]}
 EventGraph {nodes [
   n1 = OnGearChangedEvent_Event_0!() -> (OutputDelegate, then, NewGear:int, IgnoreExhaustPops:bool=~) {is=title K2Node_CustomEvent}
   > Set MotorcycleGear(MotorcycleGear:int=@n1.NewGear) -> (Output_Get:int=0)
   n2 = AnimNotify_NeutralToRightExit() {is=title K2Node_Event}
   Set bWasJumpTrickLeftOrRight(execute=[@n2.then, @n7.then, @n8.then, @n9.then, @n10.then, @n15.then], bWasJumpTrickLeftOrRight:bool=true) -> (Output_Get:bool)
   AnimNotify_StoppieLandCheckON() {is=title K2Node_Event}
   > Set bWasStoppieLanding(bWasStoppieLanding:bool=true) -> (Output_Get:bool)
   AnimNotify_StoppieLandCheckOFF() {is=title K2Node_Event}
   > Set bWasStoppieLanding(bWasStoppieLanding=false)
   AnimNotify_LocomotionExit() {is=title K2Node_Event}
   > Set JumpTimer(JumpTimer:float=0) -> (Output_Get:float)
   n3 = AnimNotify_JumpTrickOutro() {is=title K2Node_Event}
   Set bWasJumpTrickLeftOrRight(execute=[@n3.then, @n11.then, @n14.then], bWasJumpTrickLeftOrRight=false)
   > Set bIsFromJumpTrick(bIsFromJumpTrick:bool=false) -> (Output_Get:bool)
   AnimNotify_MovingLandExit() {is=title K2Node_Event}
   > n4 = Sequence()
   > Set bShouldPlayMovingJump(bShouldPlayMovingJump:bool=false) -> (Output_Get:bool)
   > Set bWasJumpTrickLeftOrRight(bWasJumpTrickLeftOrRight=false)
   > Set JumpTimer(JumpTimer=0)
   Set bIsFromJumpTrick(execute=@n4.then_1, bIsFromJumpTrick=false)
   > Set bWasSoloJumping(bWasSoloJumping:bool=false) -> (Output_Get:bool)
   AnimNotify_JumpFallExit() {is=title K2Node_Event}
   > Set bShouldPlayMovingJump(bShouldPlayMovingJump=true)
   Event BlueprintInitializeAnimation()
   > n5 = Cast To Motorcycle_Sport_Vehicle!(execute, Object=Get Motorcycle() -> (Motorcycle:Fort Motorcycle Vehicle Object Reference)) -> (then, CastFailed, AsMotorcycle Sport Vehicle:Motorcycle Sport Vehicle Object Reference) {is=title K2Node_DynamicCast}
   n6 = Cast To BP_VehicleCosmetics_AudioController_Parent!(Object=Get ChildActor!(self:Child Actor Component Object Reference=Get Motorcycle_AudioController!(self:Motorcycle Sport Vehicle Object Reference=@n5) -> (Motorcycle_AudioController:Child Actor Component Object Reference)) -> (ChildActor:Actor Object Reference)) -> (AsBP Vehicle Cosmetics Audio Controller Parent:BP Vehicle Cosmetics Audio Controller Parent Object Reference, bSuccess:bool=~) {is=title K2Node_DynamicCast}
   > Is Valid(InputObject=@n6)
   > Assign On Gear Changed Event!(execute:Exec, self:BP Vehicle Cosmetics Audio Controller Parent Object Reference=@n6, Delegate:"Delegate (by ref)"=@n1) -> (then:Exec) {is=title K2Node_AssignDelegate}
   AnimNotify_JumpTrickExit() {is=title K2Node_Event}
   > Set bIsFromJumpTrick(bIsFromJumpTrick=true)
   > Set bShouldPlayMovingJump(bShouldPlayMovingJump=true)
   AnimNotify_JumpCharge() {is=title K2Node_Event}
   > Set bHasJumpChargeStarted(bHasJumpChargeStarted:bool=true) -> (Output_Get:bool)
   n7 = AnimNotify_JumpTrickLeftExit() {is=title K2Node_Event}
   n8 = AnimNotify_JumpTrickRightExit() {is=title K2Node_Event}
   n9 = AnimNotify_LeftIntroExit() {is=title K2Node_Event}
   n10 = AnimNotify_RightIntroExit() {is=title K2Node_Event}
   n11 = AnimNotify_NeutralEnter() {is=title K2Node_Event}
   n12 = AnimNotify_NormalJumpLoopEnter() {is=title K2Node_Event}
   Set bIsFromJumpTrick(execute=[@n12.then, @n13.then], bIsFromJumpTrick=false)
   AnimNotify_JumpTrickEnter() {is=title K2Node_Event}
   > Set bShouldPlayMovingJump(bShouldPlayMovingJump=false)
   AnimNotify_IdleEnter() {is=title K2Node_Event}
   > Set bWasJumpTrickLeftOrRight(bWasJumpTrickLeftOrRight=false)
   AnimNotify_JumpLandEnter() {is=title K2Node_Event}
   > Set bShouldPlayMovingJump(bShouldPlayMovingJump=false)
   > Set bWasJumpTrickLeftOrRight(bWasJumpTrickLeftOrRight=false)
   > Set JumpTimer(JumpTimer=0)
   n13 = AnimNotify_Idle() {is=title K2Node_Event}
   AnimNotify_MoveStopEnter() {is=title K2Node_Event}
   > Set bShouldPlayMovingJump(bShouldPlayMovingJump=false)
   > Set bWasJumpTrickLeftOrRight(bWasJumpTrickLeftOrRight=false)
   > Set bIsFromJumpTrick(bIsFromJumpTrick=false)
   n14 = AnimNotify_JumpEnter() {is=title K2Node_Event}
   AnimNotify_SoloJumpTrickEnter() {is=title K2Node_Event}
   > Set bWasSoloJumping(bWasSoloJumping=true)
   AnimNotify_CombatTransitionON() {is=title K2Node_Event}
   > Set bIsCombatTransition(bIsCombatTransition:bool=true) -> (Output_Get:bool)
   AnimNotify_CombatTransitionOFF() {is=title K2Node_Event}
   > Set bIsCombatTransition(bIsCombatTransition=false)
   > Set bHasJumpChargeStarted(bHasJumpChargeStarted=false)
   n15 = AnimNotify_NeutralToLeftExit() {is=title K2Node_Event}
   Event BlueprintUpdateAnimation()
  ]}
 IK/BoneModification {nodes [
   n1 = Predictive Biped Slope Warping!(ComponentPose:Component Space Pose Link Structure=Local To Component(LocalPose=Input Pose() {name=InPoseIK}), Alpha:float=1) -> (Pose:Component Space Pose Link Structure) {
    is=title FortAnimGraphNode_PredictiveSlopeWarping
    bAbsentCurveMeansFootDown=false
    bLeftCurveHighMeansFootDown=false
    bRightCurveHighMeansFootDown=false
    footTraceChannel=ECC_Vehicle
    lFootOnGroundCurveName=LeftLegIKOnGroundVehicle
    pullPelvisDownMaxDistance=0.001
    rFootOnGroundCurveName=RightLegIKOnGroundVehicle
    alphaBoolBlend {blendOption=Linear}
    feetDefinitions [
     {numBonesInLimb=2, fKFootBone {boneName=foot_l}, iKFootBone {boneName=ik_foot_l}}
     {numBonesInLimb=2, fKFootBone {boneName=foot_r}, iKFootBone {boneName=ik_foot_r}}
    ]
    iKFootRootBone {boneName=ik_foot_root}
    lFootPrediction {boneName=ik_foot_l}
    pelvisBone {boneName=pelvis}
    rFootPrediction {boneName=ik_foot_r}
   }
   n2 = "Transform (Modify) Bone - Bone: ik_hand_l"!(ComponentPose=@n1, AlphaCurveName:Name=LeftHandFK, Translation="0,50,0", Rotation="0,0,0") -> (Pose) {
    alphaInputType=Curve
    rotationMode=BMM_Replace
    rotationSpace=BCS_WorldSpace
    translationMode=BMM_Replace
    translationSpace=BCS_WorldSpace
    alphaBoolBlend {blendOption=Linear}
    alphaScaleBiasClamp {bias=1, scale=-1}
    boneToModify {boneName=ik_hand_l}
    translation {y=50}
   }
   n3 = "Transform (Modify) Bone - Bone: ik_hand_r"!(ComponentPose=@n2, AlphaCurveName:Name=RightHandFK, Translation="0,0,0", Rotation="0,0,0") -> (Pose) {
    alphaInputType=Curve
    rotationMode=BMM_Replace
    rotationSpace=BCS_WorldSpace
    translationMode=BMM_Replace
    translationSpace=BCS_WorldSpace
    alphaBoolBlend {blendOption=Linear}
    alphaScaleBiasClamp {bias=1, scale=-1}
    boneToModify {boneName=ik_hand_r}
   }
   n4 = Leg IK!(ComponentPose:Component Space Pose Link Structure=@n3, Alpha:float=1) -> (Pose:Component Space Pose Link Structure) {
    is=title AnimGraphNode_LegIK
    alphaBoolBlend {blendOption=Linear}
    legsDefinition [{
      bEnableKneeTwistCorrection=false
      bEnableRotationLimit=false
      footBoneForwardAxis=Y
      hingeRotationAxis=Z
      minRotationAngle=15
      numBonesInLimb=2
      twistOffsetCurveName=None
      fKFootBone {boneName=hand_r}
      iKFootBone {boneName=ik_hand_r}
     }]
   }
   n5 = Leg IK(ComponentPose=@n4, Alpha=1) {
    alphaBoolBlend {blendOption=Linear}
    legsDefinition [{
      bEnableKneeTwistCorrection=false
      bEnableRotationLimit=false
      footBoneForwardAxis=Y
      hingeRotationAxis=Z
      minRotationAngle=15
      numBonesInLimb=2
      twistOffsetCurveName=None
      fKFootBone {boneName=hand_l}
      iKFootBone {boneName=ik_hand_l}
     }]
   }
   Output Pose(Result=Component To Local(ComponentPose=@n5)) {layerGroup=None}
  ]}
 RidingAdditiveGroup {nodes [
   n1 = Blendspace Player 'MotoDirt_Driver_StoppieLeans_BS'(X=0, Y=0) {blendSpace=/Motorcycle/Animation/Driver/BlendSpace/MotoDirt_Driver_StoppieLeans_BS}
   n2 = Input Pose() {name=InPoseRidingAdditive}
   AnimGraphNode_StateMachine_3 = RidingAdditive() {is=title AnimGraphNode_StateMachine, bAllowConduitEntryStates=true}
   Output Pose(Result=Apply Additive(Base=Apply Additive(Base=@n2, Additive=@n1) {alphaBoolBlend {blendInTime=0.2, blendOutTime=0.35}}, Additive=@AnimGraphNode_StateMachine_3) {alphaBoolBlend {blendOption=Linear}}) {layerGroup=None}
  ]}
 RidingAdditive {
  entry=Conduit
  path=RidingAdditiveGroup.AnimGraphNode_StateMachine_3.RidingAdditive
  ids=id
  nodes [
   AnimStateConduitNode_0 = Conduit()
   AnimStateNode_0 = AdditivePose() {is=title AnimStateNode}
   AnimStateNode_1 = Entry/RideModeTransition() {is=title AnimStateNode}
   AnimStateNode_2 = MovingNoise() {is=title AnimStateNode}
   AnimStateNode_3 = HitReact() {is=title AnimStateNode}
   AnimStateNode_4 = Revving() {is=title AnimStateNode}
   AnimStateNode_5 = RevvingOutro() {is=title AnimStateNode}
   AnimStateNode_6 = GearShifting() {is=title AnimStateNode, bAlwaysResetOnEntry=true}
   AnimStateTransitionNode_0 = AnimStateTransitionNode() {crossfadeDuration=0.033}
   AnimStateTransitionNode_1 = AnimStateTransitionNode() {priorityOrder=2}
   AnimStateTransitionNode_14 = AnimStateTransitionNode() {crossfadeDuration=0.1}
   AnimStateTransitionNode_15 = AnimStateTransitionNode() {bAutomaticRuleBasedOnSequencePlayerInState=true, priorityOrder=2}
   AnimStateTransitionNode_16 = AnimStateTransitionNode() {crossfadeDuration=0.033}
   AnimStateTransitionNode_17 = AnimStateTransitionNode() {crossfadeDuration=0.1}
   AnimStateTransitionNode_2 = AnimStateTransitionNode()
   AnimStateTransitionNode_3 = AnimStateTransitionNode() {bAutomaticRuleBasedOnSequencePlayerInState=true}
   AnimStateTransitionNode_4 = AnimStateTransitionNode() {priorityOrder=100}
   AnimStateTransitionNode_5 = AnimStateTransitionNode() {bAutomaticRuleBasedOnSequencePlayerInState=true}
   AnimStateTransitionNode_6 = AnimStateTransitionNode() {crossfadeDuration=0.067}
   AnimStateTransitionNode_7 = AnimStateTransitionNode() {crossfadeDuration=0.067}
   AnimStateTransitionNode_8 = AnimStateTransitionNode() {bAutomaticRuleBasedOnSequencePlayerInState=true, priorityOrder=2}
   AnimStateTransitionNode_9 = AnimStateTransitionNode()
  ]
  links [
   {
    dst=HitReact
    rule=RidingAdditiveGroup.AnimGraphNode_StateMachine_3.RidingAdditive.AnimStateTransitionNode_0.Transition
    src=AdditivePose
    via=AnimStateTransitionNode_0
   }
   {
    dst=MovingNoise
    rule=RidingAdditiveGroup.AnimGraphNode_StateMachine_3.RidingAdditive.AnimStateTransitionNode_2.Transition
    src=AdditivePose
    via=AnimStateTransitionNode_2
   }
   {
    dst=Revving
    rule=RidingAdditiveGroup.AnimGraphNode_StateMachine_3.RidingAdditive.AnimStateTransitionNode_6.Transition
    src=AdditivePose
    via=AnimStateTransitionNode_6
   }
   {
    dst=Entry/RideModeTransition
    rule=RidingAdditiveGroup.AnimGraphNode_StateMachine_3.RidingAdditive.AnimStateTransitionNode_4.Transition
    src=Conduit
    via=AnimStateTransitionNode_4
   }
   {
    dst=AdditivePose
    rule=RidingAdditiveGroup.AnimGraphNode_StateMachine_3.RidingAdditive.AnimStateTransitionNode_3.Transition
    src=Entry/RideModeTransition
    via=AnimStateTransitionNode_3
   }
   {
    dst=HitReact
    rule=RidingAdditiveGroup.AnimGraphNode_StateMachine_3.RidingAdditive.AnimStateTransitionNode_16.Transition
    src=GearShifting
    via=AnimStateTransitionNode_16
   }
   {
    dst=MovingNoise
    rule=RidingAdditiveGroup.AnimGraphNode_StateMachine_3.RidingAdditive.AnimStateTransitionNode_15.Transition
    src=GearShifting
    via=AnimStateTransitionNode_15
   }
   {
    dst=MovingNoise
    rule=RidingAdditiveGroup.AnimGraphNode_StateMachine_3.RidingAdditive.AnimStateTransitionNode_17.Transition
    src=GearShifting
    via=AnimStateTransitionNode_17
   }
   {
    dst=AdditivePose
    rule=RidingAdditiveGroup.AnimGraphNode_StateMachine_3.RidingAdditive.AnimStateTransitionNode_5.Transition
    src=HitReact
    via=AnimStateTransitionNode_5
   }
   {
    dst=AdditivePose
    rule=RidingAdditiveGroup.AnimGraphNode_StateMachine_3.RidingAdditive.AnimStateTransitionNode_1.Transition
    src=MovingNoise
    via=AnimStateTransitionNode_1
   }
   {
    dst=GearShifting
    rule=RidingAdditiveGroup.AnimGraphNode_StateMachine_3.RidingAdditive.AnimStateTransitionNode_14.Transition
    src=MovingNoise
    via=AnimStateTransitionNode_14
   }
   {
    dst=RevvingOutro
    rule=RidingAdditiveGroup.AnimGraphNode_StateMachine_3.RidingAdditive.AnimStateTransitionNode_7.Transition
    src=Revving
    via=AnimStateTransitionNode_7
   }
   {
    dst=AdditivePose
    rule=RidingAdditiveGroup.AnimGraphNode_StateMachine_3.RidingAdditive.AnimStateTransitionNode_8.Transition
    src=RevvingOutro
    via=AnimStateTransitionNode_8
   }
   {
    dst=AdditivePose
    rule=RidingAdditiveGroup.AnimGraphNode_StateMachine_3.RidingAdditive.AnimStateTransitionNode_9.Transition
    src=RevvingOutro
    via=AnimStateTransitionNode_9
   }
  ]
 }
 Conduit {
  path=RidingAdditiveGroup.AnimGraphNode_StateMachine_3.RidingAdditive.AnimStateConduitNode_0.Conduit
  nodes [Result(bCanEnterTransition=true)]
 }
 AdditivePose {
  path=RidingAdditiveGroup.AnimGraphNode_StateMachine_3.RidingAdditive.AnimStateNode_0.AdditivePose
  nodes [
   n1 = Blendspace Player 'MotoDirt_Driver_FeetDrag_BS'!(X=0) -> (Pose) {blendSpace=/Motorcycle/Animation/Driver/BlendSpace/MotoDirt_Driver_FeetDrag_BS}
   n2 = Additive Identity Pose()
   n3 = AND Boolean(A=Get bShouldApplyFeetNoise() -> (bShouldApplyFeetNoise:bool), B=NOT Boolean(A=Get bStoppie() -> (bStoppie:bool)))
   Output Animation Pose(Result=Apply Additive!(
     Base=Apply Additive(Base=@n2, Additive=@n1, bAlphaBoolEnabled=@n3) {alphaInputType=Bool, alphaBoolBlend {blendInTime=0.2, blendOutTime=0.3}}
     Additive="Sequence Player (additive) 'MotoDirt_Driver_FeetDrag_Noise'"!(PlayRate:float=1) -> (Pose) {sequence=/Motorcycle/Animation/Driver/MotoDirt_Driver_FeetDrag_Noise}
     bAlphaBoolEnabled:bool=@n3
    ) -> (Pose) {alphaInputType=Bool, alphaBoolBlend {blendInTime=0.2, blendOutTime=0.3}})
   Get bShouldApplyFeetNoise()
   Get bShouldApplyFeetNoise()
  ]
 }
 Entry/RideModeTransition {
  path=RidingAdditiveGroup.AnimGraphNode_StateMachine_3.RidingAdditive.AnimStateNode_1.Entry/RideModeTransition
  nodes [Output Animation Pose(Result="Sequence Player (additive) 'MotoDirt_Driver_CombatToRide'"() {bLoopAnimation=false, sequence=/Motorcycle/Animation/Driver/MotoDirt_Driver_CombatToRide})]
 }
 MovingNoise {
  path=RidingAdditiveGroup.AnimGraphNode_StateMachine_3.RidingAdditive.AnimStateNode_2.MovingNoise
  nodes [Output Animation Pose(Result=Blendspace Player 'MotoDirt_Driver_Moving_BS'!(X=0, PlayRate:float=1) -> (Pose) {blendSpace=/Motorcycle/Animation/Driver/BlendSpace/MotoDirt_Driver_Moving_BS})]
 }
 HitReact {
  path=RidingAdditiveGroup.AnimGraphNode_StateMachine_3.RidingAdditive.AnimStateNode_3.HitReact
  nodes [Output Animation Pose(Result="Sequence Player (additive) 'MotoDirt_Driver_Collision_N'"() {bLoopAnimation=false, sequence=/Motorcycle/Animation/Driver/MotoDirt_Driver_Collision_N})]
 }
 Revving {
  path=RidingAdditiveGroup.AnimGraphNode_StateMachine_3.RidingAdditive.AnimStateNode_4.Revving
  nodes [
   AnimGraphNode_StateMachine_0 = RevvingTransition() {is=title AnimGraphNode_StateMachine}
   Output Animation Pose(Result=@AnimGraphNode_StateMachine_0)
  ]
 }
 RevvingTransition {
  entry=RevvingIntro
  path=RidingAdditiveGroup.AnimGraphNode_StateMachine_3.RidingAdditive.AnimStateNode_4.Revving.AnimGraphNode_StateMachine_0.RevvingTransition
  ids=id
  nodes [
   AnimStateNode_0 = RevvingIntro() {is=title AnimStateNode}
   AnimStateNode_1 = RevvingLoop() {is=title AnimStateNode}
   AnimStateTransitionNode_1 = AnimStateTransitionNode() {bAutomaticRuleBasedOnSequencePlayerInState=true, crossfadeDuration=0.033}
  ]
  links [{
    dst=RevvingLoop
    rule=RidingAdditiveGroup.AnimGraphNode_StateMachine_3.RidingAdditive.AnimStateNode_4.Revving.AnimGraphNode_StateMachine_0.RevvingTransition.AnimStateTransitionNode_1.Transition
    src=RevvingIntro
    via=AnimStateTransitionNode_1
   }]
 }
 RevvingIntro {
  path=RidingAdditiveGroup.AnimGraphNode_StateMachine_3.RidingAdditive.AnimStateNode_4.Revving.AnimGraphNode_StateMachine_0.RevvingTransition.AnimStateNode_0.RevvingIntro
  nodes [Output Animation Pose(Result="Sequence Player (additive) 'MotoDirt_Driver_Revving_Intro'"() {bLoopAnimation=false, sequence=/Motorcycle/Animation/Driver/Revving/MotoDirt_Driver_Revving_Intro})]
 }
 RevvingLoop {
  path=RidingAdditiveGroup.AnimGraphNode_StateMachine_3.RidingAdditive.AnimStateNode_4.Revving.AnimGraphNode_StateMachine_0.RevvingTransition.AnimStateNode_1.RevvingLoop
  nodes [Output Animation Pose(Result="Sequence Player (additive) 'MotoDirt_Driver_Revving_Loop'"() {sequence=/Motorcycle/Animation/Driver/Revving/MotoDirt_Driver_Revving_Loop})]
 }
 Transition {
  path=RidingAdditiveGroup.AnimGraphNode_StateMachine_3.RidingAdditive.AnimStateNode_4.Revving.AnimGraphNode_StateMachine_0.RevvingTransition.AnimStateTransitionNode_1.Transition
  nodes [Result()]
 }
 RevvingOutro {
  path=RidingAdditiveGroup.AnimGraphNode_StateMachine_3.RidingAdditive.AnimStateNode_5.RevvingOutro
  nodes [Output Animation Pose(Result="Sequence Player (additive) 'MotoDirt_Driver_Revving_Outro'"() {bLoopAnimation=false, sequence=/Motorcycle/Animation/Driver/Revving/MotoDirt_Driver_Revving_Outro})]
 }
 GearShifting {
  path=RidingAdditiveGroup.AnimGraphNode_StateMachine_3.RidingAdditive.AnimStateNode_6.GearShifting
  nodes [
   "Sequence Player (additive) 'MotoDirt_Driver_Collision_N'"() {bLoopAnimation=false, sequence=&MotoDirt_Driver_Collision_N}
   Output Animation Pose(Result="Sequence Player (additive) 'SportBike_Driver_GearShiftJolt_CMM'"() {bLoopAnimation=false, sequence=/Sportbike/Animation/Driver/SportBike_Driver_GearShiftJolt_CMM})
  ]
 }
 Transition {
  path=RidingAdditiveGroup.AnimGraphNode_StateMachine_3.RidingAdditive.AnimStateTransitionNode_0.Transition
  nodes [Result(bCanEnterTransition=Get bShouldPlayHitReact() -> (bShouldPlayHitReact:bool))]
 }
 Transition {
  path=RidingAdditiveGroup.AnimGraphNode_StateMachine_3.RidingAdditive.AnimStateTransitionNode_1.Transition
  nodes [Result(bCanEnterTransition=NOT Boolean(A=Get bIsMoving()))]
 }
 Transition {
  path=RidingAdditiveGroup.AnimGraphNode_StateMachine_3.RidingAdditive.AnimStateTransitionNode_14.Transition
  nodes [Result(bCanEnterTransition=Get IsChangingGear() -> (IsChangingGear:bool))]
 }
 Transition {path=RidingAdditiveGroup.AnimGraphNode_StateMachine_3.RidingAdditive.AnimStateTransitionNode_15.Transition, nodes [Result()]}
 Transition {
  path=RidingAdditiveGroup.AnimGraphNode_StateMachine_3.RidingAdditive.AnimStateTransitionNode_16.Transition
  nodes [Result(bCanEnterTransition=Get bShouldPlayHitReact())]
 }
 Transition {
  path=RidingAdditiveGroup.AnimGraphNode_StateMachine_3.RidingAdditive.AnimStateTransitionNode_17.Transition
  nodes [Result(bCanEnterTransition=Get IsChangingGear())]
 }
 Transition {
  path=RidingAdditiveGroup.AnimGraphNode_StateMachine_3.RidingAdditive.AnimStateTransitionNode_2.Transition
  nodes [Result(bCanEnterTransition=Get bIsMoving())]
 }
 Transition {path=RidingAdditiveGroup.AnimGraphNode_StateMachine_3.RidingAdditive.AnimStateTransitionNode_3.Transition, nodes [Result()]}
 Transition {
  path=RidingAdditiveGroup.AnimGraphNode_StateMachine_3.RidingAdditive.AnimStateTransitionNode_4.Transition
  nodes [Result(bCanEnterTransition=true)]
 }
 Transition {path=RidingAdditiveGroup.AnimGraphNode_StateMachine_3.RidingAdditive.AnimStateTransitionNode_5.Transition, nodes [Result()]}
 Transition {
  path=RidingAdditiveGroup.AnimGraphNode_StateMachine_3.RidingAdditive.AnimStateTransitionNode_6.Transition
  nodes [Result(bCanEnterTransition=Get bIsRevving() -> (bIsRevving:bool))]
 }
 Transition {
  path=RidingAdditiveGroup.AnimGraphNode_StateMachine_3.RidingAdditive.AnimStateTransitionNode_7.Transition
  nodes [Result(bCanEnterTransition=NOT Boolean(A=Get bIsRevving()))]
 }
 Transition {path=RidingAdditiveGroup.AnimGraphNode_StateMachine_3.RidingAdditive.AnimStateTransitionNode_8.Transition, nodes [Result()]}
 Transition {
  path=RidingAdditiveGroup.AnimGraphNode_StateMachine_3.RidingAdditive.AnimStateTransitionNode_9.Transition
  nodes [Result(bCanEnterTransition=Get bIsRevving())]
 }
 VehicleFinalPoseOverride {nodes [
   Input Pose() {name=InFinalPose}
   Output Pose() {layerGroup=VehicleOverride}
  ]}
 VehicleFullBodyOverride {nodes [
   n1 = Use cached pose 'PostUpperBodyPose'()
   AnimGraphNode_StateMachine_5 = CombatMode() {is=title AnimGraphNode_StateMachine, bAllowConduitEntryStates=true}
   n2 = CombatAimOffsetGroup!(InPoseCombatAimOffset:Pose Link Structure=@AnimGraphNode_StateMachine_5) -> (Pose:Pose Link Structure) {is=title AnimGraphNode_LinkedAnimLayer, layer=CombatAimOffsetGroup}
   n3 = Blend Poses by bool(
    BlendPose_0=Rotate Root Bone!(
     BasePose:Pose Link Structure=CombatAdditiveGroup!(InPoseCombatAdditive:Pose Link Structure=@n2) -> (Pose:Pose Link Structure) {is=title AnimGraphNode_LinkedAnimLayer, layer=CombatAdditiveGroup}
     Yaw:float=0
     MeshToComponent:Rotator="0,0,0"
    ) -> (Pose:Pose Link Structure) {is=title AnimGraphNode_RotateRootBone}
    BlendPose_1=IK/BoneModification!(InPoseIK:Pose Link Structure=RidingAdditiveGroup!(InPoseRidingAdditive:Pose Link Structure=@n1) -> (Pose:Pose Link Structure) {is=title AnimGraphNode_LinkedAnimLayer, layer=RidingAdditiveGroup}) -> (Pose:Pose Link Structure) {is=title AnimGraphNode_LinkedAnimLayer, layer=IK/BoneModification}
    BlendTime_0=0
    BlendTime_1=0
   ) {blendType=Linear, transitionType=Inertialization}
   Output Pose(Result=Inertialization!(Source:Pose Link Structure=AdjustmentGroup!(InPoseAdjustment:Pose Link Structure=@n3) -> (Pose:Pose Link Structure) {is=title AnimGraphNode_LinkedAnimLayer, layer=AdjustmentGroup}) -> (Pose:Pose Link Structure) {is=title AnimGraphNode_Inertialization}) {layerGroup=VehicleOverride}
   Save cached pose 'UpperAndLowerPose'(Pose=Input Pose() {name=InPoseFullBody}) {cacheName=UpperAndLowerPose}
   AnimGraphNode_StateMachine_1 = RidingFullBody() {is=title AnimGraphNode_StateMachine, bAllowConduitEntryStates=true}
   Save cached pose 'FullBodyRiding'(Pose=@AnimGraphNode_StateMachine_1) {cacheName=FullBodyRiding}
   Save cached pose 'LocomotionLoopPose'(Pose=CachedPose!(InPoseCached:Pose Link Structure=Use cached pose 'LocoLeansPose'()) -> (Pose:Pose Link Structure) {is=title AnimGraphNode_LinkedAnimLayer, layer=CachedPose}) {cacheName=LocomotionLoopPose}
   AnimGraphNode_StateMachine_2 = UpperBody() {is=title AnimGraphNode_StateMachine}
   n4 = Layered blend per bone(
    BasePose=Use cached pose 'FullBodyRiding'()
    BlendPoses_0=@AnimGraphNode_StateMachine_2
    BlendWeights_0=Get UpperBodyBlendAlpha() -> (UpperBodyBlendAlpha:float)
   ) {
    bBlendRootMotionBasedOnRootBone=false
    bMeshSpaceRotationBlend=true
    layerSetup [{branchFilters [
       {boneName=Root}
       {blendDepth=-1, boneName=pelvis}
       {boneName=Spine_01}
       {blendDepth=-1, boneName=thigh_l}
       {blendDepth=-1, boneName=thigh_r}
       {blendDepth=-1, boneName=ik_foot_root}
      ]}]
   }
   Save cached pose 'PostUpperBodyPose'(Pose=Apply Additive(
     Base=@n4
     Additive="Sequence Player (additive) 'Sportbike_Driver_FemaleCorrection'"() {sequence=/Sportbike/Animation/Driver/Sportbike_Driver_FemaleCorrection}
     bAlphaBoolEnabled=True
    ) {alphaInputType=Bool, alphaBoolBlend {blendOption=Linear}}) {cacheName=PostUpperBodyPose}
   Save cached pose 'ReverseNoFeetPose'(Pose=Sequence Player 'Sportbike_Driver_ReverseNoFeet'() {sequence=/Sportbike/Animation/Driver/Sportbike_Driver_ReverseNoFeet}) {cacheName=ReverseNoFeetPose}
   Save cached pose 'SportbikeLocoPose'(Pose="Sequence Player (additive) 'Sportbike_Driver_DrivingAdditivePose'"() {sequence=/Sportbike/Animation/Driver/Sportbike_Driver_DrivingAdditivePose}) {cacheName=SportbikeLocoPose}
   Save cached pose 'LocoLeansPose'(Pose=Blendspace Player 'Sportbike_Driver_Loco_Leans_BS'(X=0, Y=0) {blendSpace=/Sportbike/Animation/Driver/Sportbike_Driver_Loco_Leans_BS}) {cacheName=LocoLeansPose}
  ]}
 RidingFullBody {
  entry=Idle
  path=VehicleFullBodyOverride.AnimGraphNode_StateMachine_1.RidingFullBody
  ids=id
  nodes [
   ToLocoSelector() {is=title AnimStateAliasNode, also {aliases [Idle,JumpLand,MoveStop,ReverseStop]}}
   ByPassToMovingJumpLand() {is=title AnimStateAliasNode, also {aliases [Falling,NormalJump]}}
   ToIdle() {is=title AnimStateAliasNode, also {aliases [JumpLand,MoveStop,ReverseStop]}}
   ToJumpLand() {is=title AnimStateAliasNode, also {aliases [Falling,NormalJump]}}
   ToJump() {is=title AnimStateAliasNode, also {aliases [Braking,BrakingOutro,Idle,JumpLand,Locomotion,MoveStop,Reverse,ReverseStop,SuperSlide,Wheelie]}}
   ToBoost() {
    is=title AnimStateAliasNode
    also {aliases [Boost_Outro,Braking,BrakingOutro,Falling,Idle,JumpLand,Locomotion,MoveStop,NormalJump,Reverse,ReverseStop,Stoppie,StoppieLand,SuperSlide,Swim]}
   }
   ToLocomotion() {is=title AnimStateAliasNode, also {aliases [Locomotion]}}
   FromTrickToBoost() {is=title AnimStateAliasNode, also {aliases [Boost]}}
   AnimStateConduitNode_0 = JumpSelector() {is=title AnimStateConduitNode}
   AnimStateConduitNode_4 = LocoSelector() {is=title AnimStateConduitNode}
   AnimStateNode_0 = ReverseStop() {is=title AnimStateNode}
   AnimStateNode_10 = BrakingOutro() {is=title AnimStateNode}
   AnimStateNode_11 = Falling() {is=title AnimStateNode, stateLeft {guid=#1}}
   AnimStateNode_12 = GroundTrickOutro() {is=title AnimStateNode}
   AnimStateNode_13 = Idle() {
    stateEntered {guid=#2, notifyName=IdleEnter}
    stateFullyBlended {guid=#3, notifyName=Idle}
   }
   AnimStateNode_14 = Locomotion() {
    is=title AnimStateNode
    stateEntered {guid=#4}
    stateFullyBlended {guid=#5}
    stateLeft {guid=#6, notifyName=LocomotionExit}
   }
   AnimStateNode_16 = JumpLand() {
    is=title AnimStateNode
    stateEntered {guid=#7, notifyName=JumpLandEnter}
    stateFullyBlended {guid=#8, notifyName=JumpLand}
   }
   AnimStateNode_17 = Boost() {is=title AnimStateNode}
   AnimStateNode_18 = Boost_Outro() {is=title AnimStateNode}
   AnimStateNode_24 = MoveStop() {is=title AnimStateNode, stateEntered {guid=#9, notifyName=MoveStopEnter}}
   AnimStateNode_25 = Reverse() {is=title AnimStateNode}
   AnimStateNode_3 = NormalJump() {
    is=title AnimStateNode
    stateEntered {guid=#10, notifyName=JumpEnter}
    stateLeft {guid=#11, notifyName=JumpFallExit}
   }
   AnimStateNode_4 = SuperSlide() {is=title AnimStateNode, bAlwaysResetOnEntry=true}
   AnimStateNode_5 = Swim() {is=title AnimStateNode}
   AnimStateNode_6 = Wheelie() {is=title AnimStateNode, bAlwaysResetOnEntry=true}
   AnimStateNode_7 = Stoppie() {is=title AnimStateNode, stateLeft {guid=#12, notifyName=StoppieLandCheckON}}
   AnimStateNode_8 = StoppieLand() {is=title AnimStateNode}
   AnimStateNode_9 = Braking() {
    is=title AnimStateNode
    stateEntered {guid=#13}
    stateFullyBlended {guid=#14}
    stateLeft {guid=#15}
   }
   AnimStateTransitionNode_0 = AnimStateTransitionNode()
   AnimStateTransitionNode_1 = AnimStateTransitionNode()
   AnimStateTransitionNode_10 = AnimStateTransitionNode() {priorityOrder=3}
   AnimStateTransitionNode_12 = AnimStateTransitionNode() {priorityOrder=5}
   AnimStateTransitionNode_13 = AnimStateTransitionNode()
   AnimStateTransitionNode_14 = AnimStateTransitionNode()
   AnimStateTransitionNode_15 = AnimStateTransitionNode()
   AnimStateTransitionNode_16 = AnimStateTransitionNode() {crossfadeDuration=0.15}
   AnimStateTransitionNode_17 = AnimStateTransitionNode()
   AnimStateTransitionNode_18 = AnimStateTransitionNode()
   AnimStateTransitionNode_19 = AnimStateTransitionNode() {priorityOrder=4}
   AnimStateTransitionNode_2 = AnimStateTransitionNode()
   AnimStateTransitionNode_20 = AnimStateTransitionNode()
   AnimStateTransitionNode_21 = AnimStateTransitionNode()
   AnimStateTransitionNode_22 = AnimStateTransitionNode() {bAutomaticRuleBasedOnSequencePlayerInState=true, crossfadeDuration=0.3, priorityOrder=2}
   AnimStateTransitionNode_23 = AnimStateTransitionNode()
   AnimStateTransitionNode_25 = AnimStateTransitionNode() {priorityOrder=2}
   AnimStateTransitionNode_26 = AnimStateTransitionNode()
   AnimStateTransitionNode_27 = AnimStateTransitionNode()
   AnimStateTransitionNode_28 = AnimStateTransitionNode() {crossfadeDuration=0.067}
   AnimStateTransitionNode_29 = AnimStateTransitionNode() {priorityOrder=2}
   AnimStateTransitionNode_3 = AnimStateTransitionNode() {crossfadeDuration=0.067, priorityOrder=2}
   AnimStateTransitionNode_31 = AnimStateTransitionNode() {bAutomaticRuleBasedOnSequencePlayerInState=true, crossfadeDuration=0.4}
   AnimStateTransitionNode_32 = AnimStateTransitionNode() {crossfadeDuration=0.4}
   AnimStateTransitionNode_33 = AnimStateTransitionNode()
   AnimStateTransitionNode_34 = AnimStateTransitionNode()
   AnimStateTransitionNode_35 = AnimStateTransitionNode()
   AnimStateTransitionNode_36 = AnimStateTransitionNode() {bAutomaticRuleBasedOnSequencePlayerInState=true, crossfadeDuration=0.4, priorityOrder=3}
   AnimStateTransitionNode_4 = AnimStateTransitionNode() {crossfadeDuration=0.3, priorityOrder=2}
   AnimStateTransitionNode_5 = AnimStateTransitionNode()
   AnimStateTransitionNode_50 = AnimStateTransitionNode() {bAutomaticRuleBasedOnSequencePlayerInState=true, crossfadeDuration=0.4}
   AnimStateTransitionNode_6 = AnimStateTransitionNode() {priorityOrder=100}
   AnimStateTransitionNode_7 = AnimStateTransitionNode()
   AnimStateTransitionNode_75 = AnimStateTransitionNode() {crossfadeDuration=0.25, priorityOrder=100}
   AnimStateTransitionNode_79 = AnimStateTransitionNode() {crossfadeDuration=0.35, priorityOrder=0}
   AnimStateTransitionNode_8 = AnimStateTransitionNode() {crossfadeDuration=0.3}
   AnimStateTransitionNode_9 = AnimStateTransitionNode() {bAutomaticRuleBasedOnSequencePlayerInState=true}
  ]
  links [
   {
    dst=Boost_Outro
    rule=VehicleFullBodyOverride.AnimGraphNode_StateMachine_1.RidingFullBody.AnimStateTransitionNode_33.Transition
    src=Boost
    via=AnimStateTransitionNode_33
   }
   {
    dst=ToLocomotion
    rule=VehicleFullBodyOverride.AnimGraphNode_StateMachine_1.RidingFullBody.AnimStateTransitionNode_31.Transition
    src=Boost_Outro
    via=AnimStateTransitionNode_31
   }
   {
    dst=BrakingOutro
    rule=VehicleFullBodyOverride.AnimGraphNode_StateMachine_1.RidingFullBody.AnimStateTransitionNode_20.AnimationTransitionGraph_1
    src=Braking
    via=AnimStateTransitionNode_20
   }
   {
    dst=LocoSelector
    rule=VehicleFullBodyOverride.AnimGraphNode_StateMachine_1.RidingFullBody.AnimStateTransitionNode_21.Transition
    src=Braking
    via=AnimStateTransitionNode_21
   }
   {
    dst=Locomotion
    rule=VehicleFullBodyOverride.AnimGraphNode_StateMachine_1.RidingFullBody.AnimStateTransitionNode_22.Transition
    src=BrakingOutro
    via=AnimStateTransitionNode_22
   }
   {
    dst=Locomotion
    rule=VehicleFullBodyOverride.AnimGraphNode_StateMachine_1.RidingFullBody.AnimStateTransitionNode_23.Transition
    src=BrakingOutro
    via=AnimStateTransitionNode_23
   }
   {
    dst=Locomotion
    rule=VehicleFullBodyOverride.AnimGraphNode_StateMachine_1.RidingFullBody.AnimStateTransitionNode_16.Transition
    src=ByPassToMovingJumpLand
    via=AnimStateTransitionNode_16
   }
   {
    dst=FromTrickToBoost
    rule=VehicleFullBodyOverride.AnimGraphNode_StateMachine_1.RidingFullBody.AnimStateTransitionNode_25.Transition
    src=GroundTrickOutro
    via=AnimStateTransitionNode_25
   }
   {
    dst=Locomotion
    rule=VehicleFullBodyOverride.AnimGraphNode_StateMachine_1.RidingFullBody.AnimStateTransitionNode_35.Transition
    src=GroundTrickOutro
    via=AnimStateTransitionNode_35
   }
   {
    dst=Locomotion
    rule=VehicleFullBodyOverride.AnimGraphNode_StateMachine_1.RidingFullBody.AnimStateTransitionNode_36.Transition
    src=GroundTrickOutro
    via=AnimStateTransitionNode_36
   }
   {
    dst=Falling
    rule=VehicleFullBodyOverride.AnimGraphNode_StateMachine_1.RidingFullBody.AnimStateTransitionNode_27.Transition
    src=JumpSelector
    via=AnimStateTransitionNode_27
   }
   {
    dst=Falling
    rule=VehicleFullBodyOverride.AnimGraphNode_StateMachine_1.RidingFullBody.AnimStateTransitionNode_6.Transition
    src=JumpSelector
    via=AnimStateTransitionNode_6
   }
   {
    dst=NormalJump
    rule=VehicleFullBodyOverride.AnimGraphNode_StateMachine_1.RidingFullBody.AnimStateTransitionNode_3.Transition
    src=JumpSelector
    via=AnimStateTransitionNode_3
   }
   {
    dst=Braking
    rule=VehicleFullBodyOverride.AnimGraphNode_StateMachine_1.RidingFullBody.AnimStateTransitionNode_19.Transition
    src=LocoSelector
    via=AnimStateTransitionNode_19
   }
   {
    dst=Locomotion
    rule=VehicleFullBodyOverride.AnimGraphNode_StateMachine_1.RidingFullBody.AnimStateTransitionNode_75.Transition
    src=LocoSelector
    via=AnimStateTransitionNode_75
   }
   {
    dst=Reverse
    rule=VehicleFullBodyOverride.AnimGraphNode_StateMachine_1.RidingFullBody.AnimStateTransitionNode_79.Transition
    src=LocoSelector
    via=AnimStateTransitionNode_79
   }
   {
    dst=Stoppie
    rule=VehicleFullBodyOverride.AnimGraphNode_StateMachine_1.RidingFullBody.AnimStateTransitionNode_10.Transition
    src=LocoSelector
    via=AnimStateTransitionNode_10
   }
   {
    dst=SuperSlide
    rule=VehicleFullBodyOverride.AnimGraphNode_StateMachine_1.RidingFullBody.AnimStateTransitionNode_8.Transition
    src=LocoSelector
    via=AnimStateTransitionNode_8
   }
   {
    dst=Swim
    rule=VehicleFullBodyOverride.AnimGraphNode_StateMachine_1.RidingFullBody.AnimStateTransitionNode_12.Transition
    src=LocoSelector
    via=AnimStateTransitionNode_12
   }
   {
    dst=Wheelie
    rule=VehicleFullBodyOverride.AnimGraphNode_StateMachine_1.RidingFullBody.AnimStateTransitionNode_4.Transition
    src=LocoSelector
    via=AnimStateTransitionNode_4
   }
   {
    dst=LocoSelector
    rule=VehicleFullBodyOverride.AnimGraphNode_StateMachine_1.RidingFullBody.AnimStateTransitionNode_14.Transition
    src=Locomotion
    via=AnimStateTransitionNode_14
   }
   {
    dst=MoveStop
    rule=VehicleFullBodyOverride.AnimGraphNode_StateMachine_1.RidingFullBody.AnimStateTransitionNode_34.Transition
    src=Locomotion
    via=AnimStateTransitionNode_34
   }
   {
    dst=LocoSelector
    rule=VehicleFullBodyOverride.AnimGraphNode_StateMachine_1.RidingFullBody.AnimStateTransitionNode_17.Transition
    src=Reverse
    via=AnimStateTransitionNode_17
   }
   {
    dst=ReverseStop
    rule=VehicleFullBodyOverride.AnimGraphNode_StateMachine_1.RidingFullBody.AnimStateTransitionNode_0.Transition
    src=Reverse
    via=AnimStateTransitionNode_0
   }
   {
    dst=LocoSelector
    rule=VehicleFullBodyOverride.AnimGraphNode_StateMachine_1.RidingFullBody.AnimStateTransitionNode_18.Transition
    src=Stoppie
    via=AnimStateTransitionNode_18
   }
   {
    dst=StoppieLand
    rule=VehicleFullBodyOverride.AnimGraphNode_StateMachine_1.RidingFullBody.AnimStateTransitionNode_15.Transition
    src=Stoppie
    via=AnimStateTransitionNode_15
   }
   {
    dst=Locomotion
    rule=VehicleFullBodyOverride.AnimGraphNode_StateMachine_1.RidingFullBody.AnimStateTransitionNode_9.Transition
    src=StoppieLand
    via=AnimStateTransitionNode_9
   }
   {
    dst=LocoSelector
    rule=VehicleFullBodyOverride.AnimGraphNode_StateMachine_1.RidingFullBody.AnimStateTransitionNode_1.Transition
    src=SuperSlide
    via=AnimStateTransitionNode_1
   }
   {
    dst=LocoSelector
    rule=VehicleFullBodyOverride.AnimGraphNode_StateMachine_1.RidingFullBody.AnimStateTransitionNode_13.Transition
    src=Swim
    via=AnimStateTransitionNode_13
   }
   {
    dst=Boost
    rule=VehicleFullBodyOverride.AnimGraphNode_StateMachine_1.RidingFullBody.AnimStateTransitionNode_32.Transition
    src=ToBoost
    via=AnimStateTransitionNode_32
   }
   {
    dst=Idle
    rule=VehicleFullBodyOverride.AnimGraphNode_StateMachine_1.RidingFullBody.AnimStateTransitionNode_50.Transition
    src=ToIdle
    via=AnimStateTransitionNode_50
   }
   {
    dst=JumpSelector
    rule=VehicleFullBodyOverride.AnimGraphNode_StateMachine_1.RidingFullBody.AnimStateTransitionNode_2.Transition
    src=ToJump
    via=AnimStateTransitionNode_2
   }
   {
    dst=JumpLand
    rule=VehicleFullBodyOverride.AnimGraphNode_StateMachine_1.RidingFullBody.AnimStateTransitionNode_28.Transition
    src=ToJumpLand
    via=AnimStateTransitionNode_28
   }
   {
    dst=LocoSelector
    rule=VehicleFullBodyOverride.AnimGraphNode_StateMachine_1.RidingFullBody.AnimStateTransitionNode_26.Transition
    src=ToLocoSelector
    via=AnimStateTransitionNode_26
   }
   {
    dst=GroundTrickOutro
    rule=VehicleFullBodyOverride.AnimGraphNode_StateMachine_1.RidingFullBody.AnimStateTransitionNode_29.Transition
    src=Wheelie
    via=AnimStateTransitionNode_29
   }
   {
    dst=GroundTrickOutro
    rule=VehicleFullBodyOverride.AnimGraphNode_StateMachine_1.RidingFullBody.AnimStateTransitionNode_5.Transition
    src=Wheelie
    via=AnimStateTransitionNode_5
   }
   {
    dst=LocoSelector
    rule=VehicleFullBodyOverride.AnimGraphNode_StateMachine_1.RidingFullBody.AnimStateTransitionNode_7.Transition
    src=Wheelie
    via=AnimStateTransitionNode_7
   }
  ]
 }
 JumpSelector {
  path=VehicleFullBodyOverride.AnimGraphNode_StateMachine_1.RidingFullBody.AnimStateConduitNode_0.JumpSelector
  nodes [Result(bCanEnterTransition=true)]
 }
 LocoSelector {
  path=VehicleFullBodyOverride.AnimGraphNode_StateMachine_1.RidingFullBody.AnimStateConduitNode_4.LocoSelector
  nodes [Result(bCanEnterTransition=true)]
 }
 ReverseStop {
  path=VehicleFullBodyOverride.AnimGraphNode_StateMachine_1.RidingFullBody.AnimStateNode_0.ReverseStop
  nodes [Output Animation Pose(Result=Sequence Player 'Sportbike_Reverse_Stop_CMM'() {bLoopAnimation=false, sequence=/Sportbike/Animation/Driver/Sportbike_Reverse_Stop_CMM})]
 }
 BrakingOutro {
  path=VehicleFullBodyOverride.AnimGraphNode_StateMachine_1.RidingFullBody.AnimStateNode_10.BrakingOutro
  nodes [Output Animation Pose(Result=Sequence Player 'Sportbike_Driver_Brake_End'() {bLoopAnimation=false, sequence=/Sportbike/Animation/Driver/Sportbike_Driver_Brake_End})]
 }
 Falling {
  path=VehicleFullBodyOverride.AnimGraphNode_StateMachine_1.RidingFullBody.AnimStateNode_11.Falling
  nodes [
   n1 = Apply Additive!(
    Base=Sequence Player 'MotoDirt_Driver_Placeholder_Pose_CMM'() {sequence=/Motorcycle/Animation/Driver/MotoDirt_Driver_Placeholder_Pose_CMM}
    Additive=Use cached pose 'LocoLeansPose'()
    Alpha:float
   ) -> (Pose) {alphaBoolBlend {blendOption=Linear}}
   AnimGraphNode_StateMachine_1 = FallingInMotionAdditive() {is=title AnimGraphNode_StateMachine}
   Output Animation Pose(Result=Apply Additive(
     Base=Apply Additive(Base=@n1, Additive=@AnimGraphNode_StateMachine_1) {alphaBoolBlend {blendOption=Linear}}
     Additive=Use cached pose 'SportbikeLocoPose'()
    ) {alphaBoolBlend {blendOption=Linear}})
  ]
 }
 FallingInMotionAdditive {
  entry=FallingStart
  path=VehicleFullBodyOverride.AnimGraphNode_StateMachine_1.RidingFullBody.AnimStateNode_11.Falling.AnimGraphNode_StateMachine_1.FallingInMotionAdditive
  ids=id
  nodes [
   AnimStateNode_0 = FallingStart() {is=title AnimStateNode, bAlwaysResetOnEntry=true}
   AnimStateNode_1 = FallingLoop() {is=title AnimStateNode, bAlwaysResetOnEntry=true}
   AnimStateTransitionNode_14 = AnimStateTransitionNode() {bAutomaticRuleBasedOnSequencePlayerInState=true, priorityOrder=2}
  ]
  links [{
    dst=FallingLoop
    rule=VehicleFullBodyOverride.AnimGraphNode_StateMachine_1.RidingFullBody.AnimStateNode_11.Falling.AnimGraphNode_StateMachine_1.FallingInMotionAdditive.AnimStateTransitionNode_14.Transition
    src=FallingStart
    via=AnimStateTransitionNode_14
   }]
 }
 FallingStart {
  path=VehicleFullBodyOverride.AnimGraphNode_StateMachine_1.RidingFullBody.AnimStateNode_11.Falling.AnimGraphNode_StateMachine_1.FallingInMotionAdditive.AnimStateNode_0.FallingStart
  nodes [Output Animation Pose(Result="Sequence Player (additive) 'MotoDirt_Driver_Falling_Start'"() {bLoopAnimation=false, sequence=/Motorcycle/Animation/Driver/MotoDirt_Driver_Falling_Start})]
 }
 FallingLoop {
  path=VehicleFullBodyOverride.AnimGraphNode_StateMachine_1.RidingFullBody.AnimStateNode_11.Falling.AnimGraphNode_StateMachine_1.FallingInMotionAdditive.AnimStateNode_1.FallingLoop
  nodes [Output Animation Pose(Result="Sequence Player (additive) 'MotoDirt_Driver_Falling_Loop'"() {sequence=/Motorcycle/Animation/Driver/MotoDirt_Driver_Falling_Loop})]
 }
 Transition {
  path=VehicleFullBodyOverride.AnimGraphNode_StateMachine_1.RidingFullBody.AnimStateNode_11.Falling.AnimGraphNode_StateMachine_1.FallingInMotionAdditive.AnimStateTransitionNode_14.Transition
  nodes [Result()]
 }
 GroundTrickOutro {
  path=VehicleFullBodyOverride.AnimGraphNode_StateMachine_1.RidingFullBody.AnimStateNode_12.GroundTrickOutro
  nodes [Output Animation Pose(Result=Sequence Player 'SportBike_Driver_GroundTrick_Outro'() {bLoopAnimation=false, sequence=/Sportbike/Animation/Driver/SportBike_Driver_GroundTrick_Outro})]
 }
 Idle {
  path=VehicleFullBodyOverride.AnimGraphNode_StateMachine_1.RidingFullBody.AnimStateNode_13.Idle
  nodes [
   AimOffset Player 'Sportbike_Driver_HeadAO'!(X:float=0, Y:float=0, BasePose:Pose Link Structure, bAlphaBoolEnabled:bool=False) -> (Pose:Pose Link Structure) {alphaInputType=Bool, blendSpace=/Sportbike/Animation/Driver/AO/Sportbike_Driver_HeadAO, alphaBoolBlend {blendInTime=0.35, blendOutTime=0.2}}
   n1 = Blendspace Player 'MotoDirt_Driver_Steer_BS'!(X=Break Fort Anim Input Motorcycle!(FortAnimInput_Motorcycle:"Fort Anim Input Motorcycle Structure (by ref)"=Get MotorcycleInput() -> (MotorcycleInput:Fort Anim Input Motorcycle Structure)) -> (
     bIsUsingMotorcycle:bool=~
     bIsDriver:bool=~
     bIsFrontPassenger:bool=~
     bIsBackPassenger:bool=~
     bHasPassenger:bool=~
     bIsBackPassengerAndLeaning:bool=~
     bIsDrifting:bool=~
     bIsBoosting:bool=~
     bIsReversing:bool=~
     bIsBraking:bool=~
     bIsMoving:bool=~
     bIsMovingForward:bool=~
     bIsSteeringRight:bool=~
     bIsSteeringLeft:bool=~
     RunForwardAlpha:float=~
     VerticalVelocity:float=~
     VerticalAcceleration:float=~
     bAimFWD:bool=~
     bAimBWD:bool=~
     bAimLFT:bool=~
     bAimRGT:bool=~
     PawnToVehicleDeltaYawAngleDegrees:float=~
     AimCardDirDeadZoneAngleDegrees:float=~
     AimCardDirAngleOffsetDegrees:float=~
     AimFWDDeltaAngleDegrees:float=~
     AimBWDDeltaAngleDegrees:float=~
     AimLFTDeltaAngleDegrees:float=~
     AimRGTDeltaAngleDegrees:float=~
     SlopePitchDegreeAngle:float=~
     SlopeRollDegreeAngle:float=~
     SteerAngle:float=~
     SteerAlpha:float=~
     SteerAngleDeadZoneDegrees:float=~
     SteeringRotation:float=~
     VehiclePitch:float=~
     VehicleRoll:float=~
    ) {is=title K2Node_BreakStruct}.SteerAngle) -> (Pose) {blendSpace=/Motorcycle/Animation/Driver/BlendSpace/MotoDirt_Driver_Steer_BS}
   n2 = Sequence Player 'MotoDirt_Driver_Idle_Stopped'() {sequence=/Motorcycle/Animation/Driver/MotoDirt_Driver_Idle_Stopped}
   Output Animation Pose(Result=Apply Additive(
     Base=Apply Additive(Base=@n2, Additive=@n1) {alphaBoolBlend {blendOption=Linear}}
     Additive="Sequence Player (additive) 'Sportbike_Driver_IdleStoppedAdditivePose'"() {sequence=/Sportbike/Animation/Driver/Sportbike_Driver_IdleStoppedAdditivePose}
    ) {alphaBoolBlend {blendOption=Linear}})
  ]
 }
 Locomotion {
  path=VehicleFullBodyOverride.AnimGraphNode_StateMachine_1.RidingFullBody.AnimStateNode_14.Locomotion
  nodes [
   AnimGraphNode_StateMachine_0 = Movement() {is=title AnimGraphNode_StateMachine, bAllowConduitEntryStates=true}
   Output Animation Pose(Result=Apply Additive(Base=@AnimGraphNode_StateMachine_0, Additive=Use cached pose 'SportbikeLocoPose'()) {alphaBoolBlend {blendOption=Linear}})
  ]
 }
 Movement {
  entry=Conduit
  path=VehicleFullBodyOverride.AnimGraphNode_StateMachine_1.RidingFullBody.AnimStateNode_14.Locomotion.AnimGraphNode_StateMachine_0.Movement
  ids=id
  nodes [
   AnimStateConduitNode_0 = Conduit()
   AnimStateNode_0 = MovingStart() {is=title AnimStateNode}
   AnimStateNode_1 = MovingLoop() {is=title AnimStateNode, stateEntered {guid=#16, notifyName=StoppieLandCheckOFF}}
   AnimStateTransitionNode_0 = AnimStateTransitionNode() {blendMode=Linear, priorityOrder=0}
   AnimStateTransitionNode_1 = AnimStateTransitionNode() {priorityOrder=0}
   AnimStateTransitionNode_19 = AnimStateTransitionNode() {bAutomaticRuleBasedOnSequencePlayerInState=true, crossfadeDuration=0.5}
   AnimStateTransitionNode_55 = AnimStateTransitionNode() {priorityOrder=100}
   AnimStateTransitionNode_56 = AnimStateTransitionNode() {crossfadeDuration=0.5}
  ]
  links [
   {
    dst=MovingLoop
    rule=VehicleFullBodyOverride.AnimGraphNode_StateMachine_1.RidingFullBody.AnimStateNode_14.Locomotion.AnimGraphNode_StateMachine_0.Movement.AnimStateTransitionNode_1.Transition
    src=Conduit
    via=AnimStateTransitionNode_1
   }
   {
    dst=MovingLoop
    rule=VehicleFullBodyOverride.AnimGraphNode_StateMachine_1.RidingFullBody.AnimStateNode_14.Locomotion.AnimGraphNode_StateMachine_0.Movement.AnimStateTransitionNode_56.Transition
    src=Conduit
    via=AnimStateTransitionNode_56
   }
   {
    dst=MovingStart
    rule=VehicleFullBodyOverride.AnimGraphNode_StateMachine_1.RidingFullBody.AnimStateNode_14.Locomotion.AnimGraphNode_StateMachine_0.Movement.AnimStateTransitionNode_55.Transition
    src=Conduit
    via=AnimStateTransitionNode_55
   }
   {
    dst=MovingLoop
    rule=VehicleFullBodyOverride.AnimGraphNode_StateMachine_1.RidingFullBody.AnimStateNode_14.Locomotion.AnimGraphNode_StateMachine_0.Movement.AnimStateTransitionNode_0.Transition
    src=MovingStart
    via=AnimStateTransitionNode_0
   }
   {
    dst=MovingLoop
    rule=VehicleFullBodyOverride.AnimGraphNode_StateMachine_1.RidingFullBody.AnimStateNode_14.Locomotion.AnimGraphNode_StateMachine_0.Movement.AnimStateTransitionNode_19.Transition
    src=MovingStart
    via=AnimStateTransitionNode_19
   }
  ]
 }
 Conduit {
  path=VehicleFullBodyOverride.AnimGraphNode_StateMachine_1.RidingFullBody.AnimStateNode_14.Locomotion.AnimGraphNode_StateMachine_0.Movement.AnimStateConduitNode_0.Conduit
  nodes [Result(bCanEnterTransition=true)]
 }
 MovingStart {
  path=VehicleFullBodyOverride.AnimGraphNode_StateMachine_1.RidingFullBody.AnimStateNode_14.Locomotion.AnimGraphNode_StateMachine_0.Movement.AnimStateNode_0.MovingStart
  nodes [Output Animation Pose(Result=Sequence Player 'MotoDirt_Driver_MoveStart'() {bLoopAnimation=false, sequence=/Motorcycle/Animation/Driver/MotoDirt_Driver_MoveStart})]
 }
 MovingLoop {
  path=VehicleFullBodyOverride.AnimGraphNode_StateMachine_1.RidingFullBody.AnimStateNode_14.Locomotion.AnimGraphNode_StateMachine_0.Movement.AnimStateNode_1.MovingLoop
  nodes [
   AnimGraphNode_StateMachine_0 = MovingJumpLandAdditive() {is=title AnimGraphNode_StateMachine}
   Output Animation Pose(Result=Apply Additive(Base=Use cached pose 'LocomotionLoopPose'(), Additive=@AnimGraphNode_StateMachine_0) {alphaBoolBlend {blendOption=Linear}})
  ]
 }
 MovingJumpLandAdditive {
  entry=AdditiveRefPose
  path=VehicleFullBodyOverride.AnimGraphNode_StateMachine_1.RidingFullBody.AnimStateNode_14.Locomotion.AnimGraphNode_StateMachine_0.Movement.AnimStateNode_1.MovingLoop.AnimGraphNode_StateMachine_0.MovingJumpLandAdditive
  ids=id
  nodes [
   AnimStateNode_0 = AdditiveRefPose() {
    is=title AnimStateNode
    stateEntered {guid=#17}
    stateFullyBlended {guid=#18}
    stateLeft {guid=#19}
   }
   AnimStateNode_2 = FallingLandAdditive() {is=title AnimStateNode, stateLeft {guid=#20, notifyName=MovingLandExit}}
   AnimStateNode_3 = MovingJumpLandAdditive() {is=title AnimStateNode, stateLeft {guid=#21, notifyName=MovingLandExit}}
   AnimStateTransitionNode_5 = AnimStateTransitionNode() {crossfadeDuration=0.15}
   AnimStateTransitionNode_57 = AnimStateTransitionNode() {crossfadeDuration=0.15, priorityOrder=2}
   AnimStateTransitionNode_58 = AnimStateTransitionNode() {bAutomaticRuleBasedOnSequencePlayerInState=true}
   AnimStateTransitionNode_6 = AnimStateTransitionNode() {bAutomaticRuleBasedOnSequencePlayerInState=true, crossfadeDuration=0.1}
  ]
  links [
   {
    dst=FallingLandAdditive
    rule=VehicleFullBodyOverride.AnimGraphNode_StateMachine_1.RidingFullBody.AnimStateNode_14.Locomotion.AnimGraphNode_StateMachine_0.Movement.AnimStateNode_1.MovingLoop.AnimGraphNode_StateMachine_0.MovingJumpLandAdditive.AnimStateTransitionNode_5.Transition
    src=AdditiveRefPose
    via=AnimStateTransitionNode_5
   }
   {
    dst=MovingJumpLandAdditive
    rule=VehicleFullBodyOverride.AnimGraphNode_StateMachine_1.RidingFullBody.AnimStateNode_14.Locomotion.AnimGraphNode_StateMachine_0.Movement.AnimStateNode_1.MovingLoop.AnimGraphNode_StateMachine_0.MovingJumpLandAdditive.AnimStateTransitionNode_57.Transition
    src=AdditiveRefPose
    via=AnimStateTransitionNode_57
   }
   {
    dst=AdditiveRefPose
    rule=VehicleFullBodyOverride.AnimGraphNode_StateMachine_1.RidingFullBody.AnimStateNode_14.Locomotion.AnimGraphNode_StateMachine_0.Movement.AnimStateNode_1.MovingLoop.AnimGraphNode_StateMachine_0.MovingJumpLandAdditive.AnimStateTransitionNode_6.Transition
    src=FallingLandAdditive
    via=AnimStateTransitionNode_6
   }
   {
    dst=AdditiveRefPose
    rule=VehicleFullBodyOverride.AnimGraphNode_StateMachine_1.RidingFullBody.AnimStateNode_14.Locomotion.AnimGraphNode_StateMachine_0.Movement.AnimStateNode_1.MovingLoop.AnimGraphNode_StateMachine_0.MovingJumpLandAdditive.AnimStateTransitionNode_58.Transition
    src=MovingJumpLandAdditive
    via=AnimStateTransitionNode_58
   }
  ]
 }
 AdditiveRefPose {
  path=VehicleFullBodyOverride.AnimGraphNode_StateMachine_1.RidingFullBody.AnimStateNode_14.Locomotion.AnimGraphNode_StateMachine_0.Movement.AnimStateNode_1.MovingLoop.AnimGraphNode_StateMachine_0.MovingJumpLandAdditive.AnimStateNode_0.AdditiveRefPose
  nodes [Output Animation Pose(Result=Additive Identity Pose())]
 }
 FallingLandAdditive {
  path=VehicleFullBodyOverride.AnimGraphNode_StateMachine_1.RidingFullBody.AnimStateNode_14.Locomotion.AnimGraphNode_StateMachine_0.Movement.AnimStateNode_1.MovingLoop.AnimGraphNode_StateMachine_0.MovingJumpLandAdditive.AnimStateNode_2.FallingLandAdditive
  nodes [Output Animation Pose(Result="Sequence Player (additive) 'MotoDirt_Driver_Falling_Land'"!(StartPosition:float=0) -> (Pose) {bLoopAnimation=false, sequence=/Motorcycle/Animation/Driver/MotoDirt_Driver_Falling_Land})]
 }
 MovingJumpLandAdditive {
  path=VehicleFullBodyOverride.AnimGraphNode_StateMachine_1.RidingFullBody.AnimStateNode_14.Locomotion.AnimGraphNode_StateMachine_0.Movement.AnimStateNode_1.MovingLoop.AnimGraphNode_StateMachine_0.MovingJumpLandAdditive.AnimStateNode_3.MovingJumpLandAdditive
  nodes [
   n1 = "Sequence Player (additive) 'MotoDirt_Driver_MovingJumpLand'"() {bLoopAnimation=false, groupName=MovingJumpLand, method=SyncGroup, sequence=/Motorcycle/Animation/Driver/MotoDirt_Driver_MovingJumpLand}
   n2 = Blend Poses by bool(
    BlendPose_0=@n1
    BlendPose_1="Sequence Player (additive) 'MotoDirt_Driver_JumpTrickMovingLand'"!(StartPosition:float=0) -> (Pose) {bLoopAnimation=false, sequence=/Motorcycle/Animation/Driver/MotoDirt_Driver_JumpTrickMovingLand}
    BlendTime_1=0.2
   ) {blendType=Linear}
   n3 = "Sequence Player (additive) 'MotoDirt_Driver_JumpSoloTrickMovingLand'"() {
    bLoopAnimation=false
    groupName=MovingJumpLand
    method=SyncGroup
    sequence=/Motorcycle/Animation/Driver/MotoDirt_Driver_JumpSoloTrickMovingLand
   }
   n4 = "Sequence Player (additive) 'MotoDirt_Driver_MovingJumpLand'"() {bLoopAnimation=false, groupName=MovingJumpLand, method=SyncGroup, sequence=&MotoDirt_Driver_MovingJumpLand}
   Output Animation Pose(Result=Blend Poses by bool(BlendPose_0=Blend Poses by bool(BlendPose_0=@n3, BlendPose_1=@n2) {blendType=Linear}, BlendPose_1=@n4) {blendType=Linear})
  ]
 }
 Transition {
  path=VehicleFullBodyOverride.AnimGraphNode_StateMachine_1.RidingFullBody.AnimStateNode_14.Locomotion.AnimGraphNode_StateMachine_0.Movement.AnimStateNode_1.MovingLoop.AnimGraphNode_StateMachine_0.MovingJumpLandAdditive.AnimStateTransitionNode_5.Transition
  nodes [Result(bCanEnterTransition=Get bWasFalling() -> (bWasFalling:bool))]
 }
 Transition {
  path=VehicleFullBodyOverride.AnimGraphNode_StateMachine_1.RidingFullBody.AnimStateNode_14.Locomotion.AnimGraphNode_StateMachine_0.Movement.AnimStateNode_1.MovingLoop.AnimGraphNode_StateMachine_0.MovingJumpLandAdditive.AnimStateTransitionNode_57.Transition
  nodes [Result(bCanEnterTransition=Get bShouldAllowMovingJumpLand() -> (bShouldAllowMovingJumpLand:bool))]
 }
 Transition {
  path=VehicleFullBodyOverride.AnimGraphNode_StateMachine_1.RidingFullBody.AnimStateNode_14.Locomotion.AnimGraphNode_StateMachine_0.Movement.AnimStateNode_1.MovingLoop.AnimGraphNode_StateMachine_0.MovingJumpLandAdditive.AnimStateTransitionNode_58.Transition
  nodes [Result()]
 }
 Transition {
  path=VehicleFullBodyOverride.AnimGraphNode_StateMachine_1.RidingFullBody.AnimStateNode_14.Locomotion.AnimGraphNode_StateMachine_0.Movement.AnimStateNode_1.MovingLoop.AnimGraphNode_StateMachine_0.MovingJumpLandAdditive.AnimStateTransitionNode_6.Transition
  nodes [Result()]
 }
 Transition {
  path=VehicleFullBodyOverride.AnimGraphNode_StateMachine_1.RidingFullBody.AnimStateNode_14.Locomotion.AnimGraphNode_StateMachine_0.Movement.AnimStateTransitionNode_0.Transition
  nodes [Result(bCanEnterTransition=Get bEarlyOutFromMoveStart() -> (bEarlyOutFromMoveStart:bool))]
 }
 Transition {
  path=VehicleFullBodyOverride.AnimGraphNode_StateMachine_1.RidingFullBody.AnimStateNode_14.Locomotion.AnimGraphNode_StateMachine_0.Movement.AnimStateTransitionNode_1.Transition
  nodes [Result(bCanEnterTransition=Get bWasStoppieLanding() -> (bWasStoppieLanding:bool))]
 }
 Transition {
  path=VehicleFullBodyOverride.AnimGraphNode_StateMachine_1.RidingFullBody.AnimStateNode_14.Locomotion.AnimGraphNode_StateMachine_0.Movement.AnimStateTransitionNode_19.Transition
  nodes [Result()]
 }
 Transition {
  path=VehicleFullBodyOverride.AnimGraphNode_StateMachine_1.RidingFullBody.AnimStateNode_14.Locomotion.AnimGraphNode_StateMachine_0.Movement.AnimStateTransitionNode_55.Transition
  nodes [Result(bCanEnterTransition=true)]
 }
 Transition {
  path=VehicleFullBodyOverride.AnimGraphNode_StateMachine_1.RidingFullBody.AnimStateNode_14.Locomotion.AnimGraphNode_StateMachine_0.Movement.AnimStateTransitionNode_56.Transition
  nodes [Result(bCanEnterTransition=Get bStateTransitionRuleToMovingLoop() -> (bStateTransitionRuleToMovingLoop:bool))]
 }
 JumpLand {
  path=VehicleFullBodyOverride.AnimGraphNode_StateMachine_1.RidingFullBody.AnimStateNode_16.JumpLand
  nodes [
   n1 = Sequence Player 'MotoDirt_Driver_JumpTrickLand'() {bLoopAnimation=false, groupName=JumpLand, method=SyncGroup, sequence=/Motorcycle/Animation/Driver/MotoDirt_Driver_JumpTrickLand}
   n2 = Blend Poses by bool(
    BlendPose_0=@n1
    BlendPose_1=Sequence Player 'MotoDirt_Driver_Jump_Land'() {groupName=JumpLand, method=SyncGroup, sequence=/Motorcycle/Animation/Driver/MotoDirt_Driver_Jump_Land}
    BlendTime_0=0.2
   ) {blendType=Linear}
   n3 = Sequence Player 'MotoDirt_Driver_Jump_Land_Combat'() {groupName=JumpLand, method=SyncGroup, sequence=/Motorcycle/Animation/Driver/MotoDirt_Driver_Jump_Land_Combat}
   Output Animation Pose(Result=Apply Additive(Base=Blend Poses by bool(BlendPose_0=@n3, BlendPose_1=@n2) {blendType=Linear}, Additive=Use cached pose 'SportbikeLocoPose'()) {alphaBoolBlend {blendOption=Linear}})
  ]
 }
 Boost {
  path=VehicleFullBodyOverride.AnimGraphNode_StateMachine_1.RidingFullBody.AnimStateNode_17.Boost
  nodes [
   AnimGraphNode_StateMachine_0 = BoostingTransition() {is=title AnimGraphNode_StateMachine}
   Output Animation Pose(Result=@AnimGraphNode_StateMachine_0)
  ]
 }
 BoostingTransition {
  entry=Boost_Intro
  path=VehicleFullBodyOverride.AnimGraphNode_StateMachine_1.RidingFullBody.AnimStateNode_17.Boost.AnimGraphNode_StateMachine_0.BoostingTransition
  ids=id
  nodes [
   AnimStateNode_0 = Boost_Intro() {is=title AnimStateNode}
   AnimStateNode_1 = Boost_Loop() {is=title AnimStateNode}
   AnimStateTransitionNode_3 = AnimStateTransitionNode() {bAutomaticRuleBasedOnSequencePlayerInState=true}
  ]
  links [{
    dst=Boost_Loop
    rule=VehicleFullBodyOverride.AnimGraphNode_StateMachine_1.RidingFullBody.AnimStateNode_17.Boost.AnimGraphNode_StateMachine_0.BoostingTransition.AnimStateTransitionNode_3.Transition
    src=Boost_Intro
    via=AnimStateTransitionNode_3
   }]
 }
 Boost_Intro {
  path=VehicleFullBodyOverride.AnimGraphNode_StateMachine_1.RidingFullBody.AnimStateNode_17.Boost.AnimGraphNode_StateMachine_0.BoostingTransition.AnimStateNode_0.Boost_Intro
  nodes [Output Animation Pose(Result=Sequence Player 'SportBike_Driver_Boost_Intro'() {bLoopAnimation=false, sequence=/Sportbike/Animation/Driver/SportBike_Driver_Boost_Intro})]
 }
 Boost_Loop {
  path=VehicleFullBodyOverride.AnimGraphNode_StateMachine_1.RidingFullBody.AnimStateNode_17.Boost.AnimGraphNode_StateMachine_0.BoostingTransition.AnimStateNode_1.Boost_Loop
  nodes [Output Animation Pose(Result=Sequence Player 'SportBike_Driver_Boost_Loop'() {sequence=/Sportbike/Animation/Driver/SportBike_Driver_Boost_Loop})]
 }
 Transition {
  path=VehicleFullBodyOverride.AnimGraphNode_StateMachine_1.RidingFullBody.AnimStateNode_17.Boost.AnimGraphNode_StateMachine_0.BoostingTransition.AnimStateTransitionNode_3.Transition
  nodes [Result()]
 }
 Boost_Outro {
  path=VehicleFullBodyOverride.AnimGraphNode_StateMachine_1.RidingFullBody.AnimStateNode_18.Boost_Outro
  nodes [Output Animation Pose(Result=Sequence Player 'SportBike_Driver_Boost_Outro'() {bLoopAnimation=false, sequence=/Sportbike/Animation/Driver/SportBike_Driver_Boost_Outro})]
 }
 MoveStop {
  path=VehicleFullBodyOverride.AnimGraphNode_StateMachine_1.RidingFullBody.AnimStateNode_24.MoveStop
  nodes [Output Animation Pose(Result=Sequence Player 'SportBike_Driver_MoveStop'() {bLoopAnimation=false, sequence=/Sportbike/Animation/Driver/SportBike_Driver_MoveStop})]
 }
 Reverse {
  path=VehicleFullBodyOverride.AnimGraphNode_StateMachine_1.RidingFullBody.AnimStateNode_25.Reverse
  nodes [
   n1 = Break Fort Anim Input Motorcycle(FortAnimInput_Motorcycle=Get MotorcycleInput()) -> (
    bIsUsingMotorcycle=~
    bIsDriver=~
    bIsFrontPassenger=~
    bIsBackPassenger=~
    bHasPassenger=~
    bIsBackPassengerAndLeaning=~
    bIsDrifting=~
    bIsBoosting=~
    bIsReversing=~
    bIsBraking=~
    bIsMoving=~
    bIsMovingForward=~
    bIsSteeringRight=~
    bIsSteeringLeft=~
    RunForwardAlpha=~
    VerticalVelocity=~
    VerticalAcceleration=~
    bAimFWD=~
    bAimBWD=~
    bAimLFT=~
    bAimRGT=~
    PawnToVehicleDeltaYawAngleDegrees=~
    AimCardDirDeadZoneAngleDegrees=~
    AimCardDirAngleOffsetDegrees=~
    AimFWDDeltaAngleDegrees=~
    AimBWDDeltaAngleDegrees=~
    AimLFTDeltaAngleDegrees=~
    AimRGTDeltaAngleDegrees=~
    SlopePitchDegreeAngle=~
    SlopeRollDegreeAngle=~
    SteerAngle=~
    SteerAlpha=~
    SteerAngleDeadZoneDegrees=~
    SteeringRotation=~
    VehiclePitch=~
    VehicleRoll=~
   )
   AnimGraphNode_StateMachine_0 = ReverseMovement() {is=title AnimGraphNode_StateMachine, bAllowConduitEntryStates=true}
   Output Animation Pose(Result=Apply Additive(
     Base=@AnimGraphNode_StateMachine_0
     Additive=Blendspace Player 'Sportbike_Driver_Reverse_Leans_BS'!(X=@n1.SteerAngle) -> (Pose) {blendSpace=/Sportbike/Animation/Driver/Sportbike_Driver_Reverse_Leans_BS}
    ) {alphaBoolBlend {blendOption=Linear}})
  ]
 }
 ReverseMovement {
  entry=Conduit
  path=VehicleFullBodyOverride.AnimGraphNode_StateMachine_1.RidingFullBody.AnimStateNode_25.Reverse.AnimGraphNode_StateMachine_0.ReverseMovement
  ids=id
  nodes [
   AnimStateConduitNode_0 = Conduit()
   AnimStateNode_1 = ReverseStart() {is=title AnimStateNode}
   AnimStateNode_2 = ReverseLoop() {is=title AnimStateNode}
   AnimStateTransitionNode_11 = AnimStateTransitionNode() {crossfadeDuration=0.35}
   AnimStateTransitionNode_2 = AnimStateTransitionNode() {priorityOrder=100}
   AnimStateTransitionNode_3 = AnimStateTransitionNode() {bAutomaticRuleBasedOnSequencePlayerInState=true, crossfadeDuration=0.033, logicType=TLT_Inertialization}
  ]
  links [
   {
    dst=ReverseLoop
    rule=VehicleFullBodyOverride.AnimGraphNode_StateMachine_1.RidingFullBody.AnimStateNode_25.Reverse.AnimGraphNode_StateMachine_0.ReverseMovement.AnimStateTransitionNode_11.Transition
    src=Conduit
    via=AnimStateTransitionNode_11
   }
   {
    dst=ReverseStart
    rule=VehicleFullBodyOverride.AnimGraphNode_StateMachine_1.RidingFullBody.AnimStateNode_25.Reverse.AnimGraphNode_StateMachine_0.ReverseMovement.AnimStateTransitionNode_2.Transition
    src=Conduit
    via=AnimStateTransitionNode_2
   }
   {
    dst=ReverseLoop
    rule=VehicleFullBodyOverride.AnimGraphNode_StateMachine_1.RidingFullBody.AnimStateNode_25.Reverse.AnimGraphNode_StateMachine_0.ReverseMovement.AnimStateTransitionNode_3.Transition
    src=ReverseStart
    via=AnimStateTransitionNode_3
   }
  ]
 }
 Conduit {
  path=VehicleFullBodyOverride.AnimGraphNode_StateMachine_1.RidingFullBody.AnimStateNode_25.Reverse.AnimGraphNode_StateMachine_0.ReverseMovement.AnimStateConduitNode_0.Conduit
  nodes [Result(bCanEnterTransition=true)]
 }
 ReverseStart {
  path=VehicleFullBodyOverride.AnimGraphNode_StateMachine_1.RidingFullBody.AnimStateNode_25.Reverse.AnimGraphNode_StateMachine_0.ReverseMovement.AnimStateNode_1.ReverseStart
  nodes [Output Animation Pose(Result=Sequence Player 'Sportbike_Reverse_Start_CMM'() {bLoopAnimation=false, sequence=/Sportbike/Animation/Driver/Sportbike_Reverse_Start_CMM})]
 }
 ReverseLoop {
  path=VehicleFullBodyOverride.AnimGraphNode_StateMachine_1.RidingFullBody.AnimStateNode_25.Reverse.AnimGraphNode_StateMachine_0.ReverseMovement.AnimStateNode_2.ReverseLoop
  nodes [Output Animation Pose(Result=Blend Poses by bool(
     BlendPose_0=Use cached pose 'ReverseNoFeetPose'()
     BlendPose_1=Sequence Player 'Sportbike_ReverseLoop_CMM'!(PlayRate:float=1) -> (Pose) {sequence=/Sportbike/Animation/Driver/Sportbike_ReverseLoop_CMM}
     BlendTime_0=0.6
     BlendTime_1=0.2
    ) {blendType=Linear})]
 }
 Transition {
  path=VehicleFullBodyOverride.AnimGraphNode_StateMachine_1.RidingFullBody.AnimStateNode_25.Reverse.AnimGraphNode_StateMachine_0.ReverseMovement.AnimStateTransitionNode_11.Transition
  nodes [Result(bCanEnterTransition=Get bStateTransitionRuleToReverseLoop() -> (bStateTransitionRuleToReverseLoop:bool))]
 }
 Transition {
  path=VehicleFullBodyOverride.AnimGraphNode_StateMachine_1.RidingFullBody.AnimStateNode_25.Reverse.AnimGraphNode_StateMachine_0.ReverseMovement.AnimStateTransitionNode_2.Transition
  nodes [Result(bCanEnterTransition=true)]
 }
 Transition {
  path=VehicleFullBodyOverride.AnimGraphNode_StateMachine_1.RidingFullBody.AnimStateNode_25.Reverse.AnimGraphNode_StateMachine_0.ReverseMovement.AnimStateTransitionNode_3.Transition
  nodes [Result()]
 }
 NormalJump {
  path=VehicleFullBodyOverride.AnimGraphNode_StateMachine_1.RidingFullBody.AnimStateNode_3.NormalJump
  nodes [
   AnimGraphNode_StateMachine_0 = NormalJumpTransition() {is=title AnimGraphNode_StateMachine, bAllowConduitEntryStates=true}
   Output Animation Pose(Result=@AnimGraphNode_StateMachine_0)
  ]
 }
 NormalJumpTransition {
  entry=Conduit
  path=VehicleFullBodyOverride.AnimGraphNode_StateMachine_1.RidingFullBody.AnimStateNode_3.NormalJump.AnimGraphNode_StateMachine_0.NormalJumpTransition
  ids=id
  nodes [
   AnimStateConduitNode_0 = Conduit()
   AnimStateNode_0 = NormalJumpStart() {is=title AnimStateNode}
   AnimStateNode_1 = NormalJumpLoop() {is=title AnimStateNode, stateEntered {guid=#22, notifyName=NormalJumpLoopEnter}}
   AnimStateTransitionNode_2 = AnimStateTransitionNode() {priorityOrder=100}
   AnimStateTransitionNode_25 = AnimStateTransitionNode() {bAutomaticRuleBasedOnSequencePlayerInState=true}
   AnimStateTransitionNode_6 = AnimStateTransitionNode() {priorityOrder=2}
  ]
  links [
   {
    dst=NormalJumpLoop
    rule=VehicleFullBodyOverride.AnimGraphNode_StateMachine_1.RidingFullBody.AnimStateNode_3.NormalJump.AnimGraphNode_StateMachine_0.NormalJumpTransition.AnimStateTransitionNode_6.Transition
    src=Conduit
    via=AnimStateTransitionNode_6
   }
   {
    dst=NormalJumpStart
    rule=VehicleFullBodyOverride.AnimGraphNode_StateMachine_1.RidingFullBody.AnimStateNode_3.NormalJump.AnimGraphNode_StateMachine_0.NormalJumpTransition.AnimStateTransitionNode_2.Transition
    src=Conduit
    via=AnimStateTransitionNode_2
   }
   {
    dst=NormalJumpLoop
    rule=VehicleFullBodyOverride.AnimGraphNode_StateMachine_1.RidingFullBody.AnimStateNode_3.NormalJump.AnimGraphNode_StateMachine_0.NormalJumpTransition.AnimStateTransitionNode_25.Transition
    src=NormalJumpStart
    via=AnimStateTransitionNode_25
   }
  ]
 }
 Conduit {
  path=VehicleFullBodyOverride.AnimGraphNode_StateMachine_1.RidingFullBody.AnimStateNode_3.NormalJump.AnimGraphNode_StateMachine_0.NormalJumpTransition.AnimStateConduitNode_0.Conduit
  nodes [Result(bCanEnterTransition=true)]
 }
 NormalJumpStart {
  path=VehicleFullBodyOverride.AnimGraphNode_StateMachine_1.RidingFullBody.AnimStateNode_3.NormalJump.AnimGraphNode_StateMachine_0.NormalJumpTransition.AnimStateNode_0.NormalJumpStart
  nodes [
   n1 = Sequence Player 'MotoDirt_Driver_Jump_Start'() {bLoopAnimation=false, groupName=JumpStart, method=SyncGroup, sequence=/Motorcycle/Animation/Driver/MotoDirt_Driver_Jump_Start}
   n2 = Sequence Player 'MotoDirt_Driver_Jump_Start_Combat'() {bLoopAnimation=false, groupName=JumpStart, method=SyncGroup, sequence=/Motorcycle/Animation/Driver/MotoDirt_Driver_Jump_Start_Combat}
   Output Animation Pose(Result=Blend Poses by bool(BlendPose_0=@n2, BlendPose_1=@n1) {blendType=Linear})
  ]
 }
 NormalJumpLoop {
  path=VehicleFullBodyOverride.AnimGraphNode_StateMachine_1.RidingFullBody.AnimStateNode_3.NormalJump.AnimGraphNode_StateMachine_0.NormalJumpTransition.AnimStateNode_1.NormalJumpLoop
  nodes [
   n1 = Blendspace Player 'MotoDirt_Driver_Jump_Leans_BS'(X=0, Y=0) {blendSpace=/Motorcycle/Animation/Driver/BlendSpace/MotoDirt_Driver_Jump_Leans_BS}
   n2 = Sequence Player 'MotoDirt_Driver_Jump_Loop'() {groupName=JumpLoop, method=SyncGroup, sequence=/Motorcycle/Animation/Driver/MotoDirt_Driver_Jump_Loop}
   Output Animation Pose(Result=Blend Poses by bool(
     BlendPose_0=Sequence Player 'MotoDirt_Driver_Jump_Loop_Combat'() {groupName=JumpLoop, method=SyncGroup, sequence=/Motorcycle/Animation/Driver/MotoDirt_Driver_Jump_Loop_Combat}
     BlendPose_1=Apply Additive(Base=@n2, Additive=@n1) {alphaBoolBlend {blendOption=Linear}}
    ) {blendType=Linear})
  ]
 }
 Transition {
  path=VehicleFullBodyOverride.AnimGraphNode_StateMachine_1.RidingFullBody.AnimStateNode_3.NormalJump.AnimGraphNode_StateMachine_0.NormalJumpTransition.AnimStateTransitionNode_2.Transition
  nodes [Result(bCanEnterTransition=true)]
 }
 Transition {
  path=VehicleFullBodyOverride.AnimGraphNode_StateMachine_1.RidingFullBody.AnimStateNode_3.NormalJump.AnimGraphNode_StateMachine_0.NormalJumpTransition.AnimStateTransitionNode_25.Transition
  nodes [Result()]
 }
 Transition {
  path=VehicleFullBodyOverride.AnimGraphNode_StateMachine_1.RidingFullBody.AnimStateNode_3.NormalJump.AnimGraphNode_StateMachine_0.NormalJumpTransition.AnimStateTransitionNode_6.Transition
  nodes [Result(bCanEnterTransition=Get bIsFromJumpTrick() -> (bIsFromJumpTrick:bool))]
 }
 SuperSlide {
  path=VehicleFullBodyOverride.AnimGraphNode_StateMachine_1.RidingFullBody.AnimStateNode_4.SuperSlide
  nodes [
   AnimGraphNode_StateMachine_0 = SuperSlideDirection() {is=title AnimGraphNode_StateMachine, bAllowConduitEntryStates=true}
   Output Animation Pose(Result=@AnimGraphNode_StateMachine_0)
  ]
 }
 SuperSlideDirection {
  entry=Conduit
  path=VehicleFullBodyOverride.AnimGraphNode_StateMachine_1.RidingFullBody.AnimStateNode_4.SuperSlide.AnimGraphNode_StateMachine_0.SuperSlideDirection
  ids=id
  nodes [
   AnimStateConduitNode_0 = Conduit()
   AnimStateNode_0 = SuperSlideStartLeft() {is=title AnimStateNode, stateEntered {guid=#23, notifyName=PowerSlidingDirectionCheckON}}
   AnimStateNode_1 = SuperSlideStartRight() {is=title AnimStateNode}
   AnimStateNode_2 = SuperSlideLoopRight() {is=title AnimStateNode}
   AnimStateNode_3 = SuperSlideLoopLeft() {is=title AnimStateNode}
   AnimStateTransitionNode_10 = AnimStateTransitionNode() {bAutomaticRuleBasedOnSequencePlayerInState=true, logicType=TLT_Inertialization}
   AnimStateTransitionNode_11 = AnimStateTransitionNode() {bAutomaticRuleBasedOnSequencePlayerInState=true, logicType=TLT_Inertialization}
   AnimStateTransitionNode_8 = AnimStateTransitionNode() {priorityOrder=100}
   AnimStateTransitionNode_9 = AnimStateTransitionNode()
  ]
  links [
   {
    dst=SuperSlideStartLeft
    rule=VehicleFullBodyOverride.AnimGraphNode_StateMachine_1.RidingFullBody.AnimStateNode_4.SuperSlide.AnimGraphNode_StateMachine_0.SuperSlideDirection.AnimStateTransitionNode_8.AnimationTransitionGraph_1
    src=Conduit
    via=AnimStateTransitionNode_8
   }
   {
    dst=SuperSlideStartRight
    rule=VehicleFullBodyOverride.AnimGraphNode_StateMachine_1.RidingFullBody.AnimStateNode_4.SuperSlide.AnimGraphNode_StateMachine_0.SuperSlideDirection.AnimStateTransitionNode_9.Transition
    src=Conduit
    via=AnimStateTransitionNode_9
   }
   {
    dst=SuperSlideLoopLeft
    rule=VehicleFullBodyOverride.AnimGraphNode_StateMachine_1.RidingFullBody.AnimStateNode_4.SuperSlide.AnimGraphNode_StateMachine_0.SuperSlideDirection.AnimStateTransitionNode_11.AnimationTransitionGraph_1
    src=SuperSlideStartLeft
    via=AnimStateTransitionNode_11
   }
   {
    dst=SuperSlideLoopRight
    rule=VehicleFullBodyOverride.AnimGraphNode_StateMachine_1.RidingFullBody.AnimStateNode_4.SuperSlide.AnimGraphNode_StateMachine_0.SuperSlideDirection.AnimStateTransitionNode_10.AnimationTransitionGraph_1
    src=SuperSlideStartRight
    via=AnimStateTransitionNode_10
   }
  ]
 }
 Conduit {
  path=VehicleFullBodyOverride.AnimGraphNode_StateMachine_1.RidingFullBody.AnimStateNode_4.SuperSlide.AnimGraphNode_StateMachine_0.SuperSlideDirection.AnimStateConduitNode_0.Conduit
  nodes [Result(bCanEnterTransition=true)]
 }
 SuperSlideStartLeft {
  path=VehicleFullBodyOverride.AnimGraphNode_StateMachine_1.RidingFullBody.AnimStateNode_4.SuperSlide.AnimGraphNode_StateMachine_0.SuperSlideDirection.AnimStateNode_0.SuperSlideStartLeft
  nodes [Output Animation Pose(Result=Blendspace Player 'Sportbike_SuperSlide_Start_Left_BS'!(Y=0) -> (Pose) {blendSpace=/Sportbike/Animation/Driver/Sportbike_SuperSlide_Start_Left_BS})]
 }
 SuperSlideStartRight {
  path=VehicleFullBodyOverride.AnimGraphNode_StateMachine_1.RidingFullBody.AnimStateNode_4.SuperSlide.AnimGraphNode_StateMachine_0.SuperSlideDirection.AnimStateNode_1.SuperSlideStartRight
  nodes [Output Animation Pose(Result=Blendspace Player 'Sportbike_SuperSlide_Start_Right_BS'!(Y=0) -> (Pose) {blendSpace=/Sportbike/Animation/Driver/Sportbike_SuperSlide_Start_Right_BS})]
 }
 SuperSlideLoopRight {
  path=VehicleFullBodyOverride.AnimGraphNode_StateMachine_1.RidingFullBody.AnimStateNode_4.SuperSlide.AnimGraphNode_StateMachine_0.SuperSlideDirection.AnimStateNode_2.SuperSlideLoopRight
  nodes [Output Animation Pose(Result=Blendspace Player 'Sportbike_SuperSlide_Loop_Right_BS'!(Y=0) -> (Pose) {blendSpace=/Sportbike/Animation/Driver/Sportbike_SuperSlide_Loop_Right_BS})]
 }
 SuperSlideLoopLeft {
  path=VehicleFullBodyOverride.AnimGraphNode_StateMachine_1.RidingFullBody.AnimStateNode_4.SuperSlide.AnimGraphNode_StateMachine_0.SuperSlideDirection.AnimStateNode_3.SuperSlideLoopLeft
  nodes [Output Animation Pose(Result=Blendspace Player 'Sportbike_SuperSlide_Loop_Left_BS'!(Y=0) -> (Pose) {blendSpace=/Sportbike/Animation/Driver/Sportbike_SuperSlide_Loop_Left_BS})]
 }
 AnimationTransitionGraph_1 {
  path=VehicleFullBodyOverride.AnimGraphNode_StateMachine_1.RidingFullBody.AnimStateNode_4.SuperSlide.AnimGraphNode_StateMachine_0.SuperSlideDirection.AnimStateTransitionNode_10.AnimationTransitionGraph_1
  nodes [Result()]
 }
 AnimationTransitionGraph_1 {
  path=VehicleFullBodyOverride.AnimGraphNode_StateMachine_1.RidingFullBody.AnimStateNode_4.SuperSlide.AnimGraphNode_StateMachine_0.SuperSlideDirection.AnimStateTransitionNode_11.AnimationTransitionGraph_1
  nodes [Result()]
 }
 AnimationTransitionGraph_1 {
  path=VehicleFullBodyOverride.AnimGraphNode_StateMachine_1.RidingFullBody.AnimStateNode_4.SuperSlide.AnimGraphNode_StateMachine_0.SuperSlideDirection.AnimStateTransitionNode_8.AnimationTransitionGraph_1
  nodes [Result(bCanEnterTransition=true)]
 }
 Transition {
  path=VehicleFullBodyOverride.AnimGraphNode_StateMachine_1.RidingFullBody.AnimStateNode_4.SuperSlide.AnimGraphNode_StateMachine_0.SuperSlideDirection.AnimStateTransitionNode_9.Transition
  nodes [Result(bCanEnterTransition=Get bIsRightInput() -> (bIsRightInput:bool))]
 }
 Swim {
  path=VehicleFullBodyOverride.AnimGraphNode_StateMachine_1.RidingFullBody.AnimStateNode_5.Swim
  nodes [
   n1 = Use cached pose 'ReverseNoFeetPose'()
   n2 = Use cached pose 'LocomotionLoopPose'()
   n3 = Get MotorcycleInput()
   n4 = Apply Additive(
    Base=Sequence Player 'MotoDirt_Driver_Idle_Slow'() {sequence=/Motorcycle/Animation/Driver/MotoDirt_Driver_Idle_Slow}
    Additive=Blendspace Player 'MotoDirt_Driver_Steer_BS'(X=Break Fort Anim Input Motorcycle(FortAnimInput_Motorcycle=@n3) -> (
      bIsUsingMotorcycle=~
      bIsDriver=~
      bIsFrontPassenger=~
      bIsBackPassenger=~
      bHasPassenger=~
      bIsBackPassengerAndLeaning=~
      bIsDrifting=~
      bIsBoosting=~
      bIsReversing=~
      bIsBraking=~
      bIsMoving=~
      bIsMovingForward=~
      bIsSteeringRight=~
      bIsSteeringLeft=~
      RunForwardAlpha=~
      VerticalVelocity=~
      VerticalAcceleration=~
      bAimFWD=~
      bAimBWD=~
      bAimLFT=~
      bAimRGT=~
      PawnToVehicleDeltaYawAngleDegrees=~
      AimCardDirDeadZoneAngleDegrees=~
      AimCardDirAngleOffsetDegrees=~
      AimFWDDeltaAngleDegrees=~
      AimBWDDeltaAngleDegrees=~
      AimLFTDeltaAngleDegrees=~
      AimRGTDeltaAngleDegrees=~
      SlopePitchDegreeAngle=~
      SlopeRollDegreeAngle=~
      SteerAngle=~
      SteerAlpha=~
      SteerAngleDeadZoneDegrees=~
      SteeringRotation=~
      VehiclePitch=~
      VehicleRoll=~
     ).SteerAngle) {blendSpace=&MotoDirt_Driver_Steer_BS}
   ) {alphaBoolBlend {blendOption=Linear}}
   Output Animation Pose(Result=Blend Poses by bool(
     BlendPose_0=Blend Poses by bool(BlendPose_0=@n1, BlendPose_1=@n2, BlendTime_0=0.5, BlendTime_1=0.5) {blendType=Linear}
     BlendPose_1=AimOffset Player 'MotoDirt_Driver_Pitch_AO_IdleSwim'!(X:float=0, BasePose:Pose Link Structure=@n4, Alpha:float=1) -> (Pose:Pose Link Structure) {blendSpace=/Motorcycle/Animation/Driver/AO/MotoDirt_Driver_Pitch_AO_IdleSwim, alphaBoolBlend {blendOption=Linear}}
     BlendTime_0=0.5
     BlendTime_1=0.5
    ) {blendType=Linear})
  ]
 }
 Wheelie {
  path=VehicleFullBodyOverride.AnimGraphNode_StateMachine_1.RidingFullBody.AnimStateNode_6.Wheelie
  nodes [
   AnimGraphNode_StateMachine_1 = WheelieTransition() {is=title AnimGraphNode_StateMachine}
   AnimGraphNode_StateMachine_2 = WheelieAdditiveCorrection() {is=title AnimGraphNode_StateMachine}
   AnimGraphNode_StateMachine_0 = GroundTrickTransition() {is=title AnimGraphNode_StateMachine}
   Output Animation Pose(Result=Blend Poses by bool(
     BlendPose_0=Apply Additive(Base=@AnimGraphNode_StateMachine_1, Additive=@AnimGraphNode_StateMachine_2) {alphaBoolBlend {blendOption=Linear}}
     BlendPose_1=@AnimGraphNode_StateMachine_0
    ) {blendType=Linear})
  ]
 }
 GroundTrickTransition {
  entry=GroundTrickStart
  path=VehicleFullBodyOverride.AnimGraphNode_StateMachine_1.RidingFullBody.AnimStateNode_6.Wheelie.AnimGraphNode_StateMachine_0.GroundTrickTransition
  ids=id
  nodes [
   AnimStateNode_0 = GroundTrickStart() {is=title AnimStateNode}
   AnimStateNode_1 = GroundTrickLoop() {is=title AnimStateNode}
   AnimStateTransitionNode_2 = AnimStateTransitionNode() {bAutomaticRuleBasedOnSequencePlayerInState=true}
  ]
  links [{
    dst=GroundTrickLoop
    rule=VehicleFullBodyOverride.AnimGraphNode_StateMachine_1.RidingFullBody.AnimStateNode_6.Wheelie.AnimGraphNode_StateMachine_0.GroundTrickTransition.AnimStateTransitionNode_2.Transition
    src=GroundTrickStart
    via=AnimStateTransitionNode_2
   }]
 }
 GroundTrickStart {
  path=VehicleFullBodyOverride.AnimGraphNode_StateMachine_1.RidingFullBody.AnimStateNode_6.Wheelie.AnimGraphNode_StateMachine_0.GroundTrickTransition.AnimStateNode_0.GroundTrickStart
  nodes [Output Animation Pose(Result=Sequence Player 'SportBike_Driver_GroundTrick_Intro'() {bLoopAnimation=false, sequence=/Sportbike/Animation/Driver/SportBike_Driver_GroundTrick_Intro})]
 }
 GroundTrickLoop {
  path=VehicleFullBodyOverride.AnimGraphNode_StateMachine_1.RidingFullBody.AnimStateNode_6.Wheelie.AnimGraphNode_StateMachine_0.GroundTrickTransition.AnimStateNode_1.GroundTrickLoop
  nodes [Output Animation Pose(Result=Sequence Player 'SportBike_Driver_GroundTrick_Loop'() {sequence=/Sportbike/Animation/Driver/SportBike_Driver_GroundTrick_Loop})]
 }
 Transition {
  path=VehicleFullBodyOverride.AnimGraphNode_StateMachine_1.RidingFullBody.AnimStateNode_6.Wheelie.AnimGraphNode_StateMachine_0.GroundTrickTransition.AnimStateTransitionNode_2.Transition
  nodes [Result()]
 }
 WheelieTransition {
  entry=WheelieStart
  path=VehicleFullBodyOverride.AnimGraphNode_StateMachine_1.RidingFullBody.AnimStateNode_6.Wheelie.AnimGraphNode_StateMachine_1.WheelieTransition
  ids=id
  nodes [
   AnimStateNode_0 = WheelieStart() {is=title AnimStateNode}
   AnimStateNode_1 = WheelieLoop() {is=title AnimStateNode}
   AnimStateTransitionNode_2 = AnimStateTransitionNode() {bAutomaticRuleBasedOnSequencePlayerInState=true}
  ]
  links [{
    dst=WheelieLoop
    rule=VehicleFullBodyOverride.AnimGraphNode_StateMachine_1.RidingFullBody.AnimStateNode_6.Wheelie.AnimGraphNode_StateMachine_1.WheelieTransition.AnimStateTransitionNode_2.Transition
    src=WheelieStart
    via=AnimStateTransitionNode_2
   }]
 }
 WheelieStart {
  path=VehicleFullBodyOverride.AnimGraphNode_StateMachine_1.RidingFullBody.AnimStateNode_6.Wheelie.AnimGraphNode_StateMachine_1.WheelieTransition.AnimStateNode_0.WheelieStart
  nodes [Output Animation Pose(Result=Sequence Player 'MotoDirt_Driver_Wheelie_Start'() {bLoopAnimation=false, sequence=/Motorcycle/Animation/Driver/MotoDirt_Driver_Wheelie_Start})]
 }
 WheelieLoop {
  path=VehicleFullBodyOverride.AnimGraphNode_StateMachine_1.RidingFullBody.AnimStateNode_6.Wheelie.AnimGraphNode_StateMachine_1.WheelieTransition.AnimStateNode_1.WheelieLoop
  nodes [Output Animation Pose(Result=Sequence Player 'MotoDirt_Driver_Wheelie_Loop'() {sequence=/Motorcycle/Animation/Driver/MotoDirt_Driver_Wheelie_Loop})]
 }
 Transition {
  path=VehicleFullBodyOverride.AnimGraphNode_StateMachine_1.RidingFullBody.AnimStateNode_6.Wheelie.AnimGraphNode_StateMachine_1.WheelieTransition.AnimStateTransitionNode_2.Transition
  nodes [Result()]
 }
 WheelieAdditiveCorrection {
  entry=AdditivePose
  path=VehicleFullBodyOverride.AnimGraphNode_StateMachine_1.RidingFullBody.AnimStateNode_6.Wheelie.AnimGraphNode_StateMachine_2.WheelieAdditiveCorrection
  ids=id
  nodes [AnimStateNode_0 = AdditivePose()]
 }
 AdditivePose {
  path=VehicleFullBodyOverride.AnimGraphNode_StateMachine_1.RidingFullBody.AnimStateNode_6.Wheelie.AnimGraphNode_StateMachine_2.WheelieAdditiveCorrection.AnimStateNode_0.AdditivePose
  nodes [Output Animation Pose(Result=Blend Poses by bool(
     BlendPose_0="Sequence Player (additive) 'Sportbike_Driver_WheelieAdditivePose'"() {sequence=/Sportbike/Animation/Driver/Sportbike_Driver_WheelieAdditivePose}
     BlendPose_1=Additive Identity Pose()
    ) {blendType=Linear})]
 }
 Stoppie {
  path=VehicleFullBodyOverride.AnimGraphNode_StateMachine_1.RidingFullBody.AnimStateNode_7.Stoppie
  nodes [
   AnimGraphNode_StateMachine_0 = StoppieAdditiveCorrection() {is=title AnimGraphNode_StateMachine}
   AnimGraphNode_StateMachine_1 = StoppieTransition() {is=title AnimGraphNode_StateMachine}
   Output Animation Pose(Result=Apply Additive(Base=@AnimGraphNode_StateMachine_1, Additive=@AnimGraphNode_StateMachine_0) {alphaBoolBlend {blendOption=Linear}})
  ]
 }
 StoppieAdditiveCorrection {
  entry=StoppieStart
  path=VehicleFullBodyOverride.AnimGraphNode_StateMachine_1.RidingFullBody.AnimStateNode_7.Stoppie.AnimGraphNode_StateMachine_0.StoppieAdditiveCorrection
  ids=id
  nodes [
   AnimStateNode_0 = StoppieStart() {is=title AnimStateNode}
   AnimStateNode_1 = StoppieLoop() {is=title AnimStateNode}
   AnimStateTransitionNode_10 = AnimStateTransitionNode() {bAutomaticRuleBasedOnSequencePlayerInState=true}
  ]
  links [{
    dst=StoppieLoop
    rule=VehicleFullBodyOverride.AnimGraphNode_StateMachine_1.RidingFullBody.AnimStateNode_7.Stoppie.AnimGraphNode_StateMachine_0.StoppieAdditiveCorrection.AnimStateTransitionNode_10.Transition
    src=StoppieStart
    via=AnimStateTransitionNode_10
   }]
 }
 StoppieStart {
  path=VehicleFullBodyOverride.AnimGraphNode_StateMachine_1.RidingFullBody.AnimStateNode_7.Stoppie.AnimGraphNode_StateMachine_0.StoppieAdditiveCorrection.AnimStateNode_0.StoppieStart
  nodes [Output Animation Pose(Result="Sequence Player (additive) 'SportBike_Driver_StoppieStart'"() {bLoopAnimation=false, sequence=/Sportbike/Animation/Driver/SportBike_Driver_StoppieStart})]
 }
 StoppieLoop {
  path=VehicleFullBodyOverride.AnimGraphNode_StateMachine_1.RidingFullBody.AnimStateNode_7.Stoppie.AnimGraphNode_StateMachine_0.StoppieAdditiveCorrection.AnimStateNode_1.StoppieLoop
  nodes [Output Animation Pose(Result="Sequence Player (additive) 'SportBike_Driver_StoppieLoop'"() {sequence=/Sportbike/Animation/Driver/SportBike_Driver_StoppieLoop})]
 }
 Transition {
  path=VehicleFullBodyOverride.AnimGraphNode_StateMachine_1.RidingFullBody.AnimStateNode_7.Stoppie.AnimGraphNode_StateMachine_0.StoppieAdditiveCorrection.AnimStateTransitionNode_10.Transition
  nodes [Result()]
 }
 StoppieTransition {
  entry=StoppieStart
  path=VehicleFullBodyOverride.AnimGraphNode_StateMachine_1.RidingFullBody.AnimStateNode_7.Stoppie.AnimGraphNode_StateMachine_1.StoppieTransition
  ids=id
  nodes [
   AnimStateNode_0 = StoppieStart()
   AnimStateNode_1 = StoppieLoop()
   AnimStateTransitionNode_7 = AnimStateTransitionNode() {bAutomaticRuleBasedOnSequencePlayerInState=true}
  ]
  links [{
    dst=StoppieLoop
    rule=VehicleFullBodyOverride.AnimGraphNode_StateMachine_1.RidingFullBody.AnimStateNode_7.Stoppie.AnimGraphNode_StateMachine_1.StoppieTransition.AnimStateTransitionNode_7.Transition
    src=StoppieStart
    via=AnimStateTransitionNode_7
   }]
 }
 StoppieStart {
  path=VehicleFullBodyOverride.AnimGraphNode_StateMachine_1.RidingFullBody.AnimStateNode_7.Stoppie.AnimGraphNode_StateMachine_1.StoppieTransition.AnimStateNode_0.StoppieStart
  nodes [Output Animation Pose(Result=Sequence Player 'MotoDirt_Driver_StoppieStart'() {bLoopAnimation=false, sequence=/Motorcycle/Animation/Driver/MotoDirt_Driver_StoppieStart})]
 }
 StoppieLoop {
  path=VehicleFullBodyOverride.AnimGraphNode_StateMachine_1.RidingFullBody.AnimStateNode_7.Stoppie.AnimGraphNode_StateMachine_1.StoppieTransition.AnimStateNode_1.StoppieLoop
  nodes [Output Animation Pose(Result=Sequence Player 'MotoDirt_Driver_StoppieLoop'() {sequence=/Motorcycle/Animation/Driver/MotoDirt_Driver_StoppieLoop})]
 }
 Transition {
  path=VehicleFullBodyOverride.AnimGraphNode_StateMachine_1.RidingFullBody.AnimStateNode_7.Stoppie.AnimGraphNode_StateMachine_1.StoppieTransition.AnimStateTransitionNode_7.Transition
  nodes [Result()]
 }
 StoppieLand {
  path=VehicleFullBodyOverride.AnimGraphNode_StateMachine_1.RidingFullBody.AnimStateNode_8.StoppieLand
  nodes [Output Animation Pose(Result=Apply Additive(
     Base=Sequence Player 'MotoDirt_Driver_StoppieEnd'!(StartPosition:float=0) -> (Pose) {bLoopAnimation=false, sequence=/Motorcycle/Animation/Driver/MotoDirt_Driver_StoppieEnd}
     Additive="Sequence Player (additive) 'SportBike_Driver_StoppieEnd'"!(StartPosition:float=0) -> (Pose) {bLoopAnimation=false, sequence=/Sportbike/Animation/Driver/SportBike_Driver_StoppieEnd}
    ) {alphaBoolBlend {blendOption=Linear}})]
 }
 Braking {
  path=VehicleFullBodyOverride.AnimGraphNode_StateMachine_1.RidingFullBody.AnimStateNode_9.Braking
  nodes [
   AnimGraphNode_StateMachine_0 = BrakingTransition() {is=title AnimGraphNode_StateMachine}
   Output Animation Pose(Result=@AnimGraphNode_StateMachine_0)
  ]
 }
 BrakingTransition {
  entry=BrakingStart
  path=VehicleFullBodyOverride.AnimGraphNode_StateMachine_1.RidingFullBody.AnimStateNode_9.Braking.AnimGraphNode_StateMachine_0.BrakingTransition
  ids=id
  nodes [
   AnimStateNode_0 = BrakingStart() {is=title AnimStateNode}
   AnimStateNode_1 = BrakingLoop() {is=title AnimStateNode}
   AnimStateTransitionNode_1 = AnimStateTransitionNode() {bAutomaticRuleBasedOnSequencePlayerInState=true}
  ]
  links [{
    dst=BrakingLoop
    rule=VehicleFullBodyOverride.AnimGraphNode_StateMachine_1.RidingFullBody.AnimStateNode_9.Braking.AnimGraphNode_StateMachine_0.BrakingTransition.AnimStateTransitionNode_1.Transition
    src=BrakingStart
    via=AnimStateTransitionNode_1
   }]
 }
 BrakingStart {
  path=VehicleFullBodyOverride.AnimGraphNode_StateMachine_1.RidingFullBody.AnimStateNode_9.Braking.AnimGraphNode_StateMachine_0.BrakingTransition.AnimStateNode_0.BrakingStart
  nodes [Output Animation Pose(Result=Sequence Player 'Sportbike_Driver_Brake_Start'() {bLoopAnimation=false, sequence=/Sportbike/Animation/Driver/Sportbike_Driver_Brake_Start})]
 }
 BrakingLoop {
  path=VehicleFullBodyOverride.AnimGraphNode_StateMachine_1.RidingFullBody.AnimStateNode_9.Braking.AnimGraphNode_StateMachine_0.BrakingTransition.AnimStateNode_1.BrakingLoop
  nodes [Output Animation Pose(Result=Sequence Player 'Sportbike_Driver_Brake_Pose'() {sequence=/Sportbike/Animation/Driver/Sportbike_Driver_Brake_Pose})]
 }
 Transition {
  path=VehicleFullBodyOverride.AnimGraphNode_StateMachine_1.RidingFullBody.AnimStateNode_9.Braking.AnimGraphNode_StateMachine_0.BrakingTransition.AnimStateTransitionNode_1.Transition
  nodes [Result()]
 }
 Transition {
  path=VehicleFullBodyOverride.AnimGraphNode_StateMachine_1.RidingFullBody.AnimStateTransitionNode_0.Transition
  nodes [Result(bCanEnterTransition=Get bShouldPlayReverseOutro() -> (bShouldPlayReverseOutro:bool))]
 }
 Transition {
  path=VehicleFullBodyOverride.AnimGraphNode_StateMachine_1.RidingFullBody.AnimStateTransitionNode_1.Transition
  nodes [Result(bCanEnterTransition=NOT Boolean(A=Get IsSuperSliding() -> (IsSuperSliding:bool)))]
 }
 Transition {
  path=VehicleFullBodyOverride.AnimGraphNode_StateMachine_1.RidingFullBody.AnimStateTransitionNode_10.Transition
  nodes [Result(bCanEnterTransition=Get bStoppie())]
 }
 Transition {
  path=VehicleFullBodyOverride.AnimGraphNode_StateMachine_1.RidingFullBody.AnimStateTransitionNode_12.Transition
  nodes [Result(bCanEnterTransition=Get bIsSurfaceWater() -> (bIsSurfaceWater:bool))]
 }
 Transition {
  path=VehicleFullBodyOverride.AnimGraphNode_StateMachine_1.RidingFullBody.AnimStateTransitionNode_13.Transition
  nodes [Result(bCanEnterTransition=NOT Boolean(A=Get bIsSurfaceWater()))]
 }
 Transition {
  path=VehicleFullBodyOverride.AnimGraphNode_StateMachine_1.RidingFullBody.AnimStateTransitionNode_14.Transition
  nodes [Result(bCanEnterTransition=Get bShouldExitFromLocomotion() -> (bShouldExitFromLocomotion:bool))]
 }
 Transition {
  path=VehicleFullBodyOverride.AnimGraphNode_StateMachine_1.RidingFullBody.AnimStateTransitionNode_15.Transition
  nodes [Result(bCanEnterTransition=NOT Boolean(A=Get bStoppie()))]
 }
 Transition {
  path=VehicleFullBodyOverride.AnimGraphNode_StateMachine_1.RidingFullBody.AnimStateTransitionNode_16.Transition
  nodes [Result(bCanEnterTransition=Get bStateRuleFallingToMoving() -> (bStateRuleFallingToMoving:bool))]
 }
 Transition {
  path=VehicleFullBodyOverride.AnimGraphNode_StateMachine_1.RidingFullBody.AnimStateTransitionNode_17.Transition
  nodes [Result(bCanEnterTransition=Get bIsSurfaceWater())]
 }
 Transition {
  path=VehicleFullBodyOverride.AnimGraphNode_StateMachine_1.RidingFullBody.AnimStateTransitionNode_18.Transition
  nodes [Result(bCanEnterTransition=Get bWheelie() -> (bWheelie:bool))]
 }
 Transition {
  path=VehicleFullBodyOverride.AnimGraphNode_StateMachine_1.RidingFullBody.AnimStateTransitionNode_19.Transition
  nodes [Result(bCanEnterTransition=Get bShouldBrake() -> (bShouldBrake:bool))]
 }
 Transition {
  path=VehicleFullBodyOverride.AnimGraphNode_StateMachine_1.RidingFullBody.AnimStateTransitionNode_2.Transition
  nodes [Result(bCanEnterTransition=Get bShouldAllowJump() -> (bShouldAllowJump:bool))]
 }
 AnimationTransitionGraph_1 {
  path=VehicleFullBodyOverride.AnimGraphNode_StateMachine_1.RidingFullBody.AnimStateTransitionNode_20.AnimationTransitionGraph_1
  nodes [Result(bCanEnterTransition=NOT Boolean(A=Get bIsBraking() -> (bIsBraking:bool)))]
 }
 Transition {
  path=VehicleFullBodyOverride.AnimGraphNode_StateMachine_1.RidingFullBody.AnimStateTransitionNode_21.Transition
  nodes [Result(bCanEnterTransition=Get bStoppie())]
 }
 Transition {path=VehicleFullBodyOverride.AnimGraphNode_StateMachine_1.RidingFullBody.AnimStateTransitionNode_22.Transition, nodes [Result()]}
 Transition {
  path=VehicleFullBodyOverride.AnimGraphNode_StateMachine_1.RidingFullBody.AnimStateTransitionNode_23.Transition
  nodes [Result(bCanEnterTransition=Get bEarlyOutBrakingOutro() -> (bEarlyOutBrakingOutro:bool))]
 }
 Transition {
  path=VehicleFullBodyOverride.AnimGraphNode_StateMachine_1.RidingFullBody.AnimStateTransitionNode_25.Transition
  nodes [
   n1 = "Get Relevant Anim Time Remaining (GroundTrickOutro)"!() -> (ReturnValue:float) {is=title K2Node_AnimGetter}
   Result(bCanEnterTransition=AND Boolean(A="float < float"(A=@n1, B=0.2), B=Get bIsBoosting() -> (bIsBoosting:bool)))
  ]
 }
 Transition {
  path=VehicleFullBodyOverride.AnimGraphNode_StateMachine_1.RidingFullBody.AnimStateTransitionNode_26.Transition
  nodes [Result(bCanEnterTransition=Get bIsMoving())]
 }
 Transition {
  path=VehicleFullBodyOverride.AnimGraphNode_StateMachine_1.RidingFullBody.AnimStateTransitionNode_27.Transition
  nodes [Result(bCanEnterTransition=Get bHasJumpChargeStarted() -> (bHasJumpChargeStarted:bool))]
 }
 Transition {
  path=VehicleFullBodyOverride.AnimGraphNode_StateMachine_1.RidingFullBody.AnimStateTransitionNode_28.Transition
  nodes [Result(bCanEnterTransition=Get bIsLandingInPlace() -> (bIsLandingInPlace:bool))]
 }
 Transition {
  path=VehicleFullBodyOverride.AnimGraphNode_StateMachine_1.RidingFullBody.AnimStateTransitionNode_29.Transition
  nodes [Result(bCanEnterTransition=Get StateRuleFromWheelieToGroundTrickOutro() -> (StateRuleFromWheelieToGroundTrickOutro:bool))]
 }
 Transition {
  path=VehicleFullBodyOverride.AnimGraphNode_StateMachine_1.RidingFullBody.AnimStateTransitionNode_3.Transition
  nodes [Result(bCanEnterTransition=Get bShouldJump() -> (bShouldJump:bool))]
 }
 Transition {path=VehicleFullBodyOverride.AnimGraphNode_StateMachine_1.RidingFullBody.AnimStateTransitionNode_31.Transition, nodes [Result()]}
 Transition {
  path=VehicleFullBodyOverride.AnimGraphNode_StateMachine_1.RidingFullBody.AnimStateTransitionNode_32.Transition
  nodes [Result(bCanEnterTransition=Get bIsBoosting())]
 }
 Transition {
  path=VehicleFullBodyOverride.AnimGraphNode_StateMachine_1.RidingFullBody.AnimStateTransitionNode_33.Transition
  nodes [Result(bCanEnterTransition=NOT Boolean(A=Get bIsBoosting()))]
 }
 Transition {
  path=VehicleFullBodyOverride.AnimGraphNode_StateMachine_1.RidingFullBody.AnimStateTransitionNode_34.Transition
  nodes [Result(bCanEnterTransition=Get bShouldStop() -> (bShouldStop:bool))]
 }
 Transition {
  path=VehicleFullBodyOverride.AnimGraphNode_StateMachine_1.RidingFullBody.AnimStateTransitionNode_35.Transition
  nodes [Result(bCanEnterTransition=Get EarlyExitFromGroundTrick() -> (EarlyExitFromGroundTrick:bool))]
 }
 Transition {path=VehicleFullBodyOverride.AnimGraphNode_StateMachine_1.RidingFullBody.AnimStateTransitionNode_36.Transition, nodes [Result()]}
 Transition {
  path=VehicleFullBodyOverride.AnimGraphNode_StateMachine_1.RidingFullBody.AnimStateTransitionNode_4.Transition
  nodes [
   n1 = Get bWheelie()
   Result(bCanEnterTransition=@n1)
   AND Boolean(A=@n1, B="float > float"(A=Get VehiclePitch() -> (VehiclePitch:float), B=10))
  ]
 }
 Transition {
  path=VehicleFullBodyOverride.AnimGraphNode_StateMachine_1.RidingFullBody.AnimStateTransitionNode_5.Transition
  nodes [Result(bCanEnterTransition=Get IsBoostingWithNoPassenger() -> (IsBoostingWithNoPassenger:bool))]
 }
 Transition {path=VehicleFullBodyOverride.AnimGraphNode_StateMachine_1.RidingFullBody.AnimStateTransitionNode_50.Transition, nodes [Result()]}
 Transition {
  path=VehicleFullBodyOverride.AnimGraphNode_StateMachine_1.RidingFullBody.AnimStateTransitionNode_6.Transition
  nodes [Result(bCanEnterTransition=true)]
 }
 Transition {
  path=VehicleFullBodyOverride.AnimGraphNode_StateMachine_1.RidingFullBody.AnimStateTransitionNode_7.Transition
  nodes [Result(bCanEnterTransition=Get SportbikeWheelieEarlyOut() -> (SportbikeWheelieEarlyOut:bool))]
 }
 Transition {
  path=VehicleFullBodyOverride.AnimGraphNode_StateMachine_1.RidingFullBody.AnimStateTransitionNode_75.Transition
  nodes [Result(bCanEnterTransition=true)]
 }
 Transition {
  path=VehicleFullBodyOverride.AnimGraphNode_StateMachine_1.RidingFullBody.AnimStateTransitionNode_79.Transition
  nodes [Result(bCanEnterTransition=Break Fort Anim Input Motorcycle(FortAnimInput_Motorcycle=Get MotorcycleInput()) -> (
     bIsUsingMotorcycle=~
     bIsDriver=~
     bIsFrontPassenger=~
     bIsBackPassenger=~
     bHasPassenger=~
     bIsBackPassengerAndLeaning=~
     bIsDrifting=~
     bIsBoosting=~
     bIsReversing=~
     bIsBraking=~
     bIsMoving=~
     bIsMovingForward=~
     bIsSteeringRight=~
     bIsSteeringLeft=~
     RunForwardAlpha=~
     VerticalVelocity=~
     VerticalAcceleration=~
     bAimFWD=~
     bAimBWD=~
     bAimLFT=~
     bAimRGT=~
     PawnToVehicleDeltaYawAngleDegrees=~
     AimCardDirDeadZoneAngleDegrees=~
     AimCardDirAngleOffsetDegrees=~
     AimFWDDeltaAngleDegrees=~
     AimBWDDeltaAngleDegrees=~
     AimLFTDeltaAngleDegrees=~
     AimRGTDeltaAngleDegrees=~
     SlopePitchDegreeAngle=~
     SlopeRollDegreeAngle=~
     SteerAngle=~
     SteerAlpha=~
     SteerAngleDeadZoneDegrees=~
     SteeringRotation=~
     VehiclePitch=~
     VehicleRoll=~
    ).bIsReversing)]
 }
 Transition {
  path=VehicleFullBodyOverride.AnimGraphNode_StateMachine_1.RidingFullBody.AnimStateTransitionNode_8.Transition
  nodes [Result(bCanEnterTransition=Get IsSuperSliding())]
 }
 Transition {path=VehicleFullBodyOverride.AnimGraphNode_StateMachine_1.RidingFullBody.AnimStateTransitionNode_9.Transition, nodes [Result()]}
 UpperBody {
  entry=FullBodyPose
  path=VehicleFullBodyOverride.AnimGraphNode_StateMachine_2.UpperBody
  ids=id
  nodes [
   AnimStateNode_0 = FullBodyPose() {is=title AnimStateNode, stateFullyBlended {guid=#24, notifyName=CombatTransitionOFF}}
   AnimStateNode_3 = JumpCharge() {is=title AnimStateNode, stateEntered {guid=#25, notifyName=JumpCharge}}
   AnimStateTransitionNode_1 = AnimStateTransitionNode() {crossfadeDuration=0.35}
   AnimStateTransitionNode_3 = AnimStateTransitionNode() {crossfadeDuration=0.4, logicType=TLT_Inertialization}
  ]
  links [
   {
    dst=JumpCharge
    rule=VehicleFullBodyOverride.AnimGraphNode_StateMachine_2.UpperBody.AnimStateTransitionNode_1.Transition
    src=FullBodyPose
    via=AnimStateTransitionNode_1
   }
   {
    dst=FullBodyPose
    rule=VehicleFullBodyOverride.AnimGraphNode_StateMachine_2.UpperBody.AnimStateTransitionNode_3.Transition
    src=JumpCharge
    via=AnimStateTransitionNode_3
   }
  ]
 }
 FullBodyPose {
  path=VehicleFullBodyOverride.AnimGraphNode_StateMachine_2.UpperBody.AnimStateNode_0.FullBodyPose
  nodes [Output Animation Pose(Result=Use cached pose 'FullBodyRiding'())]
 }
 JumpCharge {
  path=VehicleFullBodyOverride.AnimGraphNode_StateMachine_2.UpperBody.AnimStateNode_3.JumpCharge
  nodes [
   n1 = Blendspace Player 'MotoDirt_Driver_JumpCharge_Leans_BS'!(X=Break Fort Anim Input Motorcycle(FortAnimInput_Motorcycle=Get MotorcycleInput()) -> (
     bIsUsingMotorcycle=~
     bIsDriver=~
     bIsFrontPassenger=~
     bIsBackPassenger=~
     bHasPassenger=~
     bIsBackPassengerAndLeaning=~
     bIsDrifting=~
     bIsBoosting=~
     bIsReversing=~
     bIsBraking=~
     bIsMoving=~
     bIsMovingForward=~
     bIsSteeringRight=~
     bIsSteeringLeft=~
     RunForwardAlpha=~
     VerticalVelocity=~
     VerticalAcceleration=~
     bAimFWD=~
     bAimBWD=~
     bAimLFT=~
     bAimRGT=~
     PawnToVehicleDeltaYawAngleDegrees=~
     AimCardDirDeadZoneAngleDegrees=~
     AimCardDirAngleOffsetDegrees=~
     AimFWDDeltaAngleDegrees=~
     AimBWDDeltaAngleDegrees=~
     AimLFTDeltaAngleDegrees=~
     AimRGTDeltaAngleDegrees=~
     SlopePitchDegreeAngle=~
     SlopeRollDegreeAngle=~
     SteerAngle=~
     SteerAlpha=~
     SteerAngleDeadZoneDegrees=~
     SteeringRotation=~
     VehiclePitch=~
     VehicleRoll=~
    ).SteerAngle) -> (Pose) {blendSpace=/Motorcycle/Animation/Driver/BlendSpace/MotoDirt_Driver_JumpCharge_Leans_BS}
   n2 = Sequence Player 'MotoDirt_Driver_BunnyHop_Charge_Intro'() {bLoopAnimation=false, sequence=/Motorcycle/Animation/Driver/MotoDirt_Driver_BunnyHop_Charge_Intro}
   Output Animation Pose(Result=Apply Additive!(
     Base=Apply Additive!(Base=@n2, Additive=@n1, Alpha:float) -> (Pose) {alphaBoolBlend {blendOption=Linear}}
     Additive="Sequence Player (additive) 'MotoDirt_Driver_BunnyHop_Charge_Loop'"!(PlayRate:float=2) -> (Pose) {sequence=/Motorcycle/Animation/Driver/MotoDirt_Driver_BunnyHop_Charge_Loop}
     bAlphaBoolEnabled:bool=True
    ) -> (Pose) {alphaInputType=Bool, alphaBoolBlend {blendInTime=0.2, blendOption=Linear, blendOutTime=0.2}})
  ]
 }
 Transition {
  path=VehicleFullBodyOverride.AnimGraphNode_StateMachine_2.UpperBody.AnimStateTransitionNode_1.Transition
  nodes [Result(bCanEnterTransition=Get bShouldJumpCharge() -> (bShouldJumpCharge:bool))]
 }
 Transition {
  path=VehicleFullBodyOverride.AnimGraphNode_StateMachine_2.UpperBody.AnimStateTransitionNode_3.Transition
  nodes [Result(bCanEnterTransition=NOT Boolean(A=Get bShouldJumpCharge()))]
 }
 CombatMode {
  entry=Conduit
  path=VehicleFullBodyOverride.AnimGraphNode_StateMachine_5.CombatMode
  ids=id
  nodes [
   AnimStateConduitNode_0 = Conduit()
   AnimStateNode_0 = CombatTransition() {is=title AnimStateNode}
   AnimStateNode_1 = CombatLoop() {is=title AnimStateNode}
   AnimStateTransitionNode_0 = AnimStateTransitionNode() {bAutomaticRuleBasedOnSequencePlayerInState=true}
   AnimStateTransitionNode_1 = AnimStateTransitionNode() {priorityOrder=100}
   AnimStateTransitionNode_2 = AnimStateTransitionNode()
  ]
  links [
   {
    dst=CombatLoop
    rule=VehicleFullBodyOverride.AnimGraphNode_StateMachine_5.CombatMode.AnimStateTransitionNode_0.Transition
    src=CombatTransition
    via=AnimStateTransitionNode_0
   }
   {
    dst=CombatLoop
    rule=VehicleFullBodyOverride.AnimGraphNode_StateMachine_5.CombatMode.AnimStateTransitionNode_2.Transition
    src=Conduit
    via=AnimStateTransitionNode_2
   }
   {
    dst=CombatTransition
    rule=VehicleFullBodyOverride.AnimGraphNode_StateMachine_5.CombatMode.AnimStateTransitionNode_1.Transition
    src=Conduit
    via=AnimStateTransitionNode_1
   }
  ]
 }
 Conduit {
  path=VehicleFullBodyOverride.AnimGraphNode_StateMachine_5.CombatMode.AnimStateConduitNode_0.Conduit
  nodes [Result(bCanEnterTransition=true)]
 }
 CombatTransition {
  path=VehicleFullBodyOverride.AnimGraphNode_StateMachine_5.CombatMode.AnimStateNode_0.CombatTransition
  nodes [
   n1 = Layered blend per bone(BasePose=Use cached pose 'PostUpperBodyPose'(), BlendPoses_0=Use cached pose 'UpperAndLowerPose'()) {
    bMeshSpaceRotationBlend=true
    layerSetup [{branchFilters [
       {boneName=Root}
       {blendDepth=-1, boneName=pelvis}
       {boneName=Spine_01}
       {blendDepth=-1, boneName=thigh_l}
       {blendDepth=-1, boneName=thigh_r}
       {blendDepth=-1, boneName=ik_foot_root}
      ]}]
   }
   Output Animation Pose(Result=Apply Additive(
     Base=@n1
     Additive="Sequence Player (additive) 'MotoDirt_Driver_Combat_Transition'"() {sequence=/Motorcycle/Animation/Driver/MotoDirt_Driver_Combat_Transition}
    ) {alphaBoolBlend {blendOption=Linear}})
  ]
 }
 CombatLoop {
  path=VehicleFullBodyOverride.AnimGraphNode_StateMachine_5.CombatMode.AnimStateNode_1.CombatLoop
  nodes [
   n1 = Layered blend per bone(BasePose=Use cached pose 'PostUpperBodyPose'(), BlendPoses_0=Use cached pose 'UpperAndLowerPose'()) {
    bMeshSpaceRotationBlend=true
    layerSetup [{branchFilters [
       {boneName=Root}
       {blendDepth=-1, boneName=pelvis}
       {boneName=Spine_01}
       {blendDepth=-1, boneName=thigh_l}
       {blendDepth=-1, boneName=thigh_r}
       {blendDepth=-1, boneName=ik_foot_root}
      ]}]
   }
   Output Animation Pose(Result=@n1)
  ]
 }
 Transition {path=VehicleFullBodyOverride.AnimGraphNode_StateMachine_5.CombatMode.AnimStateTransitionNode_0.Transition, nodes [Result()]}
 Transition {
  path=VehicleFullBodyOverride.AnimGraphNode_StateMachine_5.CombatMode.AnimStateTransitionNode_1.Transition
  nodes [Result(bCanEnterTransition=true)]
 }
 Transition {
  path=VehicleFullBodyOverride.AnimGraphNode_StateMachine_5.CombatMode.AnimStateTransitionNode_2.Transition
  nodes [Result(bCanEnterTransition=Get bIsUsingConsumable() -> (bIsUsingConsumable:bool))]
 }
 VehicleLowerBodyOverride {nodes [
   Input Pose() {name=InPoseLowerBodyDefault}
   Input Pose() {name=InPoseUpperBody}
   Output Pose() {layerGroup=VehicleOverride}
  ]}
 VehicleSplitBodyOverride {nodes [
   Input Pose() {name=InPoseSplitBody}
   Input Pose() {name=InPoseUpperAndLowerBody}
   Input Pose() {name=InPoseUpperAndLowerBodyRemoveRoot}
   Output Pose() {layerGroup=Vehicle Override}
  ]}
]
variables []
