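# Sportbike passenger AnimBlueprint on the Fortnite player skeleton, supplying VehicleOverride layers and reusing many dirt-bike animations.
# PassengerFullBody handles idle, locomotion, jumps, boost, wheelies and stoppies, then feeds additive and hand-IK layers into a cached riding pose, replaced when DBNO.
# CombatTransition moves between riding and stand-up combat; seated actions instead layer the incoming body pose over riding, with inertialized output.
# VehicleFinalPoseOverride lowers pelvis and root, applying foot IK.
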AnimBlueprint /Sportbike/Animation/Sportbike_Passenger_AnimBP
target_skeleton=&Fortnite_M_Avg_Player_Skeleton
graphs [
 AnimGraph {nodes [Output Pose() {layerGroup=None}]}
 EventGraph {links [], nodes []}
 FullBodyAdditiveGroup {nodes [
   n1 = Apply Additive!(
    Base=Input Pose() {name=InPoseFullBodyAdditive}
    Additive=Blendspace Player 'MotoDirt_Passenger_RidingSlopePitch_BS'!(X=0) -> (Pose) {blendSpace=/Motorcycle/Animation/Passenger/BlendSpace/MotoDirt_Passenger_RidingSlopePitch_BS}
    bAlphaBoolEnabled:bool=True
   ) -> (Pose) {alphaInputType=Bool, alphaBoolBlend {blendInTime=0.2, blendOption=Linear, blendOutTime=0.4}}
   AnimGraphNode_StateMachine_3 = FullBodyAdditive() {is=title AnimGraphNode_StateMachine, bAllowConduitEntryStates=true}
   Output Pose(Result=Apply Additive(
     Base=Apply Additive!(Base=@n1, Additive=@AnimGraphNode_StateMachine_3, Alpha:float=0) -> (Pose) {alphaBoolBlend {blendOption=Linear}}
     Additive="Sequence Player (additive) 'SportBike_Passenger_LeanBack'"() {bLoopAnimation=false, sequence=/Sportbike/Animation/Passenger/SportBike_Passenger_LeanBack}
     bAlphaBoolEnabled=True
    ) {alphaInputType=Bool, alphaBoolBlend {blendInTime=0.35, blendOutTime=0.55}}) {layerGroup=None}
  ]}
 FullBodyAdditive {
  entry=AdditivePose
  path=FullBodyAdditiveGroup.AnimGraphNode_StateMachine_3.FullBodyAdditive
  ids=id
  nodes [
   AnimStateNode_0 = AdditivePose() {is=title AnimStateNode}
   AnimStateNode_3 = HitReact() {is=title AnimStateNode}
   AnimStateTransitionNode_0 = AnimStateTransitionNode() {crossfadeDuration=0.033}
   AnimStateTransitionNode_5 = AnimStateTransitionNode() {bAutomaticRuleBasedOnSequencePlayerInState=true}
  ]
  links [
   {
    dst=HitReact
    rule=FullBodyAdditiveGroup.AnimGraphNode_StateMachine_3.FullBodyAdditive.AnimStateTransitionNode_0.Transition
    src=AdditivePose
    via=AnimStateTransitionNode_0
   }
   {
    dst=AdditivePose
    rule=FullBodyAdditiveGroup.AnimGraphNode_StateMachine_3.FullBodyAdditive.AnimStateTransitionNode_5.Transition
    src=HitReact
    via=AnimStateTransitionNode_5
   }
  ]
 }
 AdditivePose {
  path=FullBodyAdditiveGroup.AnimGraphNode_StateMachine_3.FullBodyAdditive.AnimStateNode_0.AdditivePose
  nodes [Output Animation Pose(Result=Additive Identity Pose())]
 }
 HitReact {
  path=FullBodyAdditiveGroup.AnimGraphNode_StateMachine_3.FullBodyAdditive.AnimStateNode_3.HitReact
  nodes [Output Animation Pose(Result="Sequence Player (additive) 'MotoDirt_Driver_Collision_N'"() {bLoopAnimation=false, sequence=/Motorcycle/Animation/Driver/MotoDirt_Driver_Collision_N})]
 }
 Transition {
  path=FullBodyAdditiveGroup.AnimGraphNode_StateMachine_3.FullBodyAdditive.AnimStateTransitionNode_0.Transition
  nodes [Result(bCanEnterTransition=Get bShouldPlayHitReact() -> (bShouldPlayHitReact:bool))]
 }
 Transition {path=FullBodyAdditiveGroup.AnimGraphNode_StateMachine_3.FullBodyAdditive.AnimStateTransitionNode_5.Transition, nodes [Result()]}
 IK/BoneModification {nodes [
   n1 = "Transform (Modify) Bone - Bone: ik_hand_root"!(ComponentPose=Local To Component(LocalPose=Input Pose() {name=InPoseIK}), bAlphaBoolEnabled:bool=True, Translation="0,0,3") -> (Pose) {
    alphaInputType=Bool
    translationMode=BMM_Additive
    alphaBoolBlend {blendOption=Linear}
    boneToModify {boneName=ik_hand_root}
    translation {z=3}
   }
   n2 = Leg IK!(ComponentPose:Component Space Pose Link Structure=@n1, Alpha:float=1) -> (Pose:Component Space Pose Link Structure) {
    is=title AnimGraphNode_LegIK
    alphaBoolBlend {blendOption=Linear}
    legsDefinition [{
      bEnableKneeTwistCorrection=false
      bEnableRotationLimit=false
      footBoneForwardAxis=Y
      hingeRotationAxis=Z
      minRotationAngle=15
      numBonesInLimb=2
      twistOffsetCurveName=None
      fKFootBone {boneName=hand_r}
      iKFootBone {boneName=ik_hand_r}
     }]
   }
   n3 = Leg IK(ComponentPose=@n2, Alpha=1) {
    alphaBoolBlend {blendOption=Linear}
    legsDefinition [{
      bEnableKneeTwistCorrection=false
      bEnableRotationLimit=false
      footBoneForwardAxis=Y
      hingeRotationAxis=Z
      minRotationAngle=15
      numBonesInLimb=2
      twistOffsetCurveName=None
      fKFootBone {boneName=hand_l}
      iKFootBone {boneName=ik_hand_l}
     }]
   }
   Output Pose(Result=Component To Local(ComponentPose=@n3)) {layerGroup=None}
  ]}
 VehicleFinalPoseOverride {nodes [
   n1 = "Transform (Modify) Bone - Bone: pelvis"!(ComponentPose=Local To Component(LocalPose=Input Pose() {name=InFinalPose}), bAlphaBoolEnabled:bool=True, Translation="0,0,-2") -> (Pose) {alphaInputType=Bool, translationMode=BMM_Additive, alphaBoolBlend {blendOption=Linear}, boneToModify {boneName=pelvis}, translation {z=-2}}
   n2 = "Transform (Modify) Bone - Bone: Root"!(ComponentPose=@n1, bAlphaBoolEnabled:bool=True, Translation="0,0,-3") -> (Pose) {alphaInputType=Bool, translationMode=BMM_Additive, alphaBoolBlend {blendOption=Linear}, boneToModify {boneName=Root}, translation {z=-3}}
   n3 = Leg IK(ComponentPose=@n2, Alpha=1) {
    alphaBoolBlend {blendOption=Linear}
    legsDefinition [
     {
      bEnableKneeTwistCorrection=false
      bEnableRotationLimit=false
      footBoneForwardAxis=Y
      hingeRotationAxis=Z
      minRotationAngle=15
      numBonesInLimb=2
      twistOffsetCurveName=None
      fKFootBone {boneName=foot_l}
      iKFootBone {boneName=ik_foot_l}
     }
     {
      bEnableKneeTwistCorrection=false
      bEnableRotationLimit=false
      footBoneForwardAxis=Y
      hingeRotationAxis=Z
      minRotationAngle=15
      numBonesInLimb=2
      twistOffsetCurveName=None
      fKFootBone {boneName=foot_r}
      iKFootBone {boneName=ik_foot_r}
     }
    ]
   }
   Output Pose(Result=Component To Local(ComponentPose=@n3)) {layerGroup=VehicleOverride}
  ]}
 VehicleFullBodyOverride {nodes [
   n1 = Layered blend per bone(BasePose=Use cached pose 'PassengerIdlePose'(), BlendPoses_0=Use cached pose 'UpperAndLowerPose'()) {bMeshSpaceRotationBlend=true, layerSetup [{branchFilters [{boneName=Spine_01}, {boneName=ik_hand_root}]}]}
   n2 = Rotate Root Bone!(BasePose:Pose Link Structure=@n1, Pitch:float=0, Yaw:float=0, MeshToComponent:Rotator="0,0,0") -> (Pose:Pose Link Structure) {is=title AnimGraphNode_RotateRootBone}
   AnimGraphNode_StateMachine_0 = CombatTransition() {is=title AnimGraphNode_StateMachine}
   n3 = Get bIsSeatedAction() -> (bIsSeatedAction:bool)
   Output Pose(Result=Inertialization!(Source:Pose Link Structure=Blend Poses by bool(BlendPose_0=@n2, BlendPose_1=@AnimGraphNode_StateMachine_0, BlendTime_1=0.2, bActiveValue=@n3) {blendType=Linear, childUpateMode=ResetChildOnActivate}) -> (Pose:Pose Link Structure) {is=title AnimGraphNode_Inertialization}) {layerGroup=VehicleOverride}
   Save cached pose 'UpperAndLowerPose'(Pose=Input Pose() {name=InPoseFullBody}) {cacheName=UpperAndLowerPose, cachePoseName=UpperAndLowerPose}
   n4 = Sequence Player 'MotoDirt_Driver_Idle_Stopped'() {sequence=/Motorcycle/Animation/Driver/MotoDirt_Driver_Idle_Stopped}
   n5 = "Sequence Player (additive) 'MotoDirt_Passenger_Idle_Pose'"() {sequence=/Motorcycle/Animation/Passenger/MotoDirt_Passenger_Idle_Pose}
   Save cached pose 'IdleCachedPose'(Pose=Apply Additive(
     Base=Apply Additive(Base=@n4, Additive=@n5) {alphaBoolBlend {blendOption=Linear}}
     Additive="Sequence Player (additive) 'SportBike_Passenger_Idle_AdditivePose'"() {bLoopAnimation=false, sequence=/Sportbike/Animation/Passenger/SportBike_Passenger_Idle_AdditivePose}
    ) {alphaBoolBlend {blendOption=Linear}}) {cacheName=IdleCachedPose}
   AnimGraphNode_StateMachine_1 = PassengerFullBody() {is=title AnimGraphNode_StateMachine}
   n6 = FullBodyAdditiveGroup!(InPoseFullBodyAdditive:Pose Link Structure=@AnimGraphNode_StateMachine_1) -> (Pose:Pose Link Structure) {is=title AnimGraphNode_LinkedAnimLayer, layer=FullBodyAdditiveGroup}
   Save cached pose 'PassengerIdlePose'(Pose=Blend Poses by bool(
     BlendPose_0=Sequence Player 'TrendConcentrate_RidingDBNO_Additive'() {sequence=/Game/Animation/Game/MainPlayer/Locomotion/Medium/Male/DBNO/TrendConcentrate_RidingDBNO_Additive}
     BlendPose_1=IK/BoneModification!(InPoseIK:Pose Link Structure=@n6) -> (Pose:Pose Link Structure) {is=title AnimGraphNode_LinkedAnimLayer, layer=IK/BoneModification}
     bActiveValue=Get isDBNO() -> (isDBNO:bool)
    )) {cacheName=PassengerIdlePose, cachePoseName=PassengerIdlePose}
  ]}
 CombatTransition {
  entry=PassengerIdle
  path=VehicleFullBodyOverride.AnimGraphNode_StateMachine_0.CombatTransition
  ids=id
  nodes [
   AnimStateNode_2 = PassengerIdle() {is=title AnimStateNode, bAlwaysResetOnEntry=true}
   AnimStateNode_3 = PassengerCombat() {is=title AnimStateNode}
   AnimStateNode_4 = PassengerCombatToIdle() {is=title AnimStateNode}
   AnimStateTransitionNode_2 = AnimStateTransitionNode() {crossfadeDuration=0}
   AnimStateTransitionNode_3 = AnimStateTransitionNode() {crossfadeDuration=0}
   AnimStateTransitionNode_4 = AnimStateTransitionNode() {bAutomaticRuleBasedOnSequencePlayerInState=true}
  ]
  links [
   {
    dst=PassengerCombatToIdle
    rule=VehicleFullBodyOverride.AnimGraphNode_StateMachine_0.CombatTransition.AnimStateTransitionNode_3.Transition
    src=PassengerCombat
    via=AnimStateTransitionNode_3
   }
   {
    dst=PassengerIdle
    rule=VehicleFullBodyOverride.AnimGraphNode_StateMachine_0.CombatTransition.AnimStateTransitionNode_4.Transition
    src=PassengerCombatToIdle
    via=AnimStateTransitionNode_4
   }
   {
    dst=PassengerCombat
    rule=VehicleFullBodyOverride.AnimGraphNode_StateMachine_0.CombatTransition.AnimStateTransitionNode_2.Transition
    src=PassengerIdle
    via=AnimStateTransitionNode_2
   }
  ]
 }
 PassengerIdle {
  path=VehicleFullBodyOverride.AnimGraphNode_StateMachine_0.CombatTransition.AnimStateNode_2.PassengerIdle
  nodes [
   n1 = "Transform (Modify) Bone - Bone: ik_hand_root"(ComponentPose=Local To Component(LocalPose=Use cached pose 'PassengerIdlePose'()), bAlphaBoolEnabled=True, Translation="0,0,2") {
    alphaInputType=Bool
    translationMode=BMM_Additive
    alphaBoolBlend {blendOption=Linear}
    boneToModify {boneName=ik_hand_root}
    translation {z=2}
   }
   Output Animation Pose(Result=Rotate Root Bone(BasePose=Component To Local(ComponentPose=@n1), Pitch=0, Yaw=0, MeshToComponent="0,0,0"))
  ]
 }
 PassengerCombat {
  path=VehicleFullBodyOverride.AnimGraphNode_StateMachine_0.CombatTransition.AnimStateNode_3.PassengerCombat
  nodes [
   AnimGraphNode_StateMachine_0 = Combat() {is=title AnimGraphNode_StateMachine}
   n1 = Component To Local(ComponentPose=Reset Root Transform!(ComponentPose:Component Space Pose Link Structure=Local To Component(LocalPose=@AnimGraphNode_StateMachine_0), Alpha:float=1) -> (Pose:Component Space Pose Link Structure) {is=title AnimGraphNode_ResetRoot, alphaBoolBlend {blendOption=Linear}})
   n2 = Break Fort Anim Input Motorcycle!(FortAnimInput_Motorcycle:"Fort Anim Input Motorcycle Structure (by ref)"=Get MotorcycleInput() -> (MotorcycleInput:Fort Anim Input Motorcycle Structure)) -> (
    bIsUsingMotorcycle:bool=~
    bIsDriver:bool=~
    bIsFrontPassenger:bool=~
    bIsBackPassenger:bool=~
    bHasPassenger:bool=~
    bIsBackPassengerAndLeaning:bool=~
    bIsDrifting:bool=~
    bIsBoosting:bool=~
    bIsReversing:bool=~
    bIsBraking:bool=~
    bIsMoving:bool=~
    bIsMovingForward:bool=~
    bIsSteeringRight:bool=~
    bIsSteeringLeft:bool=~
    RunForwardAlpha:float=~
    VerticalVelocity:float=~
    VerticalAcceleration:float=~
    bAimFWD:bool=~
    bAimBWD:bool=~
    bAimLFT:bool=~
    bAimRGT:bool=~
    PawnToVehicleDeltaYawAngleDegrees:float=~
    AimCardDirDeadZoneAngleDegrees:float=~
    AimCardDirAngleOffsetDegrees:float=~
    AimFWDDeltaAngleDegrees:float=~
    AimBWDDeltaAngleDegrees:float=~
    AimLFTDeltaAngleDegrees:float=~
    AimRGTDeltaAngleDegrees:float=~
    SlopePitchDegreeAngle:float=~
    SlopeRollDegreeAngle:float=~
    SteerAngle:float=~
    SteerAlpha:float=~
    SteerAngleDeadZoneDegrees:float=~
    SteeringRotation:float=~
    VehiclePitch:float=~
    VehicleRoll:float=~
   ) {is=title K2Node_BreakStruct}
   n3 = "Transform (Modify) Bone - Bone: pelvis"!(
    ComponentPose=Local To Component(LocalPose=AimOffset Player 'MotoDirt_Passenger_SlopePitchRoll_AO'!(X:float=@n2.SlopeRollDegreeAngle, Y:float=@n2.SlopePitchDegreeAngle, BasePose:Pose Link Structure=@n1, Alpha:float=1) -> (Pose:Pose Link Structure) {blendSpace=/Motorcycle/Animation/Passenger/AO/MotoDirt_Passenger_SlopePitchRoll_AO, alphaBoolBlend {blendOption=Linear}})
    Alpha:float=0.25
    Rotation:Rotator="0,0,0"
   ) -> (Pose) {
    rotationMode=BMM_Additive
    rotationSpace=BCS_WorldSpace
    translationSpace=BCS_WorldSpace
    alphaBoolBlend {blendOption=Linear}
    boneToModify {boneName=pelvis}
   }
   Output Animation Pose(Result=Component To Local(ComponentPose=@n3))
  ]
 }
 Combat {
  entry=CombatStart
  path=VehicleFullBodyOverride.AnimGraphNode_StateMachine_0.CombatTransition.AnimStateNode_3.PassengerCombat.AnimGraphNode_StateMachine_0.Combat
  ids=id
  nodes [
   AnimStateNode_0 = CombatStart() {is=title AnimStateNode}
   AnimStateNode_1 = CombatLoop() {is=title AnimStateNode}
   AnimStateTransitionNode_1 = AnimStateTransitionNode() {
    bAutomaticRuleBasedOnSequencePlayerInState=true
    blendProfileWrapper {blendProfile=&Fortnite_M_Avg_Player_Skeleton:FastWeapons}
   }
  ]
  links [{
    dst=CombatLoop
    rule=VehicleFullBodyOverride.AnimGraphNode_StateMachine_0.CombatTransition.AnimStateNode_3.PassengerCombat.AnimGraphNode_StateMachine_0.Combat.AnimStateTransitionNode_1.AnimationTransitionGraph_0
    src=CombatStart
    via=AnimStateTransitionNode_1
   }]
 }
 CombatStart {
  path=VehicleFullBodyOverride.AnimGraphNode_StateMachine_0.CombatTransition.AnimStateNode_3.PassengerCombat.AnimGraphNode_StateMachine_0.Combat.AnimStateNode_0.CombatStart
  nodes [Output Animation Pose(Result=Apply Additive(
     Base=Use cached pose 'UpperAndLowerPose'()
     Additive="Sequence Player (additive) 'MotoDirt_Passenger_EquipPlaceHolder'"() {bLoopAnimation=false, sequence=/Motorcycle/Animation/Passenger/MotoDirt_Passenger_EquipPlaceHolder}
    ) {alphaBoolBlend {blendOption=Linear}})]
 }
 CombatLoop {
  path=VehicleFullBodyOverride.AnimGraphNode_StateMachine_0.CombatTransition.AnimStateNode_3.PassengerCombat.AnimGraphNode_StateMachine_0.Combat.AnimStateNode_1.CombatLoop
  nodes [
   AnimGraphNode_StateMachine_2 = CombatUpperBodyAdditive() {is=title AnimGraphNode_StateMachine}
   Output Animation Pose(Result=Apply Additive(Base=Use cached pose 'UpperAndLowerPose'(), Additive=@AnimGraphNode_StateMachine_2, Alpha=0.6) {alphaBoolBlend {blendOption=Linear}})
  ]
 }
 CombatUpperBodyAdditive {
  entry=IdleAdd
  path=VehicleFullBodyOverride.AnimGraphNode_StateMachine_0.CombatTransition.AnimStateNode_3.PassengerCombat.AnimGraphNode_StateMachine_0.Combat.AnimStateNode_1.CombatLoop.AnimGraphNode_StateMachine_2.CombatUpperBodyAdditive
  ids=id
  nodes [
   AnimStateNode_0 = IdleAdd() {is=title AnimStateNode}
   AnimStateNode_1 = RidingAdd() {is=title AnimStateNode}
   AnimStateTransitionNode_0 = AnimStateTransitionNode()
   AnimStateTransitionNode_1 = AnimStateTransitionNode()
  ]
  links [
   {
    dst=RidingAdd
    rule=VehicleFullBodyOverride.AnimGraphNode_StateMachine_0.CombatTransition.AnimStateNode_3.PassengerCombat.AnimGraphNode_StateMachine_0.Combat.AnimStateNode_1.CombatLoop.AnimGraphNode_StateMachine_2.CombatUpperBodyAdditive.AnimStateTransitionNode_0.Transition
    src=IdleAdd
    via=AnimStateTransitionNode_0
   }
   {
    dst=IdleAdd
    rule=VehicleFullBodyOverride.AnimGraphNode_StateMachine_0.CombatTransition.AnimStateNode_3.PassengerCombat.AnimGraphNode_StateMachine_0.Combat.AnimStateNode_1.CombatLoop.AnimGraphNode_StateMachine_2.CombatUpperBodyAdditive.AnimStateTransitionNode_1.Transition
    src=RidingAdd
    via=AnimStateTransitionNode_1
   }
  ]
 }
 IdleAdd {
  path=VehicleFullBodyOverride.AnimGraphNode_StateMachine_0.CombatTransition.AnimStateNode_3.PassengerCombat.AnimGraphNode_StateMachine_0.Combat.AnimStateNode_1.CombatLoop.AnimGraphNode_StateMachine_2.CombatUpperBodyAdditive.AnimStateNode_0.IdleAdd
  nodes [Output Animation Pose(Result="Sequence Player (additive) 'MotoDirt_Driver_Combat_Idle'"() {sequence=/Motorcycle/Animation/Driver/MotoDirt_Driver_Combat_Idle})]
 }
 RidingAdd {
  path=VehicleFullBodyOverride.AnimGraphNode_StateMachine_0.CombatTransition.AnimStateNode_3.PassengerCombat.AnimGraphNode_StateMachine_0.Combat.AnimStateNode_1.CombatLoop.AnimGraphNode_StateMachine_2.CombatUpperBodyAdditive.AnimStateNode_1.RidingAdd
  nodes [Output Animation Pose(Result=Blendspace Player 'MotoDirt_Driver_CombatMoving_BS'!(X=0, PlayRate:float=1) -> (Pose) {blendSpace=/Motorcycle/Animation/Driver/BlendSpace/MotoDirt_Driver_CombatMoving_BS})]
 }
 Transition {
  path=VehicleFullBodyOverride.AnimGraphNode_StateMachine_0.CombatTransition.AnimStateNode_3.PassengerCombat.AnimGraphNode_StateMachine_0.Combat.AnimStateNode_1.CombatLoop.AnimGraphNode_StateMachine_2.CombatUpperBodyAdditive.AnimStateTransitionNode_0.Transition
  nodes [Result(bCanEnterTransition=Get bIsMoving() -> (bIsMoving:bool))]
 }
 Transition {
  path=VehicleFullBodyOverride.AnimGraphNode_StateMachine_0.CombatTransition.AnimStateNode_3.PassengerCombat.AnimGraphNode_StateMachine_0.Combat.AnimStateNode_1.CombatLoop.AnimGraphNode_StateMachine_2.CombatUpperBodyAdditive.AnimStateTransitionNode_1.Transition
  nodes [Result(bCanEnterTransition=NOT Boolean(A=Get bIsMoving()))]
 }
 AnimationTransitionGraph_0 {
  path=VehicleFullBodyOverride.AnimGraphNode_StateMachine_0.CombatTransition.AnimStateNode_3.PassengerCombat.AnimGraphNode_StateMachine_0.Combat.AnimStateTransitionNode_1.AnimationTransitionGraph_0
  nodes [Result()]
 }
 PassengerCombatToIdle {
  path=VehicleFullBodyOverride.AnimGraphNode_StateMachine_0.CombatTransition.AnimStateNode_4.PassengerCombatToIdle
  nodes [
   n1 = Sequence Player 'MotoDirt_Passenger_CombatToRide'() {bLoopAnimation=false, sequence=/Motorcycle/Animation/Passenger/MotoDirt_Passenger_CombatToRide}
   n2 = "Sequence Player (additive) 'SportBike_Passenger_Idle_AdditivePose'"() {bLoopAnimation=false, sequence=&SportBike_Passenger_Idle_AdditivePose}
   Output Animation Pose(Result=Rotate Root Bone(BasePose=Apply Additive(Base=@n1, Additive=@n2) {alphaBoolBlend {blendOption=Linear}}, Pitch=0, Yaw=0, MeshToComponent="0,0,0"))
  ]
 }
 Transition {
  path=VehicleFullBodyOverride.AnimGraphNode_StateMachine_0.CombatTransition.AnimStateTransitionNode_2.Transition
  nodes [Result(bCanEnterTransition=Get bShouldPawnStandUp() -> (bShouldPawnStandUp:bool))]
 }
 Transition {
  path=VehicleFullBodyOverride.AnimGraphNode_StateMachine_0.CombatTransition.AnimStateTransitionNode_3.Transition
  nodes [Result(bCanEnterTransition=NOT Boolean(A=Get bShouldPawnStandUp()))]
 }
 Transition {path=VehicleFullBodyOverride.AnimGraphNode_StateMachine_0.CombatTransition.AnimStateTransitionNode_4.Transition, nodes [Result()]}
 PassengerFullBody {
  entry=Idle
  path=VehicleFullBodyOverride.AnimGraphNode_StateMachine_1.PassengerFullBody
  ids=id
  nodes [
   ToJump() {is=title AnimStateAliasNode, also {aliases [Idle, JumpLand, Locomotion, LocomotionStop]}}
   ToIdle() {is=title AnimStateAliasNode, also {aliases [JumpLand, LocomotionStop]}}
   ToLocoSelector() {is=title AnimStateAliasNode, also {aliases [Idle, LocomotionStop]}}
   ToBoost() {is=title AnimStateAliasNode, also {aliases [BootOutro, Idle, JumpLand, Locomotion, LocomotionStop, Stoppie, StoppieLand, Wheelie]}}
   ToLocomotion() {is=title AnimStateAliasNode, also {aliases [Locomotion]}}
   AnimStateConduitNode_0 = LocoSelector() {is=title AnimStateConduitNode}
   AnimStateNode_0 = Idle()
   AnimStateNode_1 = Locomotion() {is=title AnimStateNode}
   AnimStateNode_2 = LocomotionStop() {is=title AnimStateNode}
   AnimStateNode_3 = NormalJump() {is=title AnimStateNode, bAlwaysResetOnEntry=true}
   AnimStateNode_4 = JumpLand() {is=title AnimStateNode}
   AnimStateNode_5 = Boost() {is=title AnimStateNode}
   AnimStateNode_6 = Wheelie() {is=title AnimStateNode}
   AnimStateNode_7 = Stoppie() {is=title AnimStateNode}
   AnimStateNode_8 = StoppieLand() {is=title AnimStateNode}
   AnimStateNode_9 = BootOutro() {is=title AnimStateNode}
   AnimStateTransitionNode_0 = AnimStateTransitionNode()
   AnimStateTransitionNode_1 = AnimStateTransitionNode() {crossfadeDuration=0.4}
   AnimStateTransitionNode_10 = AnimStateTransitionNode() {logicType=TLT_Inertialization, priorityOrder=2}
   AnimStateTransitionNode_11 = AnimStateTransitionNode()
   AnimStateTransitionNode_12 = AnimStateTransitionNode() {logicType=TLT_Inertialization}
   AnimStateTransitionNode_13 = AnimStateTransitionNode() {bAutomaticRuleBasedOnSequencePlayerInState=true, logicType=TLT_Inertialization}
   AnimStateTransitionNode_14 = AnimStateTransitionNode() {bAutomaticRuleBasedOnSequencePlayerInState=true, crossfadeDuration=0.4, logicType=TLT_Inertialization}
   AnimStateTransitionNode_15 = AnimStateTransitionNode()
   AnimStateTransitionNode_16 = AnimStateTransitionNode() {logicType=TLT_Inertialization}
   AnimStateTransitionNode_17 = AnimStateTransitionNode() {bAutomaticRuleBasedOnSequencePlayerInState=true, crossfadeDuration=0.45, logicType=TLT_Inertialization, priorityOrder=3}
   AnimStateTransitionNode_18 = AnimStateTransitionNode()
   AnimStateTransitionNode_19 = AnimStateTransitionNode() {priorityOrder=2}
   AnimStateTransitionNode_2 = AnimStateTransitionNode() {logicType=TLT_Inertialization}
   AnimStateTransitionNode_3 = AnimStateTransitionNode() {crossfadeDuration=0.067, logicType=TLT_Inertialization}
   AnimStateTransitionNode_4 = AnimStateTransitionNode() {crossfadeDuration=0.4, priorityOrder=0}
   AnimStateTransitionNode_5 = AnimStateTransitionNode() {crossfadeDuration=0.45, priorityOrder=100}
   AnimStateTransitionNode_6 = AnimStateTransitionNode() {logicType=TLT_Inertialization}
   AnimStateTransitionNode_7 = AnimStateTransitionNode() {bAutomaticRuleBasedOnSequencePlayerInState=true, logicType=TLT_Inertialization}
   AnimStateTransitionNode_8 = AnimStateTransitionNode() {logicType=TLT_Inertialization}
   AnimStateTransitionNode_9 = AnimStateTransitionNode()
  ]
  links [
   {
    dst=BootOutro
    rule=VehicleFullBodyOverride.AnimGraphNode_StateMachine_1.PassengerFullBody.AnimStateTransitionNode_16.Transition
    src=Boost
    via=AnimStateTransitionNode_16
   }
   {
    dst=ToLocomotion
    rule=VehicleFullBodyOverride.AnimGraphNode_StateMachine_1.PassengerFullBody.AnimStateTransitionNode_17.Transition
    src=BootOutro
    via=AnimStateTransitionNode_17
   }
   {
    dst=ToLocomotion
    rule=VehicleFullBodyOverride.AnimGraphNode_StateMachine_1.PassengerFullBody.AnimStateTransitionNode_18.Transition
    src=BootOutro
    via=AnimStateTransitionNode_18
   }
   {
    dst=ToLocomotion
    rule=VehicleFullBodyOverride.AnimGraphNode_StateMachine_1.PassengerFullBody.AnimStateTransitionNode_19.Transition
    src=BootOutro
    via=AnimStateTransitionNode_19
   }
   {
    dst=Locomotion
    rule=VehicleFullBodyOverride.AnimGraphNode_StateMachine_1.PassengerFullBody.AnimStateTransitionNode_7.Transition
    src=JumpLand
    via=AnimStateTransitionNode_7
   }
   {
    dst=Locomotion
    rule=VehicleFullBodyOverride.AnimGraphNode_StateMachine_1.PassengerFullBody.AnimStateTransitionNode_5.Transition
    src=LocoSelector
    via=AnimStateTransitionNode_5
   }
   {
    dst=Stoppie
    rule=VehicleFullBodyOverride.AnimGraphNode_StateMachine_1.PassengerFullBody.AnimStateTransitionNode_10.Transition
    src=LocoSelector
    via=AnimStateTransitionNode_10
   }
   {
    dst=Wheelie
    rule=VehicleFullBodyOverride.AnimGraphNode_StateMachine_1.PassengerFullBody.AnimStateTransitionNode_6.Transition
    src=LocoSelector
    via=AnimStateTransitionNode_6
   }
   {
    dst=LocoSelector
    rule=VehicleFullBodyOverride.AnimGraphNode_StateMachine_1.PassengerFullBody.AnimStateTransitionNode_0.Transition
    src=Locomotion
    via=AnimStateTransitionNode_0
   }
   {
    dst=LocomotionStop
    rule=VehicleFullBodyOverride.AnimGraphNode_StateMachine_1.PassengerFullBody.AnimStateTransitionNode_1.Transition
    src=Locomotion
    via=AnimStateTransitionNode_1
   }
   {
    dst=JumpLand
    rule=VehicleFullBodyOverride.AnimGraphNode_StateMachine_1.PassengerFullBody.AnimStateTransitionNode_3.Transition
    src=NormalJump
    via=AnimStateTransitionNode_3
   }
   {
    dst=LocoSelector
    rule=VehicleFullBodyOverride.AnimGraphNode_StateMachine_1.PassengerFullBody.AnimStateTransitionNode_11.Transition
    src=Stoppie
    via=AnimStateTransitionNode_11
   }
   {
    dst=StoppieLand
    rule=VehicleFullBodyOverride.AnimGraphNode_StateMachine_1.PassengerFullBody.AnimStateTransitionNode_12.Transition
    src=Stoppie
    via=AnimStateTransitionNode_12
   }
   {
    dst=Locomotion
    rule=VehicleFullBodyOverride.AnimGraphNode_StateMachine_1.PassengerFullBody.AnimStateTransitionNode_13.Transition
    src=StoppieLand
    via=AnimStateTransitionNode_13
   }
   {
    dst=Locomotion
    rule=VehicleFullBodyOverride.AnimGraphNode_StateMachine_1.PassengerFullBody.AnimStateTransitionNode_4.Transition
    src=StoppieLand
    via=AnimStateTransitionNode_4
   }
   {
    dst=Boost
    rule=VehicleFullBodyOverride.AnimGraphNode_StateMachine_1.PassengerFullBody.AnimStateTransitionNode_8.Transition
    src=ToBoost
    via=AnimStateTransitionNode_8
   }
   {
    dst=Idle
    rule=VehicleFullBodyOverride.AnimGraphNode_StateMachine_1.PassengerFullBody.AnimStateTransitionNode_14.Transition
    src=ToIdle
    via=AnimStateTransitionNode_14
   }
   {
    dst=NormalJump
    rule=VehicleFullBodyOverride.AnimGraphNode_StateMachine_1.PassengerFullBody.AnimStateTransitionNode_2.Transition
    src=ToJump
    via=AnimStateTransitionNode_2
   }
   {
    dst=LocoSelector
    rule=VehicleFullBodyOverride.AnimGraphNode_StateMachine_1.PassengerFullBody.AnimStateTransitionNode_15.Transition
    src=ToLocoSelector
    via=AnimStateTransitionNode_15
   }
   {
    dst=LocoSelector
    rule=VehicleFullBodyOverride.AnimGraphNode_StateMachine_1.PassengerFullBody.AnimStateTransitionNode_9.Transition
    src=Wheelie
    via=AnimStateTransitionNode_9
   }
  ]
 }
 LocoSelector {
  path=VehicleFullBodyOverride.AnimGraphNode_StateMachine_1.PassengerFullBody.AnimStateConduitNode_0.LocoSelector
  nodes [Result(bCanEnterTransition=true)]
 }
 Idle {
  path=VehicleFullBodyOverride.AnimGraphNode_StateMachine_1.PassengerFullBody.AnimStateNode_0.Idle
  nodes [
   n1 = AimOffset Player 'Sportbike_Passenger_HeadAO'!(X:float=0, Y:float=0, BasePose:Pose Link Structure=Use cached pose 'IdleCachedPose'(), bAlphaBoolEnabled:bool=true) -> (Pose:Pose Link Structure) {
    alphaInputType=Bool
    blendSpace=/Sportbike/Animation/Passenger/AO/Sportbike_Passenger_HeadAO
    alphaBoolBlend {blendInTime=0.35, blendOutTime=0.2}
   }
   Output Animation Pose(Result=@n1)
  ]
 }
 Locomotion {
  path=VehicleFullBodyOverride.AnimGraphNode_StateMachine_1.PassengerFullBody.AnimStateNode_1.Locomotion
  nodes [
   AnimGraphNode_StateMachine_0 = Movement() {is=title AnimGraphNode_StateMachine, bAllowConduitEntryStates=true}
   Output Animation Pose(Result=Apply Additive(
     Base=@AnimGraphNode_StateMachine_0
     Additive=Blendspace Player 'MotoDirt_Driver_Moving_BS'!(X=0, PlayRate:float=1.2) -> (Pose) {blendSpace=/Motorcycle/Animation/Driver/BlendSpace/MotoDirt_Driver_Moving_BS}
     Alpha=0.5
    ) {alphaBoolBlend {blendOption=Linear}})
  ]
 }
 Movement {
  entry=Conduit
  path=VehicleFullBodyOverride.AnimGraphNode_StateMachine_1.PassengerFullBody.AnimStateNode_1.Locomotion.AnimGraphNode_StateMachine_0.Movement
  ids=id
  nodes [
   AnimStateConduitNode_0 = Conduit()
   AnimStateNode_0 = MovingStart() {is=title AnimStateNode}
   AnimStateNode_1 = MovingLoop() {is=title AnimStateNode}
   AnimStateTransitionNode_10 = AnimStateTransitionNode() {bAutomaticRuleBasedOnSequencePlayerInState=true, logicType=TLT_Inertialization}
   AnimStateTransitionNode_11 = AnimStateTransitionNode() {logicType=TLT_Inertialization}
   AnimStateTransitionNode_9 = AnimStateTransitionNode() {logicType=TLT_Inertialization, priorityOrder=100}
  ]
  links [
   {
    dst=MovingLoop
    rule=VehicleFullBodyOverride.AnimGraphNode_StateMachine_1.PassengerFullBody.AnimStateNode_1.Locomotion.AnimGraphNode_StateMachine_0.Movement.AnimStateTransitionNode_11.Transition
    src=Conduit
    via=AnimStateTransitionNode_11
   }
   {
    dst=MovingStart
    rule=VehicleFullBodyOverride.AnimGraphNode_StateMachine_1.PassengerFullBody.AnimStateNode_1.Locomotion.AnimGraphNode_StateMachine_0.Movement.AnimStateTransitionNode_9.Transition
    src=Conduit
    via=AnimStateTransitionNode_9
   }
   {
    dst=MovingLoop
    rule=VehicleFullBodyOverride.AnimGraphNode_StateMachine_1.PassengerFullBody.AnimStateNode_1.Locomotion.AnimGraphNode_StateMachine_0.Movement.AnimStateTransitionNode_10.Transition
    src=MovingStart
    via=AnimStateTransitionNode_10
   }
  ]
 }
 Conduit {
  path=VehicleFullBodyOverride.AnimGraphNode_StateMachine_1.PassengerFullBody.AnimStateNode_1.Locomotion.AnimGraphNode_StateMachine_0.Movement.AnimStateConduitNode_0.Conduit
  nodes [Result(bCanEnterTransition=true)]
 }
 MovingStart {
  path=VehicleFullBodyOverride.AnimGraphNode_StateMachine_1.PassengerFullBody.AnimStateNode_1.Locomotion.AnimGraphNode_StateMachine_0.Movement.AnimStateNode_0.MovingStart
  nodes [Output Animation Pose(Result=Sequence Player 'MotoDirt_Passenger_MovingStart'() {sequence=/Motorcycle/Animation/Passenger/MotoDirt_Passenger_MovingStart})]
 }
 MovingLoop {
  path=VehicleFullBodyOverride.AnimGraphNode_StateMachine_1.PassengerFullBody.AnimStateNode_1.Locomotion.AnimGraphNode_StateMachine_0.Movement.AnimStateNode_1.MovingLoop
  nodes [Output Animation Pose(Result=Blend Poses by bool(
     BlendPose_0=Use cached pose 'IdleCachedPose'()
     BlendPose_1=Blendspace Player 'Sportbike_Passenger_Loco_Leans_BS'(X=0, Y=0) {blendSpace=/Sportbike/Animation/Passenger/Sportbike_Passenger_Loco_Leans_BS}
     BlendTime_0=0.35
     BlendTime_1=0.55
    ) {blendType=Linear})]
 }
 Transition {
  path=VehicleFullBodyOverride.AnimGraphNode_StateMachine_1.PassengerFullBody.AnimStateNode_1.Locomotion.AnimGraphNode_StateMachine_0.Movement.AnimStateTransitionNode_10.Transition
  nodes [Result()]
 }
 Transition {
  path=VehicleFullBodyOverride.AnimGraphNode_StateMachine_1.PassengerFullBody.AnimStateNode_1.Locomotion.AnimGraphNode_StateMachine_0.Movement.AnimStateTransitionNode_11.Transition
  nodes [Result(bCanEnterTransition=Get bStateTransitionRuleToMovingLoop() -> (bStateTransitionRuleToMovingLoop:bool))]
 }
 Transition {
  path=VehicleFullBodyOverride.AnimGraphNode_StateMachine_1.PassengerFullBody.AnimStateNode_1.Locomotion.AnimGraphNode_StateMachine_0.Movement.AnimStateTransitionNode_9.Transition
  nodes [Result(bCanEnterTransition=true)]
 }
 LocomotionStop {
  path=VehicleFullBodyOverride.AnimGraphNode_StateMachine_1.PassengerFullBody.AnimStateNode_2.LocomotionStop
  nodes [Output Animation Pose(Result=Sequence Player 'MotoDirt_Passenger_MovingStop'() {bLoopAnimation=false, sequence=/Motorcycle/Animation/Passenger/MotoDirt_Passenger_MovingStop})]
 }
 NormalJump {
  path=VehicleFullBodyOverride.AnimGraphNode_StateMachine_1.PassengerFullBody.AnimStateNode_3.NormalJump
  nodes [
   AnimGraphNode_StateMachine_0 = NormalJumpTransition() {is=title AnimGraphNode_StateMachine}
   Output Animation Pose(Result=@AnimGraphNode_StateMachine_0)
  ]
 }
 NormalJumpTransition {
  entry=NormalJumpStart
  path=VehicleFullBodyOverride.AnimGraphNode_StateMachine_1.PassengerFullBody.AnimStateNode_3.NormalJump.AnimGraphNode_StateMachine_0.NormalJumpTransition
  ids=id
  nodes [
   AnimStateNode_0 = NormalJumpStart() {is=title AnimStateNode}
   AnimStateNode_1 = NormalJumpLoop() {is=title AnimStateNode}
   AnimStateTransitionNode_4 = AnimStateTransitionNode() {bAutomaticRuleBasedOnSequencePlayerInState=true, crossfadeDuration=0.1, logicType=TLT_Inertialization}
  ]
  links [{
    dst=NormalJumpLoop
    rule=VehicleFullBodyOverride.AnimGraphNode_StateMachine_1.PassengerFullBody.AnimStateNode_3.NormalJump.AnimGraphNode_StateMachine_0.NormalJumpTransition.AnimStateTransitionNode_4.Transition
    src=NormalJumpStart
    via=AnimStateTransitionNode_4
   }]
 }
 NormalJumpStart {
  path=VehicleFullBodyOverride.AnimGraphNode_StateMachine_1.PassengerFullBody.AnimStateNode_3.NormalJump.AnimGraphNode_StateMachine_0.NormalJumpTransition.AnimStateNode_0.NormalJumpStart
  nodes [Output Animation Pose(Result=Sequence Player 'MotoDirt_Passenger_Jump_Start'() {bLoopAnimation=false, sequence=/Motorcycle/Animation/Passenger/MotoDirt_Passenger_Jump_Start})]
 }
 NormalJumpLoop {
  path=VehicleFullBodyOverride.AnimGraphNode_StateMachine_1.PassengerFullBody.AnimStateNode_3.NormalJump.AnimGraphNode_StateMachine_0.NormalJumpTransition.AnimStateNode_1.NormalJumpLoop
  nodes [Output Animation Pose(Result=Sequence Player 'MotoDirt_Passenger_Jump_Loop'() {sequence=/Motorcycle/Animation/Passenger/MotoDirt_Passenger_Jump_Loop})]
 }
 Transition {
  path=VehicleFullBodyOverride.AnimGraphNode_StateMachine_1.PassengerFullBody.AnimStateNode_3.NormalJump.AnimGraphNode_StateMachine_0.NormalJumpTransition.AnimStateTransitionNode_4.Transition
  nodes [Result()]
 }
 JumpLand {
  path=VehicleFullBodyOverride.AnimGraphNode_StateMachine_1.PassengerFullBody.AnimStateNode_4.JumpLand
  nodes [Output Animation Pose(Result=Sequence Player 'MotoDirt_Passenger_Jump_Land'!(StartPosition:float=Get CachedJumpIntroTime() -> (CachedJumpIntroTime:float)) -> (Pose) {bLoopAnimation=false, sequence=/Motorcycle/Animation/Passenger/MotoDirt_Passenger_Jump_Land})]
 }
 Boost {
  path=VehicleFullBodyOverride.AnimGraphNode_StateMachine_1.PassengerFullBody.AnimStateNode_5.Boost
  nodes [
   AnimGraphNode_StateMachine_0 = BoostTransition() {is=title AnimGraphNode_StateMachine}
   Output Animation Pose(Result=@AnimGraphNode_StateMachine_0)
  ]
 }
 BoostTransition {
  entry=Boost_Intro
  path=VehicleFullBodyOverride.AnimGraphNode_StateMachine_1.PassengerFullBody.AnimStateNode_5.Boost.AnimGraphNode_StateMachine_0.BoostTransition
  ids=id
  nodes [
   AnimStateNode_0 = Boost_Intro() {is=title AnimStateNode}
   AnimStateNode_1 = Boost_Loop() {is=title AnimStateNode}
   AnimStateTransitionNode_3 = AnimStateTransitionNode() {bAutomaticRuleBasedOnSequencePlayerInState=true, logicType=TLT_Inertialization}
  ]
  links [{
    dst=Boost_Loop
    rule=VehicleFullBodyOverride.AnimGraphNode_StateMachine_1.PassengerFullBody.AnimStateNode_5.Boost.AnimGraphNode_StateMachine_0.BoostTransition.AnimStateTransitionNode_3.Transition
    src=Boost_Intro
    via=AnimStateTransitionNode_3
   }]
 }
 Boost_Intro {
  path=VehicleFullBodyOverride.AnimGraphNode_StateMachine_1.PassengerFullBody.AnimStateNode_5.Boost.AnimGraphNode_StateMachine_0.BoostTransition.AnimStateNode_0.Boost_Intro
  nodes [Output Animation Pose(Result=Sequence Player 'SportBike_Passenger_Boost_Intro'() {bLoopAnimation=false, sequence=/Sportbike/Animation/Passenger/SportBike_Passenger_Boost_Intro})]
 }
 Boost_Loop {
  path=VehicleFullBodyOverride.AnimGraphNode_StateMachine_1.PassengerFullBody.AnimStateNode_5.Boost.AnimGraphNode_StateMachine_0.BoostTransition.AnimStateNode_1.Boost_Loop
  nodes [Output Animation Pose(Result=Sequence Player 'SportBike_Passenger_Boost_Loop'() {sequence=/Sportbike/Animation/Passenger/SportBike_Passenger_Boost_Loop})]
 }
 Transition {
  path=VehicleFullBodyOverride.AnimGraphNode_StateMachine_1.PassengerFullBody.AnimStateNode_5.Boost.AnimGraphNode_StateMachine_0.BoostTransition.AnimStateTransitionNode_3.Transition
  nodes [Result()]
 }
 Wheelie {
  path=VehicleFullBodyOverride.AnimGraphNode_StateMachine_1.PassengerFullBody.AnimStateNode_6.Wheelie
  nodes [
   AnimGraphNode_StateMachine_0 = WheelieTransition() {is=title AnimGraphNode_StateMachine}
   Output Animation Pose(Result=@AnimGraphNode_StateMachine_0)
  ]
 }
 WheelieTransition {
  entry=WheelieStart
  path=VehicleFullBodyOverride.AnimGraphNode_StateMachine_1.PassengerFullBody.AnimStateNode_6.Wheelie.AnimGraphNode_StateMachine_0.WheelieTransition
  ids=id
  nodes [
   AnimStateNode_0 = WheelieStart() {is=title AnimStateNode}
   AnimStateNode_1 = WheelieLoop() {is=title AnimStateNode}
   AnimStateTransitionNode_2 = AnimStateTransitionNode() {bAutomaticRuleBasedOnSequencePlayerInState=true, logicType=TLT_Inertialization}
  ]
  links [{
    dst=WheelieLoop
    rule=VehicleFullBodyOverride.AnimGraphNode_StateMachine_1.PassengerFullBody.AnimStateNode_6.Wheelie.AnimGraphNode_StateMachine_0.WheelieTransition.AnimStateTransitionNode_2.Transition
    src=WheelieStart
    via=AnimStateTransitionNode_2
   }]
 }
 WheelieStart {
  path=VehicleFullBodyOverride.AnimGraphNode_StateMachine_1.PassengerFullBody.AnimStateNode_6.Wheelie.AnimGraphNode_StateMachine_0.WheelieTransition.AnimStateNode_0.WheelieStart
  nodes [Output Animation Pose(Result=Sequence Player 'MotoDirt_Passenger_WheelieStart'() {bLoopAnimation=false, sequence=/Motorcycle/Animation/Passenger/MotoDirt_Passenger_WheelieStart})]
 }
 WheelieLoop {
  path=VehicleFullBodyOverride.AnimGraphNode_StateMachine_1.PassengerFullBody.AnimStateNode_6.Wheelie.AnimGraphNode_StateMachine_0.WheelieTransition.AnimStateNode_1.WheelieLoop
  nodes [Output Animation Pose(Result=Sequence Player 'MotoDirt_Passenger_WheelieLoop'() {bLoopAnimation=false, sequence=/Motorcycle/Animation/Passenger/MotoDirt_Passenger_WheelieLoop})]
 }
 Transition {
  path=VehicleFullBodyOverride.AnimGraphNode_StateMachine_1.PassengerFullBody.AnimStateNode_6.Wheelie.AnimGraphNode_StateMachine_0.WheelieTransition.AnimStateTransitionNode_2.Transition
  nodes [Result()]
 }
 Stoppie {
  path=VehicleFullBodyOverride.AnimGraphNode_StateMachine_1.PassengerFullBody.AnimStateNode_7.Stoppie
  nodes [
   AnimGraphNode_StateMachine_0 = StoppieTransition() {is=title AnimGraphNode_StateMachine}
   Output Animation Pose(Result=@AnimGraphNode_StateMachine_0)
  ]
 }
 StoppieTransition {
  entry=StoppieStart
  path=VehicleFullBodyOverride.AnimGraphNode_StateMachine_1.PassengerFullBody.AnimStateNode_7.Stoppie.AnimGraphNode_StateMachine_0.StoppieTransition
  ids=id
  nodes [
   AnimStateNode_0 = StoppieStart() {is=title AnimStateNode}
   AnimStateNode_1 = StoppieLoop() {is=title AnimStateNode}
   AnimStateTransitionNode_7 = AnimStateTransitionNode() {bAutomaticRuleBasedOnSequencePlayerInState=true, logicType=TLT_Inertialization}
  ]
  links [{
    dst=StoppieLoop
    rule=VehicleFullBodyOverride.AnimGraphNode_StateMachine_1.PassengerFullBody.AnimStateNode_7.Stoppie.AnimGraphNode_StateMachine_0.StoppieTransition.AnimStateTransitionNode_7.Transition
    src=StoppieStart
    via=AnimStateTransitionNode_7
   }]
 }
 StoppieStart {
  path=VehicleFullBodyOverride.AnimGraphNode_StateMachine_1.PassengerFullBody.AnimStateNode_7.Stoppie.AnimGraphNode_StateMachine_0.StoppieTransition.AnimStateNode_0.StoppieStart
  nodes [Output Animation Pose(Result=Sequence Player 'MotoDirt_Passenger_StoppieStart'() {bLoopAnimation=false, sequence=/Motorcycle/Animation/Passenger/MotoDirt_Passenger_StoppieStart})]
 }
 StoppieLoop {
  path=VehicleFullBodyOverride.AnimGraphNode_StateMachine_1.PassengerFullBody.AnimStateNode_7.Stoppie.AnimGraphNode_StateMachine_0.StoppieTransition.AnimStateNode_1.StoppieLoop
  nodes [Output Animation Pose(Result=Sequence Player 'MotoDirt_Passenger_StoppieLoop'() {sequence=/Motorcycle/Animation/Passenger/MotoDirt_Passenger_StoppieLoop})]
 }
 Transition {
  path=VehicleFullBodyOverride.AnimGraphNode_StateMachine_1.PassengerFullBody.AnimStateNode_7.Stoppie.AnimGraphNode_StateMachine_0.StoppieTransition.AnimStateTransitionNode_7.Transition
  nodes [Result()]
 }
 StoppieLand {
  path=VehicleFullBodyOverride.AnimGraphNode_StateMachine_1.PassengerFullBody.AnimStateNode_8.StoppieLand
  nodes [Output Animation Pose(Result=Sequence Player 'MotoDirt_Passenger_StoppieLand'!(StartPosition:float=0) -> (Pose) {bLoopAnimation=false, sequence=/Motorcycle/Animation/Passenger/MotoDirt_Passenger_StoppieLand})]
 }
 BootOutro {
  path=VehicleFullBodyOverride.AnimGraphNode_StateMachine_1.PassengerFullBody.AnimStateNode_9.BootOutro
  nodes [Output Animation Pose(Result=Sequence Player 'SportBike_Passenger_Boost_Outro'() {bLoopAnimation=false, sequence=/Sportbike/Animation/Passenger/SportBike_Passenger_Boost_Outro})]
 }
 Transition {
  path=VehicleFullBodyOverride.AnimGraphNode_StateMachine_1.PassengerFullBody.AnimStateTransitionNode_0.Transition
  nodes [Result(bCanEnterTransition=Get bShouldExitFromLocomotion() -> (bShouldExitFromLocomotion:bool))]
 }
 Transition {
  path=VehicleFullBodyOverride.AnimGraphNode_StateMachine_1.PassengerFullBody.AnimStateTransitionNode_1.Transition
  nodes [Result(bCanEnterTransition=NOT Boolean(A=Get bIsMoving()))]
 }
 Transition {
  path=VehicleFullBodyOverride.AnimGraphNode_StateMachine_1.PassengerFullBody.AnimStateTransitionNode_10.Transition
  nodes [Result(bCanEnterTransition=Get bStoppie() -> (bStoppie:bool))]
 }
 Transition {
  path=VehicleFullBodyOverride.AnimGraphNode_StateMachine_1.PassengerFullBody.AnimStateTransitionNode_11.Transition
  nodes [Result(bCanEnterTransition=Get bWheelie() -> (bWheelie:bool))]
 }
 Transition {
  path=VehicleFullBodyOverride.AnimGraphNode_StateMachine_1.PassengerFullBody.AnimStateTransitionNode_12.Transition
  nodes [Result(bCanEnterTransition=NOT Boolean(A=Get bStoppie()))]
 }
 Transition {path=VehicleFullBodyOverride.AnimGraphNode_StateMachine_1.PassengerFullBody.AnimStateTransitionNode_13.Transition, nodes [Result()]}
 Transition {path=VehicleFullBodyOverride.AnimGraphNode_StateMachine_1.PassengerFullBody.AnimStateTransitionNode_14.Transition, nodes [Result()]}
 Transition {
  path=VehicleFullBodyOverride.AnimGraphNode_StateMachine_1.PassengerFullBody.AnimStateTransitionNode_15.Transition
  nodes [Result(bCanEnterTransition=Get bIsMoving())]
 }
 Transition {
  path=VehicleFullBodyOverride.AnimGraphNode_StateMachine_1.PassengerFullBody.AnimStateTransitionNode_16.Transition
  nodes [Result(bCanEnterTransition=NOT Boolean(A=Get bIsBoosting() -> (bIsBoosting:bool)))]
 }
 Transition {path=VehicleFullBodyOverride.AnimGraphNode_StateMachine_1.PassengerFullBody.AnimStateTransitionNode_17.Transition, nodes [Result()]}
 Transition {
  path=VehicleFullBodyOverride.AnimGraphNode_StateMachine_1.PassengerFullBody.AnimStateTransitionNode_18.Transition
  nodes [Result(bCanEnterTransition=Get bWheelie())]
 }
 Transition {
  path=VehicleFullBodyOverride.AnimGraphNode_StateMachine_1.PassengerFullBody.AnimStateTransitionNode_19.Transition
  nodes [Result(bCanEnterTransition=Get bStoppie())]
 }
 Transition {
  path=VehicleFullBodyOverride.AnimGraphNode_StateMachine_1.PassengerFullBody.AnimStateTransitionNode_2.Transition
  nodes [Result(bCanEnterTransition=Get bShouldJump() -> (bShouldJump:bool))]
 }
 Transition {
  path=VehicleFullBodyOverride.AnimGraphNode_StateMachine_1.PassengerFullBody.AnimStateTransitionNode_3.Transition
  nodes [Result(bCanEnterTransition=NOT Boolean(A=Get bIsInAir() -> (bIsInAir:bool)))]
 }
 Transition {
  path=VehicleFullBodyOverride.AnimGraphNode_StateMachine_1.PassengerFullBody.AnimStateTransitionNode_4.Transition
  nodes [Result(bCanEnterTransition=Get bEarlyOutStoppieOutro() -> (bEarlyOutStoppieOutro:bool))]
 }
 Transition {
  path=VehicleFullBodyOverride.AnimGraphNode_StateMachine_1.PassengerFullBody.AnimStateTransitionNode_5.Transition
  nodes [Result(bCanEnterTransition=true)]
 }
 Transition {
  path=VehicleFullBodyOverride.AnimGraphNode_StateMachine_1.PassengerFullBody.AnimStateTransitionNode_6.Transition
  nodes [Result(bCanEnterTransition=Get bWheelie())]
 }
 Transition {path=VehicleFullBodyOverride.AnimGraphNode_StateMachine_1.PassengerFullBody.AnimStateTransitionNode_7.Transition, nodes [Result()]}
 Transition {
  path=VehicleFullBodyOverride.AnimGraphNode_StateMachine_1.PassengerFullBody.AnimStateTransitionNode_8.Transition
  nodes [Result(bCanEnterTransition=Get bIsBoosting())]
 }
 Transition {
  path=VehicleFullBodyOverride.AnimGraphNode_StateMachine_1.PassengerFullBody.AnimStateTransitionNode_9.Transition
  nodes [Result(bCanEnterTransition=NOT Boolean(A=Get bWheelie()))]
 }
 VehicleLowerBodyOverride {nodes [
   Input Pose() {name=InPoseUpperBody}
   AnimGraphNode_StateMachine_0 = PassengerLowerBodyTurn() {is=title AnimGraphNode_StateMachine}
   Output Pose(Result=Blend Poses by bool(
     BlendPose_0=Rotate Root Bone(BasePose=@AnimGraphNode_StateMachine_0, Pitch=0, Yaw=0, MeshToComponent="0,0,0")
     BlendPose_1=Input Pose() {name=InPoseLowerBodyDefault}
    ) {blendType=Linear}) {layerGroup=VehicleOverride}
  ]}
 PassengerLowerBodyTurn {
  entry=PassengerPose_N
  path=VehicleLowerBodyOverride.AnimGraphNode_StateMachine_0.PassengerLowerBodyTurn
  ids=id
  nodes [
   AnimStateNode_10 = Transition_EtoS() {is=title AnimStateNode}
   AnimStateNode_11 = PassengerPose_S() {is=title AnimStateNode}
   AnimStateNode_12 = Transition_StoW() {is=title AnimStateNode}
   AnimStateNode_13 = PassengerPose_W() {is=title AnimStateNode}
   AnimStateNode_15 = Transition_WtoN() {is=title AnimStateNode}
   AnimStateNode_16 = PassengerPose_N() {is=title AnimStateNode}
   AnimStateNode_17 = Transition_EtoN() {is=title AnimStateNode}
   AnimStateNode_18 = Transition_StoE() {is=title AnimStateNode}
   AnimStateNode_19 = Transition_WtoS() {is=title AnimStateNode}
   AnimStateNode_20 = Transition_NtoW() {is=title AnimStateNode}
   AnimStateNode_8 = Transition_NtoE() {is=title AnimStateNode}
   AnimStateNode_9 = PassengerPose_E() {is=title AnimStateNode}
   AnimStateTransitionNode_0 = AnimStateTransitionNode() {logicType=TLT_Inertialization}
   AnimStateTransitionNode_1 = AnimStateTransitionNode() {automaticRuleTriggerTime=0, bAutomaticRuleBasedOnSequencePlayerInState=true, logicType=TLT_Inertialization}
   AnimStateTransitionNode_10 = AnimStateTransitionNode() {logicType=TLT_Inertialization, priorityOrder=2}
   AnimStateTransitionNode_11 = AnimStateTransitionNode() {logicType=TLT_Inertialization, priorityOrder=2}
   AnimStateTransitionNode_2 = AnimStateTransitionNode() {automaticRuleTriggerTime=0, bAutomaticRuleBasedOnSequencePlayerInState=true, logicType=TLT_Inertialization}
   AnimStateTransitionNode_25 = AnimStateTransitionNode() {logicType=TLT_Inertialization}
   AnimStateTransitionNode_26 = AnimStateTransitionNode() {automaticRuleTriggerTime=0, bAutomaticRuleBasedOnSequencePlayerInState=true, logicType=TLT_Inertialization}
   AnimStateTransitionNode_27 = AnimStateTransitionNode() {logicType=TLT_Inertialization}
   AnimStateTransitionNode_28 = AnimStateTransitionNode() {logicType=TLT_Inertialization}
   AnimStateTransitionNode_29 = AnimStateTransitionNode() {logicType=TLT_Inertialization}
   AnimStateTransitionNode_3 = AnimStateTransitionNode() {automaticRuleTriggerTime=0, bAutomaticRuleBasedOnSequencePlayerInState=true, logicType=TLT_Inertialization}
   AnimStateTransitionNode_30 = AnimStateTransitionNode() {logicType=TLT_Inertialization}
   AnimStateTransitionNode_31 = AnimStateTransitionNode() {logicType=TLT_Inertialization}
   AnimStateTransitionNode_32 = AnimStateTransitionNode() {logicType=TLT_Inertialization}
   AnimStateTransitionNode_35 = AnimStateTransitionNode() {logicType=TLT_Inertialization}
   AnimStateTransitionNode_36 = AnimStateTransitionNode() {logicType=TLT_Inertialization}
   AnimStateTransitionNode_37 = AnimStateTransitionNode() {logicType=TLT_Inertialization}
   AnimStateTransitionNode_38 = AnimStateTransitionNode() {automaticRuleTriggerTime=0, bAutomaticRuleBasedOnSequencePlayerInState=true, logicType=TLT_Inertialization}
   AnimStateTransitionNode_39 = AnimStateTransitionNode() {logicType=TLT_Inertialization}
   AnimStateTransitionNode_4 = AnimStateTransitionNode() {logicType=TLT_Inertialization}
   AnimStateTransitionNode_40 = AnimStateTransitionNode() {automaticRuleTriggerTime=0, bAutomaticRuleBasedOnSequencePlayerInState=true, logicType=TLT_Inertialization}
   AnimStateTransitionNode_41 = AnimStateTransitionNode() {logicType=TLT_Inertialization}
   AnimStateTransitionNode_42 = AnimStateTransitionNode() {automaticRuleTriggerTime=0, bAutomaticRuleBasedOnSequencePlayerInState=true, logicType=TLT_Inertialization}
   AnimStateTransitionNode_5 = AnimStateTransitionNode() {logicType=TLT_Inertialization}
   AnimStateTransitionNode_6 = AnimStateTransitionNode() {logicType=TLT_Inertialization}
   AnimStateTransitionNode_7 = AnimStateTransitionNode() {automaticRuleTriggerTime=0, bAutomaticRuleBasedOnSequencePlayerInState=true, logicType=TLT_Inertialization}
   AnimStateTransitionNode_8 = AnimStateTransitionNode() {logicType=TLT_Inertialization, priorityOrder=2}
   AnimStateTransitionNode_9 = AnimStateTransitionNode() {logicType=TLT_Inertialization, priorityOrder=2}
  ]
  links [
   {
    dst=PassengerPose_W
    rule=VehicleLowerBodyOverride.AnimGraphNode_StateMachine_0.PassengerLowerBodyTurn.AnimStateTransitionNode_11.AnimationTransitionGraph_0
    src=PassengerPose_E
    via=AnimStateTransitionNode_11
   }
   {
    dst=Transition_EtoN
    rule=VehicleLowerBodyOverride.AnimGraphNode_StateMachine_0.PassengerLowerBodyTurn.AnimStateTransitionNode_35.AnimationTransitionGraph_0
    src=PassengerPose_E
    via=AnimStateTransitionNode_35
   }
   {
    dst=Transition_EtoS
    rule=VehicleLowerBodyOverride.AnimGraphNode_StateMachine_0.PassengerLowerBodyTurn.AnimStateTransitionNode_27.AnimationTransitionGraph_1
    src=PassengerPose_E
    via=AnimStateTransitionNode_27
   }
   {
    dst=PassengerPose_S
    rule=VehicleLowerBodyOverride.AnimGraphNode_StateMachine_0.PassengerLowerBodyTurn.AnimStateTransitionNode_8.AnimationTransitionGraph_1
    src=PassengerPose_N
    via=AnimStateTransitionNode_8
   }
   {
    dst=Transition_NtoE
    rule=VehicleLowerBodyOverride.AnimGraphNode_StateMachine_0.PassengerLowerBodyTurn.AnimStateTransitionNode_25.AnimationTransitionGraph_0
    src=PassengerPose_N
    via=AnimStateTransitionNode_25
   }
   {
    dst=Transition_NtoW
    rule=VehicleLowerBodyOverride.AnimGraphNode_StateMachine_0.PassengerLowerBodyTurn.AnimStateTransitionNode_41.AnimationTransitionGraph_0
    src=PassengerPose_N
    via=AnimStateTransitionNode_41
   }
   {
    dst=PassengerPose_N
    rule=VehicleLowerBodyOverride.AnimGraphNode_StateMachine_0.PassengerLowerBodyTurn.AnimStateTransitionNode_9.AnimationTransitionGraph_1
    src=PassengerPose_S
    via=AnimStateTransitionNode_9
   }
   {
    dst=Transition_StoE
    rule=VehicleLowerBodyOverride.AnimGraphNode_StateMachine_0.PassengerLowerBodyTurn.AnimStateTransitionNode_37.AnimationTransitionGraph_0
    src=PassengerPose_S
    via=AnimStateTransitionNode_37
   }
   {
    dst=Transition_StoW
    rule=VehicleLowerBodyOverride.AnimGraphNode_StateMachine_0.PassengerLowerBodyTurn.AnimStateTransitionNode_29.AnimationTransitionGraph_1
    src=PassengerPose_S
    via=AnimStateTransitionNode_29
   }
   {
    dst=PassengerPose_E
    rule=VehicleLowerBodyOverride.AnimGraphNode_StateMachine_0.PassengerLowerBodyTurn.AnimStateTransitionNode_10.AnimationTransitionGraph_0
    src=PassengerPose_W
    via=AnimStateTransitionNode_10
   }
   {
    dst=Transition_WtoN
    rule=VehicleLowerBodyOverride.AnimGraphNode_StateMachine_0.PassengerLowerBodyTurn.AnimStateTransitionNode_31.AnimationTransitionGraph_0
    src=PassengerPose_W
    via=AnimStateTransitionNode_31
   }
   {
    dst=Transition_WtoS
    rule=VehicleLowerBodyOverride.AnimGraphNode_StateMachine_0.PassengerLowerBodyTurn.AnimStateTransitionNode_39.AnimationTransitionGraph_1
    src=PassengerPose_W
    via=AnimStateTransitionNode_39
   }
   {
    dst=PassengerPose_N
    rule=VehicleLowerBodyOverride.AnimGraphNode_StateMachine_0.PassengerLowerBodyTurn.AnimStateTransitionNode_36.AnimationTransitionGraph_0
    src=Transition_EtoN
    via=AnimStateTransitionNode_36
   }
   {
    dst=PassengerPose_N
    rule=VehicleLowerBodyOverride.AnimGraphNode_StateMachine_0.PassengerLowerBodyTurn.AnimStateTransitionNode_7.AnimationTransitionGraph_0
    src=Transition_EtoN
    via=AnimStateTransitionNode_7
   }
   {
    dst=PassengerPose_S
    rule=VehicleLowerBodyOverride.AnimGraphNode_StateMachine_0.PassengerLowerBodyTurn.AnimStateTransitionNode_2.AnimationTransitionGraph_1
    src=Transition_EtoS
    via=AnimStateTransitionNode_2
   }
   {
    dst=PassengerPose_S
    rule=VehicleLowerBodyOverride.AnimGraphNode_StateMachine_0.PassengerLowerBodyTurn.AnimStateTransitionNode_28.AnimationTransitionGraph_1
    src=Transition_EtoS
    via=AnimStateTransitionNode_28
   }
   {
    dst=PassengerPose_E
    rule=VehicleLowerBodyOverride.AnimGraphNode_StateMachine_0.PassengerLowerBodyTurn.AnimStateTransitionNode_0.AnimationTransitionGraph_0
    src=Transition_NtoE
    via=AnimStateTransitionNode_0
   }
   {
    dst=PassengerPose_E
    rule=VehicleLowerBodyOverride.AnimGraphNode_StateMachine_0.PassengerLowerBodyTurn.AnimStateTransitionNode_26.AnimationTransitionGraph_0
    src=Transition_NtoE
    via=AnimStateTransitionNode_26
   }
   {
    dst=PassengerPose_W
    rule=VehicleLowerBodyOverride.AnimGraphNode_StateMachine_0.PassengerLowerBodyTurn.AnimStateTransitionNode_4.AnimationTransitionGraph_0
    src=Transition_NtoW
    via=AnimStateTransitionNode_4
   }
   {
    dst=PassengerPose_W
    rule=VehicleLowerBodyOverride.AnimGraphNode_StateMachine_0.PassengerLowerBodyTurn.AnimStateTransitionNode_42.AnimationTransitionGraph_0
    src=Transition_NtoW
    via=AnimStateTransitionNode_42
   }
   {
    dst=PassengerPose_E
    rule=VehicleLowerBodyOverride.AnimGraphNode_StateMachine_0.PassengerLowerBodyTurn.AnimStateTransitionNode_38.AnimationTransitionGraph_1
    src=Transition_StoE
    via=AnimStateTransitionNode_38
   }
   {
    dst=PassengerPose_E
    rule=VehicleLowerBodyOverride.AnimGraphNode_StateMachine_0.PassengerLowerBodyTurn.AnimStateTransitionNode_6.AnimationTransitionGraph_1
    src=Transition_StoE
    via=AnimStateTransitionNode_6
   }
   {
    dst=PassengerPose_W
    rule=VehicleLowerBodyOverride.AnimGraphNode_StateMachine_0.PassengerLowerBodyTurn.AnimStateTransitionNode_1.AnimationTransitionGraph_1
    src=Transition_StoW
    via=AnimStateTransitionNode_1
   }
   {
    dst=PassengerPose_W
    rule=VehicleLowerBodyOverride.AnimGraphNode_StateMachine_0.PassengerLowerBodyTurn.AnimStateTransitionNode_30.AnimationTransitionGraph_1
    src=Transition_StoW
    via=AnimStateTransitionNode_30
   }
   {
    dst=PassengerPose_N
    rule=VehicleLowerBodyOverride.AnimGraphNode_StateMachine_0.PassengerLowerBodyTurn.AnimStateTransitionNode_3.AnimationTransitionGraph_0
    src=Transition_WtoN
    via=AnimStateTransitionNode_3
   }
   {
    dst=PassengerPose_N
    rule=VehicleLowerBodyOverride.AnimGraphNode_StateMachine_0.PassengerLowerBodyTurn.AnimStateTransitionNode_32.AnimationTransitionGraph_0
    src=Transition_WtoN
    via=AnimStateTransitionNode_32
   }
   {
    dst=PassengerPose_S
    rule=VehicleLowerBodyOverride.AnimGraphNode_StateMachine_0.PassengerLowerBodyTurn.AnimStateTransitionNode_40.AnimationTransitionGraph_0
    src=Transition_WtoS
    via=AnimStateTransitionNode_40
   }
   {
    dst=PassengerPose_S
    rule=VehicleLowerBodyOverride.AnimGraphNode_StateMachine_0.PassengerLowerBodyTurn.AnimStateTransitionNode_5.AnimationTransitionGraph_0
    src=Transition_WtoS
    via=AnimStateTransitionNode_5
   }
  ]
 }
 Transition_EtoS {
  path=VehicleLowerBodyOverride.AnimGraphNode_StateMachine_0.PassengerLowerBodyTurn.AnimStateNode_10.Transition_EtoS
  nodes [Output Animation Pose(Result=Sequence Player 'MotoDirt_Passenger_Turn_EtoS'() {bLoopAnimation=false, sequence=/Motorcycle/Animation/Passenger/MotoDirt_Passenger_Turn_EtoS})]
 }
 PassengerPose_S {
  path=VehicleLowerBodyOverride.AnimGraphNode_StateMachine_0.PassengerLowerBodyTurn.AnimStateNode_11.PassengerPose_S
  nodes [
   n1 = AimOffset Player 'ShoppingCart_Passenger_LowerBodyAO'!(
    X:float=0
    BasePose:Pose Link Structure=Rotate Root Bone!(
     BasePose=Sequence Player 'MotoDirt_Passenger_ADS_S_Pose'() {sequence=/Motorcycle/Animation/Passenger/MotoDirt_Passenger_ADS_S_Pose}
     Yaw=180
     MeshToComponent="0,0,0"
    ) -> (Pose)
    Alpha:float=1
   ) -> (Pose:Pose Link Structure) {blendSpace=/Game/Animation/Game/MainPlayer/Vehicles/ShoppingCart/AO/ShoppingCart_Passenger_LowerBodyAO, alphaBoolBlend {blendOption=Linear}}
   Output Animation Pose(Result=Rotate Root Bone(BasePose=@n1, Yaw=-180, MeshToComponent="0,0,0"))
  ]
 }
 Transition_StoW {
  path=VehicleLowerBodyOverride.AnimGraphNode_StateMachine_0.PassengerLowerBodyTurn.AnimStateNode_12.Transition_StoW
  nodes [Output Animation Pose(Result=Sequence Player 'MotoDirt_Passenger_Turn_StoW'() {bLoopAnimation=false, sequence=/Motorcycle/Animation/Passenger/MotoDirt_Passenger_Turn_StoW})]
 }
 PassengerPose_W {
  path=VehicleLowerBodyOverride.AnimGraphNode_StateMachine_0.PassengerLowerBodyTurn.AnimStateNode_13.PassengerPose_W
  nodes [
   n1 = AimOffset Player 'ShoppingCart_Passenger_LowerBodyAO'(
    X=0
    BasePose=Rotate Root Bone(
     BasePose=Sequence Player 'MotoDirt_Passenger_ADS_W_Pose'() {sequence=/Motorcycle/Animation/Passenger/MotoDirt_Passenger_ADS_W_Pose}
     Yaw=90
     MeshToComponent="0,0,0"
    )
    Alpha=1
   ) {blendSpace=&ShoppingCart_Passenger_LowerBodyAO, alphaBoolBlend {blendOption=Linear}}
   Output Animation Pose(Result=Rotate Root Bone(BasePose=@n1, Yaw=-90, MeshToComponent="0,0,0"))
  ]
 }
 Transition_WtoN {
  path=VehicleLowerBodyOverride.AnimGraphNode_StateMachine_0.PassengerLowerBodyTurn.AnimStateNode_15.Transition_WtoN
  nodes [Output Animation Pose(Result=Sequence Player 'MotoDirt_Passenger_Turn_WtoN'() {bLoopAnimation=false, sequence=/Motorcycle/Animation/Passenger/MotoDirt_Passenger_Turn_WtoN})]
 }
 PassengerPose_N {
  path=VehicleLowerBodyOverride.AnimGraphNode_StateMachine_0.PassengerLowerBodyTurn.AnimStateNode_16.PassengerPose_N
  nodes [
   n1 = AimOffset Player 'ShoppingCart_Passenger_LowerBodyAO'(
    X=0
    BasePose=Sequence Player 'MotoDirt_Passenger_ADS_N_Pose'() {sequence=/Motorcycle/Animation/Passenger/MotoDirt_Passenger_ADS_N_Pose}
    Alpha=1
   ) {blendSpace=&ShoppingCart_Passenger_LowerBodyAO, alphaBoolBlend {blendOption=Linear}}
   Output Animation Pose(Result=@n1)
  ]
 }
 Transition_EtoN {
  path=VehicleLowerBodyOverride.AnimGraphNode_StateMachine_0.PassengerLowerBodyTurn.AnimStateNode_17.Transition_EtoN
  nodes [Output Animation Pose(Result=Sequence Player 'MotoDirt_Passenger_Turn_EtoN'() {bLoopAnimation=false, sequence=/Motorcycle/Animation/Passenger/MotoDirt_Passenger_Turn_EtoN})]
 }
 Transition_StoE {
  path=VehicleLowerBodyOverride.AnimGraphNode_StateMachine_0.PassengerLowerBodyTurn.AnimStateNode_18.Transition_StoE
  nodes [Output Animation Pose(Result=Sequence Player 'MotoDirt_Passenger_Turn_StoE'() {bLoopAnimation=false, sequence=/Motorcycle/Animation/Passenger/MotoDirt_Passenger_Turn_StoE})]
 }
 Transition_WtoS {
  path=VehicleLowerBodyOverride.AnimGraphNode_StateMachine_0.PassengerLowerBodyTurn.AnimStateNode_19.Transition_WtoS
  nodes [Output Animation Pose(Result=Sequence Player 'MotoDirt_Passenger_Turn_WtoS'() {bLoopAnimation=false, sequence=/Motorcycle/Animation/Passenger/MotoDirt_Passenger_Turn_WtoS})]
 }
 Transition_NtoW {
  path=VehicleLowerBodyOverride.AnimGraphNode_StateMachine_0.PassengerLowerBodyTurn.AnimStateNode_20.Transition_NtoW
  nodes [Output Animation Pose(Result=Sequence Player 'MotoDirt_Passenger_Turn_NtoW'() {bLoopAnimation=false, sequence=/Motorcycle/Animation/Passenger/MotoDirt_Passenger_Turn_NtoW})]
 }
 Transition_NtoE {
  path=VehicleLowerBodyOverride.AnimGraphNode_StateMachine_0.PassengerLowerBodyTurn.AnimStateNode_8.Transition_NtoE
  nodes [Output Animation Pose(Result=Sequence Player 'MotoDirt_Passenger_Turn_NtoE'() {bLoopAnimation=false, sequence=/Motorcycle/Animation/Passenger/MotoDirt_Passenger_Turn_NtoE})]
 }
 PassengerPose_E {
  path=VehicleLowerBodyOverride.AnimGraphNode_StateMachine_0.PassengerLowerBodyTurn.AnimStateNode_9.PassengerPose_E
  nodes [
   n1 = AimOffset Player 'ShoppingCart_Passenger_LowerBodyAO'(
    X=0
    BasePose=Rotate Root Bone(
     BasePose=Sequence Player 'MotoDirt_Passenger_ADS_E_Pose'() {sequence=/Motorcycle/Animation/Passenger/MotoDirt_Passenger_ADS_E_Pose}
     Yaw=-90
     MeshToComponent="0,0,0"
    )
    Alpha=1
   ) {blendSpace=&ShoppingCart_Passenger_LowerBodyAO, alphaBoolBlend {blendOption=Linear}}
   Output Animation Pose(Result=Rotate Root Bone(BasePose=@n1, Yaw=90, MeshToComponent="0,0,0"))
  ]
 }
 AnimationTransitionGraph_0 {
  path=VehicleLowerBodyOverride.AnimGraphNode_StateMachine_0.PassengerLowerBodyTurn.AnimStateTransitionNode_0.AnimationTransitionGraph_0
  nodes [Result(bCanEnterTransition=NOT Boolean(A=Get bAimRGT() -> (bAimRGT:bool)))]
 }
 AnimationTransitionGraph_1 {
  path=VehicleLowerBodyOverride.AnimGraphNode_StateMachine_0.PassengerLowerBodyTurn.AnimStateTransitionNode_1.AnimationTransitionGraph_1
  nodes [Result()]
 }
 AnimationTransitionGraph_0 {
  path=VehicleLowerBodyOverride.AnimGraphNode_StateMachine_0.PassengerLowerBodyTurn.AnimStateTransitionNode_10.AnimationTransitionGraph_0
  nodes [Result(bCanEnterTransition=Get bAimRGT())]
 }
 AnimationTransitionGraph_0 {
  path=VehicleLowerBodyOverride.AnimGraphNode_StateMachine_0.PassengerLowerBodyTurn.AnimStateTransitionNode_11.AnimationTransitionGraph_0
  nodes [Result(bCanEnterTransition=Get bAimLFT() -> (bAimLFT:bool))]
 }
 AnimationTransitionGraph_1 {
  path=VehicleLowerBodyOverride.AnimGraphNode_StateMachine_0.PassengerLowerBodyTurn.AnimStateTransitionNode_2.AnimationTransitionGraph_1
  nodes [Result()]
 }
 AnimationTransitionGraph_0 {
  path=VehicleLowerBodyOverride.AnimGraphNode_StateMachine_0.PassengerLowerBodyTurn.AnimStateTransitionNode_25.AnimationTransitionGraph_0
  nodes [Result(bCanEnterTransition=Get bAimRGT())]
 }
 AnimationTransitionGraph_0 {
  path=VehicleLowerBodyOverride.AnimGraphNode_StateMachine_0.PassengerLowerBodyTurn.AnimStateTransitionNode_26.AnimationTransitionGraph_0
  nodes [Result()]
 }
 AnimationTransitionGraph_1 {
  path=VehicleLowerBodyOverride.AnimGraphNode_StateMachine_0.PassengerLowerBodyTurn.AnimStateTransitionNode_27.AnimationTransitionGraph_1
  nodes [Result(bCanEnterTransition=Get bAimBWD() -> (bAimBWD:bool))]
 }
 AnimationTransitionGraph_1 {
  path=VehicleLowerBodyOverride.AnimGraphNode_StateMachine_0.PassengerLowerBodyTurn.AnimStateTransitionNode_28.AnimationTransitionGraph_1
  nodes [Result(bCanEnterTransition=NOT Boolean(A=Get bAimBWD()))]
 }
 AnimationTransitionGraph_1 {
  path=VehicleLowerBodyOverride.AnimGraphNode_StateMachine_0.PassengerLowerBodyTurn.AnimStateTransitionNode_29.AnimationTransitionGraph_1
  nodes [Result(bCanEnterTransition=Get bAimLFT())]
 }
 AnimationTransitionGraph_0 {
  path=VehicleLowerBodyOverride.AnimGraphNode_StateMachine_0.PassengerLowerBodyTurn.AnimStateTransitionNode_3.AnimationTransitionGraph_0
  nodes [Result()]
 }
 AnimationTransitionGraph_1 {
  path=VehicleLowerBodyOverride.AnimGraphNode_StateMachine_0.PassengerLowerBodyTurn.AnimStateTransitionNode_30.AnimationTransitionGraph_1
  nodes [Result(bCanEnterTransition=NOT Boolean(A=Get bAimLFT()))]
 }
 AnimationTransitionGraph_0 {
  path=VehicleLowerBodyOverride.AnimGraphNode_StateMachine_0.PassengerLowerBodyTurn.AnimStateTransitionNode_31.AnimationTransitionGraph_0
  nodes [Result(bCanEnterTransition=Get bAimFWD() -> (bAimFWD:bool))]
 }
 AnimationTransitionGraph_0 {
  path=VehicleLowerBodyOverride.AnimGraphNode_StateMachine_0.PassengerLowerBodyTurn.AnimStateTransitionNode_32.AnimationTransitionGraph_0
  nodes [Result(bCanEnterTransition=NOT Boolean(A=Get bAimFWD()))]
 }
 AnimationTransitionGraph_0 {
  path=VehicleLowerBodyOverride.AnimGraphNode_StateMachine_0.PassengerLowerBodyTurn.AnimStateTransitionNode_35.AnimationTransitionGraph_0
  nodes [Result(bCanEnterTransition=Get bAimFWD())]
 }
 AnimationTransitionGraph_0 {
  path=VehicleLowerBodyOverride.AnimGraphNode_StateMachine_0.PassengerLowerBodyTurn.AnimStateTransitionNode_36.AnimationTransitionGraph_0
  nodes [Result(bCanEnterTransition=NOT Boolean(A=Get bAimFWD()))]
 }
 AnimationTransitionGraph_0 {
  path=VehicleLowerBodyOverride.AnimGraphNode_StateMachine_0.PassengerLowerBodyTurn.AnimStateTransitionNode_37.AnimationTransitionGraph_0
  nodes [Result(bCanEnterTransition=Get bAimRGT())]
 }
 AnimationTransitionGraph_1 {
  path=VehicleLowerBodyOverride.AnimGraphNode_StateMachine_0.PassengerLowerBodyTurn.AnimStateTransitionNode_38.AnimationTransitionGraph_1
  nodes [Result()]
 }
 AnimationTransitionGraph_1 {
  path=VehicleLowerBodyOverride.AnimGraphNode_StateMachine_0.PassengerLowerBodyTurn.AnimStateTransitionNode_39.AnimationTransitionGraph_1
  nodes [Result(bCanEnterTransition=Get bAimBWD())]
 }
 AnimationTransitionGraph_0 {
  path=VehicleLowerBodyOverride.AnimGraphNode_StateMachine_0.PassengerLowerBodyTurn.AnimStateTransitionNode_4.AnimationTransitionGraph_0
  nodes [Result(bCanEnterTransition=NOT Boolean(A=Get bAimLFT()))]
 }
 AnimationTransitionGraph_0 {
  path=VehicleLowerBodyOverride.AnimGraphNode_StateMachine_0.PassengerLowerBodyTurn.AnimStateTransitionNode_40.AnimationTransitionGraph_0
  nodes [Result()]
 }
 AnimationTransitionGraph_0 {
  path=VehicleLowerBodyOverride.AnimGraphNode_StateMachine_0.PassengerLowerBodyTurn.AnimStateTransitionNode_41.AnimationTransitionGraph_0
  nodes [Result(bCanEnterTransition=Get bAimLFT())]
 }
 AnimationTransitionGraph_0 {
  path=VehicleLowerBodyOverride.AnimGraphNode_StateMachine_0.PassengerLowerBodyTurn.AnimStateTransitionNode_42.AnimationTransitionGraph_0
  nodes [Result()]
 }
 AnimationTransitionGraph_0 {
  path=VehicleLowerBodyOverride.AnimGraphNode_StateMachine_0.PassengerLowerBodyTurn.AnimStateTransitionNode_5.AnimationTransitionGraph_0
  nodes [Result(bCanEnterTransition=NOT Boolean(A=Get bAimBWD()))]
 }
 AnimationTransitionGraph_1 {
  path=VehicleLowerBodyOverride.AnimGraphNode_StateMachine_0.PassengerLowerBodyTurn.AnimStateTransitionNode_6.AnimationTransitionGraph_1
  nodes [Result(bCanEnterTransition=NOT Boolean(A=Get bAimRGT()))]
 }
 AnimationTransitionGraph_0 {
  path=VehicleLowerBodyOverride.AnimGraphNode_StateMachine_0.PassengerLowerBodyTurn.AnimStateTransitionNode_7.AnimationTransitionGraph_0
  nodes [Result()]
 }
 AnimationTransitionGraph_1 {
  path=VehicleLowerBodyOverride.AnimGraphNode_StateMachine_0.PassengerLowerBodyTurn.AnimStateTransitionNode_8.AnimationTransitionGraph_1
  nodes [Result(bCanEnterTransition=Get bAimBWD())]
 }
 AnimationTransitionGraph_1 {
  path=VehicleLowerBodyOverride.AnimGraphNode_StateMachine_0.PassengerLowerBodyTurn.AnimStateTransitionNode_9.AnimationTransitionGraph_1
  nodes [Result(bCanEnterTransition=Get bAimFWD())]
 }
 VehicleSplitBodyOverride {nodes [
   Input Pose() {name=InPoseSplitBody}
   Input Pose() {name=InPoseUpperAndLowerBody}
   Input Pose() {name=InPoseUpperAndLowerBodyRemoveRoot}
   Output Pose() {layerGroup=Vehicle Override}
  ]}
]
variables [isDBNO {type=boolean}]
