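# Sound cue for a sportbike engine switching off, classed as exterior vehicle audio.
# A Mixer layers one of three equally weighted random engine-off recordings at full level with a power-down wave kept very low in the mix.
# The power-down layer passes through a Modulator with its volume range fixed at unity.
# The cue boosts overall volume and uses shared vehicle movement attenuation and engine concurrency settings.
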SoundCue /Sportbike/Audio/Veh_MC_SB_Engine_Off_Cue
root=Mixer_0
nodes [
 {id=Mixer_0, children [Random_0,Modulator_1], properties {inputVolume [1,0.08]}}
 {id=Modulator_1, children [WavePlayer_6], properties {volumeMax=1, volumeMin=1}}
 {id=Random_0, children [WavePlayer_1,WavePlayer_2,WavePlayer_4], properties {weights [1,1,1]}}
 {id=WavePlayer_1, properties {soundWaveAssetPtr=&/CBR_Off_01}}
 {id=WavePlayer_2, properties {soundWaveAssetPtr=&/CBR_Off_02}}
 {id=WavePlayer_4, properties {soundWaveAssetPtr=&/CBR_Off_03}}
 {id=WavePlayer_6, properties {soundWaveAssetPtr=&/Veh_MC_Sportbike_PowerDown}}
]
properties {
 attenuationSettings=/VehicleAudio/Settings/Vehicle_Movement_Att
 maxDistance=12512
 soundClassObject=&Fort_Vehicles_Exterior
 volumeMultiplier=4
 concurrencySet [/VehicleAudio/Settings/Vehicle_Engine_Conc]
}
