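# Sound cue for the sportbike's engine-on sound, played when the motorcycle engine starts.
# A Random node picks one of five CBR engine start recordings with equal weight.
# It routes through the exterior vehicle sound class with shared vehicle movement attenuation and engine concurrency, and a boosted volume multiplier.
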SoundCue /Sportbike/Audio/Veh_MC_SB_Engine_On_Cue
root=Random_0
nodes [
 {id=Random_0, children [WavePlayer_0,WavePlayer_3,WavePlayer_4,WavePlayer_5,WavePlayer_6], properties {weights [1,1,1,1,1]}}
 {id=WavePlayer_0, properties {soundWaveAssetPtr=&/CBR_Start_01}}
 {id=WavePlayer_3, properties {soundWaveAssetPtr=&/CBR_Start_02}}
 {id=WavePlayer_4, properties {soundWaveAssetPtr=&/CBR_Start_03}}
 {id=WavePlayer_5, properties {soundWaveAssetPtr=&/CBR_Start_04}}
 {id=WavePlayer_6, properties {soundWaveAssetPtr=&/CBR_Start_05}}
]
properties {
 attenuationSettings=/VehicleAudio/Settings/Vehicle_Movement_Att
 maxDistance=12512
 soundClassObject=&Fort_Vehicles_Exterior
 volumeMultiplier=3
 concurrencySet [/VehicleAudio/Settings/Vehicle_Engine_Conc]
}
