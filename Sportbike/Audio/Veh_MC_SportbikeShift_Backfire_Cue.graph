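# Sound cue for a sportbike gear shift with backfire, routed to the exterior vehicle sound class with close-detail attenuation.
# A mixer layers a gear-shift sample, a pitched-down sci-fi blow-off, a random pitched-up flame burst and a random rev-limiter pop on a short fixed delay.
# The flame burst layer has the heaviest mix weight.
# Per-owner concurrency caps simultaneous instances, stopping the oldest.
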SoundCue /Sportbike/Audio/Veh_MC_SportbikeShift_Backfire_Cue
root=Mixer_0
nodes [
 {id=Delay_1, children [Random_2], properties {delayMax=0.35, delayMin=0.35}}
 {id=Mixer_0, children [Modulator_0,Modulator_3,Modulator_5,Modulator_6], properties {inputVolume [1.6,1.3,4,1]}}
 {id=Modulator_0, children [WavePlayer_9], properties {pitchMax=1, pitchMin=1, volumeMax=1, volumeMin=1}}
 {id=Modulator_3, children [Delay_1], properties {volumeMax=1, volumeMin=1}}
 {id=Modulator_5, children [Random_1], properties {pitchMax=1.5, pitchMin=1.5, volumeMax=1, volumeMin=1}}
 {id=Modulator_6, children [WavePlayer_20], properties {pitchMax=0.5, pitchMin=0.5}}
 {
  id=Random_1
  children [WavePlayer_0, WavePlayer_1, WavePlayer_2, WavePlayer_7, WavePlayer_8, WavePlayer_19]
  properties {weights [1,1,1,1,1,1]}
 }
 {id=Random_2, children [WavePlayer_11, WavePlayer_12, WavePlayer_13, WavePlayer_18], properties {weights [1,1,1,1]}}
 {id=WavePlayer_0, properties {soundWaveAssetPtr=&/Veh_MC_Sportbike_BackfireFlameBurst_01}}
 {id=WavePlayer_1, properties {soundWaveAssetPtr=&/Veh_MC_Sportbike_BackfireFlameBurst_02}}
 {id=WavePlayer_11, properties {soundWaveAssetPtr=&/Veh_MC_Sportbike_RevLimitPops_08}}
 {id=WavePlayer_12, properties {soundWaveAssetPtr=&/Veh_MC_Sportbike_RevLimitPops_01}}
 {id=WavePlayer_13, properties {soundWaveAssetPtr=&/Veh_MC_Sportbike_RevLimitPops_02}}
 {id=WavePlayer_18, properties {soundWaveAssetPtr=&/Veh_MC_Sportbike_RevLimitPops_07}}
 {id=WavePlayer_19, properties {soundWaveAssetPtr=&/Veh_MC_Sportbike_BackfireFlameBurst_06}}
 {id=WavePlayer_2, properties {soundWaveAssetPtr=&/Veh_MC_Sportbike_BackfireFlameBurst_03}}
 {id=WavePlayer_20, properties {soundWaveAssetPtr=&/Veh_MC_Sportbike_SciFi_BlowOff}}
 {id=WavePlayer_7, properties {soundWaveAssetPtr=&/Veh_MC_Sportbike_BackfireFlameBurst_04}}
 {id=WavePlayer_8, properties {soundWaveAssetPtr=&/Veh_MC_Sportbike_BackfireFlameBurst_05}}
 {id=WavePlayer_9, properties {soundWaveAssetPtr=&/Veh_MC_Sportbike_ShiftGear}}
]
properties {
 attenuationSettings=/Sportbike/Audio/Settings/Vehicle_SportbikeShift_Close_Detail_Att
 bOverrideConcurrency=true
 maxDistance=6800
 pitchMultiplier=0.8
 soundClassObject=&Fort_Vehicles_Exterior
 volumeMultiplier=0.5
 concurrencyOverrides {bLimitToOwner=true, maxCount=2, resolutionRule=StopOldest}
}
