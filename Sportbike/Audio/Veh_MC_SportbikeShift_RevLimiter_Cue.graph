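# Sound cue for sportbike rev limiter pops on shifts, assigned to the vehicle interior sound class.
# A short randomized delay leads into a modulator for slight variation, then a random node picks among four pop samples.
# Four equally weighted empty slots in that random node leave roughly half the triggers silent, so pops are intermittent.
# Close-detail vehicle attenuation and a small concurrency override limit its reach and stacking, with lowered pitch and volume.
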SoundCue /Sportbike/Audio/Veh_MC_SportbikeShift_RevLimiter_Cue
root=Delay_0
nodes [
 {id=Delay_0, children [Modulator_0], properties {delayMax=0.25, delayMin=0.2}}
 {id=Modulator_0, children [Random_3], properties {pitchMax=1}}
 {id=Random_3, children [WavePlayer_0,WavePlayer_1,WavePlayer_3,WavePlayer_7,null,null,null,null], properties {weights [1,1,1,1,1,1,1,1]}}
 {id=WavePlayer_0, properties {soundWaveAssetPtr=&/Veh_MC_Sportbike_RevLimitPops_01}}
 {id=WavePlayer_1, properties {soundWaveAssetPtr=&/Veh_MC_Sportbike_RevLimitPops_02}}
 {id=WavePlayer_3, properties {soundWaveAssetPtr=&/Veh_MC_Sportbike_RevLimitPops_03}}
 {id=WavePlayer_7, properties {soundWaveAssetPtr=&/Veh_MC_Sportbike_RevLimitPops_07}}
]
properties {
 attenuationSettings=/VehicleAudio/Settings/Vehicle_Movement_Close_Detail_Att
 bOverrideConcurrency=true
 maxDistance=12000
 pitchMultiplier=0.8
 soundClassObject=&Fort_Vehicles_Interior
 volumeMultiplier=0.4
 concurrencyOverrides {maxCount=3}
}
