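# Sportbike upshift sound cue that layers a shift sweetener with a backfire flame burst, routed to the vehicle engine sound class.
# A mixer blends a pitch-randomized sweetener, which has its own close-detail attenuation node, with a quieter, evenly weighted random pick of six backfire waves at fixed pitch.
# A cue-level vehicle movement attenuation is also assigned.
# Overridden concurrency allows one instance per owner, stopping the oldest.
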SoundCue /Sportbike/Audio/Veh_MC_SportbikeShift_Upshift_Cue
root=Mixer_0
nodes [
 {id=Attenuation_4, children [Modulator_4], properties {attenuationSettings=/Sportbike/Audio/Settings/Vehicle_SportbikeShift_Close_Detail_Att}}
 {id=Mixer_0, children [Attenuation_4, Modulator_5], properties {inputVolume [1,0.4]}}
 {id=Modulator_4, children [WavePlayer_6], properties {pitchMax=1.1, pitchMin=0.9, volumeMax=1, volumeMin=1}}
 {id=Modulator_5, children [Random_1], properties {pitchMax=1, pitchMin=1, volumeMax=1, volumeMin=1}}
 {
  id=Random_1
  children [WavePlayer_0, WavePlayer_1, WavePlayer_2, WavePlayer_7, WavePlayer_8, WavePlayer_19]
  properties {weights [1,1,1,1,1,1]}
 }
 {id=WavePlayer_0, properties {soundWaveAssetPtr=&/Veh_MC_Sportbike_BackfireFlameBurst_01}}
 {id=WavePlayer_1, properties {soundWaveAssetPtr=&/Veh_MC_Sportbike_BackfireFlameBurst_02}}
 {id=WavePlayer_19, properties {soundWaveAssetPtr=&/Veh_MC_Sportbike_BackfireFlameBurst_06}}
 {id=WavePlayer_2, properties {soundWaveAssetPtr=&/Veh_MC_Sportbike_BackfireFlameBurst_03}}
 {id=WavePlayer_6, properties {soundWaveAssetPtr=&/Veh_MC_Sportbike_UpShiftSweetner_02}}
 {id=WavePlayer_7, properties {soundWaveAssetPtr=&/Veh_MC_Sportbike_BackfireFlameBurst_04}}
 {id=WavePlayer_8, properties {soundWaveAssetPtr=&/Veh_MC_Sportbike_BackfireFlameBurst_05}}
]
properties {
 attenuationSettings=/VehicleAudio/Settings/Vehicle_Movement_Close_Detail_Att
 bOverrideConcurrency=true
 maxDistance=12000
 soundClassObject=&Fort_Vehicles_Engine
 volumeMultiplier=1.7
 concurrencyOverrides {bLimitToOwner=true, maxCount=1, resolutionRule=StopOldest}
}
