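# Sound cue for a sportbike stoppie in the vehicle audio set, layering three one-shot sounds through a mixer.
# Randomly picked shared wheelie-start and dirtbike brake-squeal variants play alongside a single sportbike downshift, with the latter two layers mixed slightly lower.
# Each layer has its own modulator, varying the squeal in pitch and volume and pitching the downshift up.
# Shared vehicle attenuation applies, and a concurrency override prevents new instances at the limit.
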SoundCue /Sportbike/Audio/Veh_MC_SportbikeStoppie_Cue
root=Mixer_0
nodes [
 {id=Mixer_0, children [Modulator_0,Modulator_1,Modulator_2], properties {inputVolume [1,0.7,0.8]}}
 {id=Modulator_0, children [Random_0]}
 {id=Modulator_1, children [Random_1], properties {pitchMax=1.1, pitchMin=0.9, volumeMax=1, volumeMin=0.7}}
 {id=Modulator_2, children [WavePlayer_8], properties {pitchMax=1.3, pitchMin=1.2, volumeMax=1, volumeMin=1}}
 {
  id=Random_0
  children [WavePlayer_0,WavePlayer_2,WavePlayer_3,WavePlayer_4,WavePlayer_5]
  properties {bShouldExcludeFromBranchCulling=true, weights [1,1,1,1,1]}
 }
 {id=Random_1, children [WavePlayer_1,WavePlayer_6,WavePlayer_7], properties {bShouldExcludeFromBranchCulling=true, weights [1,1,1]}}
 {id=WavePlayer_0, properties {soundWaveAssetPtr=/Motorcycle/Audio/SFX/Veh_MC_Wheelie_Start_01}}
 {id=WavePlayer_1, properties {soundWaveAssetPtr=/Motorcycle/Audio/SFX/Veh_MC_Dirtbike_Stoppie_BrakeSqueal_01}}
 {id=WavePlayer_2, properties {soundWaveAssetPtr=/Motorcycle/Audio/SFX/Veh_MC_Wheelie_Start_02}}
 {id=WavePlayer_3, properties {soundWaveAssetPtr=/Motorcycle/Audio/SFX/Veh_MC_Wheelie_Start_03}}
 {id=WavePlayer_4, properties {soundWaveAssetPtr=/Motorcycle/Audio/SFX/Veh_MC_Wheelie_Start_04}}
 {id=WavePlayer_5, properties {soundWaveAssetPtr=/Motorcycle/Audio/SFX/Veh_MC_Wheelie_Start_05}}
 {id=WavePlayer_6, properties {soundWaveAssetPtr=/Motorcycle/Audio/SFX/Veh_MC_Dirtbike_Stoppie_BrakeSqueal_02}}
 {id=WavePlayer_7, properties {soundWaveAssetPtr=/Motorcycle/Audio/SFX/Veh_MC_Dirtbike_Stoppie_BrakeSqueal_03}}
 {id=WavePlayer_8, properties {soundWaveAssetPtr=&/Veh_MC_Sportbike_Downshift_01}}
]
properties {
 attenuationSettings=/VehicleAudio/Settings/Vehicle_Movement_Boost_Att
 bOverrideConcurrency=true
 maxDistance=12000
 soundClassObject=&Fort_Vehicles
 volumeMultiplier=1.2
 attenuationOverrides {bEnableReverbSend=false, distanceAlgorithm=Logarithmic, falloffDistance=1000, stereoSpread=100, attenuationShapeExtents {x=125}}
 concurrencyOverrides {resolutionRule=PreventNew}
}
