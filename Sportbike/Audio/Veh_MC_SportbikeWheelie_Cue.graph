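# Sound cue for the sportbike wheelie, played as an exterior vehicle sound with boost attenuation settings.
# A Mixer layers a modulated random pick of five shared motorcycle wheelie-start waves, a slightly quieter pitched-up sportbike upshift, and a low-level wheelie boost wave.
# Its concurrency override blocks new instances at the limit instead of cutting off playing ones.
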SoundCue /Sportbike/Audio/Veh_MC_SportbikeWheelie_Cue
root=Mixer_0
nodes [
 {id=Mixer_0, children [Modulator_1,Modulator_0,Modulator_2], properties {inputVolume [1,0.8,0.3]}}
 {id=Modulator_0, children [WavePlayer_1], properties {pitchMax=1.2, pitchMin=1.1, volumeMax=1, volumeMin=1}}
 {id=Modulator_1, children [Random_0]}
 {id=Modulator_2, children [WavePlayer_6], properties {pitchMax=1, pitchMin=1, volumeMax=1, volumeMin=1}}
 {id=Random_0, children [WavePlayer_0,WavePlayer_2,WavePlayer_3,WavePlayer_4,WavePlayer_5], properties {weights [1,1,1,1,1]}}
 {id=WavePlayer_0, properties {soundWaveAssetPtr=/Motorcycle/Audio/SFX/Veh_MC_Wheelie_Start_01}}
 {id=WavePlayer_1, properties {soundWaveAssetPtr=&/Veh_MC_Sportbike_UpShift_02}}
 {id=WavePlayer_2, properties {soundWaveAssetPtr=/Motorcycle/Audio/SFX/Veh_MC_Wheelie_Start_02}}
 {id=WavePlayer_3, properties {soundWaveAssetPtr=/Motorcycle/Audio/SFX/Veh_MC_Wheelie_Start_03}}
 {id=WavePlayer_4, properties {soundWaveAssetPtr=/Motorcycle/Audio/SFX/Veh_MC_Wheelie_Start_04}}
 {id=WavePlayer_5, properties {soundWaveAssetPtr=/Motorcycle/Audio/SFX/Veh_MC_Wheelie_Start_05}}
 {id=WavePlayer_6, properties {soundWaveAssetPtr=&/Veh_MC_Sportbike_WheelyBoost}}
]
properties {
 attenuationSettings=/VehicleAudio/Settings/Vehicle_Movement_Boost_Att
 bOverrideConcurrency=true
 maxDistance=12000
 soundClassObject=&Fort_Vehicles_Exterior
 volumeMultiplier=1.2
 attenuationOverrides {bEnableReverbSend=false, distanceAlgorithm=Logarithmic, falloffDistance=1000, stereoSpread=100, attenuationShapeExtents {x=125}}
 concurrencyOverrides {resolutionRule=PreventNew}
}
