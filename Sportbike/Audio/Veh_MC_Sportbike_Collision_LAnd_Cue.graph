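# Sportbike landing collision sound cue that plays randomized heavy dirtbike suspension hits, pitched up in the Fort_Vehicles_Exterior class.
# A quality-level node gives the first two levels a ten-wave pool and the last two a four-wave pool, each with slight random pitch modulation.
# In each, a bIsLocal branch uses an attenuation node without settings when true, and close-detail vehicle movement attenuation when false or unset.
# A concurrency override allows one instance at a time.
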SoundCue /Sportbike/Audio/Veh_MC_Sportbike_Collision_LAnd_Cue
root=QualityLevel_0
nodes [
 {id=Attenuation_2, children [Modulator_1], properties {attenuationSettings=/VehicleAudio/Settings/Vehicle_Movement_Close_Detail_Att}}
 {id=Attenuation_3, children [Modulator_1]}
 {id=Attenuation_4, children [Modulator_2], properties {attenuationSettings=&Vehicle_Movement_Close_Detail_Att}}
 {id=Attenuation_5, children [Modulator_2]}
 {id=Branch_1, children [Attenuation_3, Attenuation_2, Attenuation_2], properties {boolParameterName=bIsLocal}}
 {id=Branch_2, children [Attenuation_5, Attenuation_4, Attenuation_4], properties {boolParameterName=bIsLocal}}
 {id=Modulator_1, children [Random_0], properties {pitchMax=1.1, pitchMin=1}}
 {id=Modulator_2, children [Random_2], properties {pitchMax=1.1, pitchMin=1}}
 {id=QualityLevel_0, children [Branch_1,Branch_1,Branch_2,Branch_2]}
 {
  id=Random_0
  children [
   WavePlayer_3
   WavePlayer_4
   WavePlayer_5
   WavePlayer_10
   WavePlayer_11
   WavePlayer_12
   WavePlayer_13
   WavePlayer_14
   WavePlayer_15
   WavePlayer_16
  ]
  properties {bShouldExcludeFromBranchCulling=true, weights [1,1,1,1,1,1,1,1,1,1]}
 }
 {
  id=Random_2
  children [WavePlayer_9, WavePlayer_17, WavePlayer_18, WavePlayer_6]
  properties {bShouldExcludeFromBranchCulling=true, weights [1,1,1,1]}
 }
 {id=WavePlayer_10, properties {soundWaveAssetPtr=/Motorcycle/Audio/SFX/Veh_MC_Dirtbike_Suspension_Hv_07}}
 {id=WavePlayer_11, properties {soundWaveAssetPtr=/Motorcycle/Audio/SFX/Veh_MC_Dirtbike_Suspension_Hv_06}}
 {id=WavePlayer_12, properties {soundWaveAssetPtr=/Motorcycle/Audio/SFX/Veh_MC_Dirtbike_Suspension_Hv_05}}
 {id=WavePlayer_13, properties {soundWaveAssetPtr=/Motorcycle/Audio/SFX/Veh_MC_Dirtbike_Suspension_Hv_04}}
 {id=WavePlayer_14, properties {soundWaveAssetPtr=/Motorcycle/Audio/SFX/Veh_MC_Dirtbike_Suspension_Hv_03}}
 {id=WavePlayer_15, properties {soundWaveAssetPtr=/Motorcycle/Audio/SFX/Veh_MC_Dirtbike_Suspension_Hv_02}}
 {id=WavePlayer_16, properties {soundWaveAssetPtr=/Motorcycle/Audio/SFX/Veh_MC_Dirtbike_Suspension_Hv_01}}
 {id=WavePlayer_17, properties {soundWaveAssetPtr=&Veh_MC_Dirtbike_Suspension_Hv_03}}
 {id=WavePlayer_18, properties {soundWaveAssetPtr=&Veh_MC_Dirtbike_Suspension_Hv_02}}
 {id=WavePlayer_3, properties {soundWaveAssetPtr=/Motorcycle/Audio/SFX/Veh_MC_Dirtbike_Suspension_Hv_10}}
 {id=WavePlayer_4, properties {soundWaveAssetPtr=/Motorcycle/Audio/SFX/Veh_MC_Dirtbike_Suspension_Hv_09}}
 {id=WavePlayer_5, properties {soundWaveAssetPtr=/Motorcycle/Audio/SFX/Veh_MC_Dirtbike_Suspension_Hv_08}}
 {id=WavePlayer_6, properties {soundWaveAssetPtr=&Veh_MC_Dirtbike_Suspension_Hv_01}}
 {id=WavePlayer_9, properties {soundWaveAssetPtr=&Veh_MC_Dirtbike_Suspension_Hv_04}}
]
properties {
 bOverrideConcurrency=true
 maxDistance=2097152
 pitchMultiplier=1.25
 soundClassObject=&Fort_Vehicles_Exterior
 volumeMultiplier=1.2
 concurrencyOverrides {maxCount=1}
}
