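# Sound cue for the sportbike's super slide, routed to the vehicle engines sound class.
# Mixes three layers, each through the shared gear-shift attenuation: a gear-shift one-shot, a random pick of five slide-start sounds at doubled pitch, and a looping sweetener.
# The loop is pitched slightly down at random and ramps up in volume through an Enveloper.
# Concurrency allows one instance per owner, stopping the oldest.
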SoundCue /Sportbike/Audio/Veh_MC_Sportbike_SuperSlide_Cue
root=Mixer_1
nodes [
 {id=Attenuation_1, children [Modulator_1], properties {attenuationSettings=/Sportbike/Audio/Settings/Vehicle_SportbikeShift_Close_Detail_Att}}
 {id=Attenuation_2, children [Modulator_2], properties {attenuationSettings=&Vehicle_SportbikeShift_Close_Detail_Att}}
 {id=Attenuation_4, children [Modulator_4], properties {attenuationSettings=&Vehicle_SportbikeShift_Close_Detail_Att}}
 {
  id=Enveloper_1
  children [WavePlayer_0]
  properties {pitchCurve {editorCurveData {keys [{time=0}]}}, volumeCurve {editorCurveData {keys [<at 0 -1>,<at 0.25 1>]}}}
 }
 {id=Mixer_1, children [Attenuation_1, Attenuation_4, Attenuation_2], properties {inputVolume [1,1,1]}}
 {id=Modulator_1, children [WavePlayer_15], properties {pitchMax=1, pitchMin=1, volumeMax=1, volumeMin=1}}
 {id=Modulator_2, children [Random_0], properties {pitchMax=2, pitchMin=2, volumeMax=1, volumeMin=1}}
 {id=Modulator_4, children [Enveloper_1], properties {pitchMax=0.9, pitchMin=0.8, volumeMax=1, volumeMin=1}}
 {id=Random_0, children [WavePlayer_2,WavePlayer_3,WavePlayer_4,WavePlayer_5,WavePlayer_7], properties {weights [1,1,1,1,1]}}
 {id=WavePlayer_0, properties {bLooping=true, soundWaveAssetPtr=&/Veh_MC_Sportbike_SuperSlide_Sweetner_Loop}}
 {id=WavePlayer_15, properties {soundWaveAssetPtr=&/Veh_MC_Sportbike_ShiftGear}}
 {id=WavePlayer_2, properties {soundWaveAssetPtr=&/Veh_MC_Sportbike_SuperSlide_Start_01}}
 {id=WavePlayer_3, properties {soundWaveAssetPtr=&/Veh_MC_Sportbike_SuperSlide_Start_02}}
 {id=WavePlayer_4, properties {soundWaveAssetPtr=&/Veh_MC_Sportbike_SuperSlide_Start_03}}
 {id=WavePlayer_5, properties {soundWaveAssetPtr=&/Veh_MC_Sportbike_SuperSlide_Start_04}}
 {id=WavePlayer_7, properties {soundWaveAssetPtr=&/Veh_MC_Sportbike_SuperSlide_Start_05}}
]
properties {
 attenuationSettings=/VehicleAudio/Settings/Vehicle_Movement_Close_Detail_Att
 bOverrideConcurrency=true
 maxDistance=12000
 soundClassObject=&Fort_Vehicles_Engine
 volumeMultiplier=2.2
 concurrencyOverrides {bLimitToOwner=true, maxCount=1, resolutionRule=StopOldest}
}
