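# Sound cue for the sportbike's heavy suspension sound, reusing dirtbike suspension waves in the exterior vehicle sound class.
# A quality switch gives the first two levels ten random variations and the last two only four, each with slight upward pitch modulation.
# Within each, a bIsLocal branch separates local playback from other playback, which uses close-detail movement attenuation.
# A shared collision/suspension concurrency set governs simultaneous playback.
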SoundCue /Sportbike/Audio/Veh_MC_Sportbike_Suspension_Hv_Cue
root=QualityLevel_0
nodes [
 {id=Attenuation_0, children [Modulator_0], properties {attenuationSettings=/VehicleAudio/Settings/Vehicle_Movement_Close_Detail_Att}}
 {id=Attenuation_1, children [Modulator_0]}
 {id=Attenuation_2, children [Modulator_1], properties {attenuationSettings=&Vehicle_Movement_Close_Detail_Att}}
 {id=Attenuation_3, children [Modulator_1]}
 {id=Branch_0, children [Attenuation_1, Attenuation_0, Attenuation_0], properties {boolParameterName=bIsLocal}}
 {id=Branch_1, children [Attenuation_3, Attenuation_2, Attenuation_2], properties {boolParameterName=bIsLocal}}
 {id=Modulator_0, children [Random_1], properties {pitchMax=1.1, pitchMin=1}}
 {id=Modulator_1, children [Random_0], properties {pitchMax=1.1, pitchMin=1}}
 {id=QualityLevel_0, children [Branch_0,Branch_0,Branch_1,Branch_1]}
 {
  id=Random_0
  children [WavePlayer_9, WavePlayer_17, WavePlayer_18, WavePlayer_19]
  properties {bShouldExcludeFromBranchCulling=true, weights [1,1,1,1]}
 }
 {
  id=Random_1
  children [
   WavePlayer_0
   WavePlayer_1
   WavePlayer_2
   WavePlayer_10
   WavePlayer_11
   WavePlayer_12
   WavePlayer_13
   WavePlayer_14
   WavePlayer_15
   WavePlayer_16
  ]
  properties {bShouldExcludeFromBranchCulling=true, weights [1,1,1,1,1,1,1,1,1,1]}
 }
 {id=WavePlayer_0, properties {soundWaveAssetPtr=/Motorcycle/Audio/SFX/Veh_MC_Dirtbike_Suspension_Hv_10}}
 {id=WavePlayer_1, properties {soundWaveAssetPtr=/Motorcycle/Audio/SFX/Veh_MC_Dirtbike_Suspension_Hv_09}}
 {id=WavePlayer_10, properties {soundWaveAssetPtr=/Motorcycle/Audio/SFX/Veh_MC_Dirtbike_Suspension_Hv_07}}
 {id=WavePlayer_11, properties {soundWaveAssetPtr=/Motorcycle/Audio/SFX/Veh_MC_Dirtbike_Suspension_Hv_06}}
 {id=WavePlayer_12, properties {soundWaveAssetPtr=/Motorcycle/Audio/SFX/Veh_MC_Dirtbike_Suspension_Hv_05}}
 {id=WavePlayer_13, properties {soundWaveAssetPtr=/Motorcycle/Audio/SFX/Veh_MC_Dirtbike_Suspension_Hv_04}}
 {id=WavePlayer_14, properties {soundWaveAssetPtr=/Motorcycle/Audio/SFX/Veh_MC_Dirtbike_Suspension_Hv_03}}
 {id=WavePlayer_15, properties {soundWaveAssetPtr=/Motorcycle/Audio/SFX/Veh_MC_Dirtbike_Suspension_Hv_02}}
 {id=WavePlayer_16, properties {soundWaveAssetPtr=/Motorcycle/Audio/SFX/Veh_MC_Dirtbike_Suspension_Hv_01}}
 {id=WavePlayer_17, properties {soundWaveAssetPtr=&Veh_MC_Dirtbike_Suspension_Hv_03}}
 {id=WavePlayer_18, properties {soundWaveAssetPtr=&Veh_MC_Dirtbike_Suspension_Hv_02}}
 {id=WavePlayer_19, properties {soundWaveAssetPtr=&Veh_MC_Dirtbike_Suspension_Hv_01}}
 {id=WavePlayer_2, properties {soundWaveAssetPtr=/Motorcycle/Audio/SFX/Veh_MC_Dirtbike_Suspension_Hv_08}}
 {id=WavePlayer_9, properties {soundWaveAssetPtr=&Veh_MC_Dirtbike_Suspension_Hv_04}}
]
properties {
 maxDistance=2097152
 soundClassObject=&Fort_Vehicles_Exterior
 volumeMultiplier=1
 concurrencyOverrides {bLimitToOwner=true, maxCount=1, resolutionRule=PreventNew, retriggerTime=1}
 concurrencySet [/VehicleAudio/Settings/Collision_Susp_Conc]
}
