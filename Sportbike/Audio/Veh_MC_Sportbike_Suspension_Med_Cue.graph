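# Sound cue for the sportbike's medium suspension sound, reusing dirtbike suspension recordings in the exterior vehicle sound class with a shared collision/suspension concurrency limit.
# A quality-level node picks a six-wave random pool for its first two levels and a three-wave pool for the other two.
# Each pool gets pitch modulation, then a bIsLocal branch uses an attenuation node without assigned settings when true, or close-detail vehicle attenuation when false or unset.
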SoundCue /Sportbike/Audio/Veh_MC_Sportbike_Suspension_Med_Cue
root=QualityLevel_2
nodes [
 {id=Attenuation_0, children [Modulator_0], properties {attenuationSettings=/VehicleAudio/Settings/Vehicle_Movement_Close_Detail_Att}}
 {id=Attenuation_1, children [Modulator_0]}
 {id=Attenuation_2, children [Modulator_1], properties {attenuationSettings=&Vehicle_Movement_Close_Detail_Att}}
 {id=Attenuation_3, children [Modulator_1]}
 {id=Branch_0, children [Attenuation_1, Attenuation_0, Attenuation_0], properties {boolParameterName=bIsLocal}}
 {id=Branch_1, children [Attenuation_3, Attenuation_2, Attenuation_2], properties {boolParameterName=bIsLocal}}
 {id=Modulator_0, children [Random_0], properties {pitchMax=1.1}}
 {id=Modulator_1, children [Random_1], properties {pitchMax=1.1}}
 {id=QualityLevel_2, children [Branch_0,Branch_0,Branch_1,Branch_1]}
 {
  id=Random_0
  children [WavePlayer_0,WavePlayer_1,WavePlayer_2,WavePlayer_3,WavePlayer_4,WavePlayer_5]
  properties {bShouldExcludeFromBranchCulling=true, weights [1,1,1,1,1,1]}
 }
 {id=Random_1, children [WavePlayer_9, WavePlayer_10, WavePlayer_11], properties {bShouldExcludeFromBranchCulling=true, weights [1,1,1]}}
 {id=WavePlayer_0, properties {soundWaveAssetPtr=/Motorcycle/Audio/SFX/Veh_MC_Dirtbike_Suspension_Med_06}}
 {id=WavePlayer_1, properties {soundWaveAssetPtr=/Motorcycle/Audio/SFX/Veh_MC_Dirtbike_Suspension_Med_05}}
 {id=WavePlayer_10, properties {soundWaveAssetPtr=/Motorcycle/Audio/SFX/Veh_MC_Dirtbike_Suspension_Med_02}}
 {id=WavePlayer_11, properties {soundWaveAssetPtr=/Motorcycle/Audio/SFX/Veh_MC_Dirtbike_Suspension_Med_01}}
 {id=WavePlayer_2, properties {soundWaveAssetPtr=/Motorcycle/Audio/SFX/Veh_MC_Dirtbike_Suspension_Med_04}}
 {id=WavePlayer_3, properties {soundWaveAssetPtr=/Motorcycle/Audio/SFX/Veh_MC_Dirtbike_Suspension_Med_03}}
 {id=WavePlayer_4, properties {soundWaveAssetPtr=&Veh_MC_Dirtbike_Suspension_Med_02}}
 {id=WavePlayer_5, properties {soundWaveAssetPtr=&Veh_MC_Dirtbike_Suspension_Med_01}}
 {id=WavePlayer_9, properties {soundWaveAssetPtr=&Veh_MC_Dirtbike_Suspension_Med_03}}
]
properties {
 maxDistance=2097152
 priority=0.9
 soundClassObject=&Fort_Vehicles_Exterior
 volumeMultiplier=0.5
 concurrencySet [/VehicleAudio/Settings/Collision_Susp_Conc]
}
