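# Sound cue for the Valet sportbike's boost, played as attenuated exterior vehicle audio with owner-limited concurrency.
# A mixer layers a looping jet whine, a looping sportbike boost, and non-looping start and sweetener sounds.
# FuelRemaining drives continuous modulators on both loops, dropping pitch and volume as fuel nears empty.
# The pitched-up whine also has a nested inverse pitch modulation that climbs as fuel drains, countered by the low-fuel drop.
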SoundCue /Sportbike/Audio/Veh_Valet_SportbikeBoost_Cue
root=Mixer_1
nodes [
 {id=Mixer_1, children [ModulatorContinuous_2, ModulatorContinuous_0, WavePlayer_6, WavePlayer_7], properties {inputVolume [1.5,0.7,1.7,0.55]}}
 {
  id=ModulatorContinuous_0
  children [WavePlayer_2]
  properties {
   pitchModulationParams {maxInput=0.4, minOutput=0.75, parameterName=FuelRemaining}
   volumeModulationParams {maxInput=0.4, minOutput=0.75, parameterName=FuelRemaining}
  }
 }
 {
  id=ModulatorContinuous_1
  children [Modulator_0]
  properties {pitchModulationParams {maxOutput=0.8, minOutput=1.2, parameterName=FuelRemaining}}
 }
 {
  id=ModulatorContinuous_2
  children [ModulatorContinuous_1]
  properties {
   pitchModulationParams {maxInput=0.2, minOutput=0.7, parameterName=FuelRemaining}
   volumeModulationParams {maxInput=0.2, minOutput=0.5, parameterName=FuelRemaining}
  }
 }
 {id=Modulator_0, children [WavePlayer_5], properties {pitchMax=1.2, pitchMin=1.2}}
 {id=WavePlayer_2, properties {bLooping=true, soundWaveAssetPtr=&/Veh_MC_Sportbike_Boost_Loop}}
 {id=WavePlayer_5, properties {bLooping=true, soundWaveAssetPtr=/VehicleAudio/Assets/Boost/Veh_Valet_Boost_Loop_Jet_Whine_01}}
 {id=WavePlayer_6, properties {soundWaveAssetPtr=&/Veh_MC_Sportbike_Boost_Start}}
 {id=WavePlayer_7, properties {soundWaveAssetPtr=&/Veh_MC_Sportbike_Boost_Sweetner}}
]
properties {
 attenuationSettings=/VehicleAudio/Settings/Vehicle_Movement_Boost_Att
 bOverrideConcurrency=true
 maxDistance=12000
 soundClassObject=&Fort_Vehicles_Exterior
 volumeMultiplier=0.7
 concurrencyOverrides {bLimitToOwner=true, maxCount=2, volumeScale=0.5}
 concurrencySet [/VehicleAudio/Settings/Vehicle_Engine_Conc]
}
