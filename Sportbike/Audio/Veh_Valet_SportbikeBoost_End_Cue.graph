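# Sound cue for the end of the sportbike vehicle's boost, playing a single boost-end wave through a modulator at reduced volume.
# It uses the shared vehicle boost movement attenuation and the exterior vehicle sound class.
# Overridden concurrency limits it to one instance per owner, stopping the oldest, in place of the referenced engine concurrency set.
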SoundCue /Sportbike/Audio/Veh_Valet_SportbikeBoost_End_Cue
root=Modulator_0
nodes [{id=Modulator_0, children [WavePlayer_0]}, {id=WavePlayer_0, properties {soundWaveAssetPtr=&/Veh_MC_Sportbike_Boost_End}}]
properties {
 attenuationSettings=/VehicleAudio/Settings/Vehicle_Movement_Boost_Att
 bOverrideConcurrency=true
 maxDistance=12000
 soundClassObject=&Fort_Vehicles_Exterior
 volumeMultiplier=0.5
 concurrencyOverrides {bLimitToOwner=true, maxCount=1, resolutionRule=StopOldest, volumeScale=0.5}
 concurrencySet [/VehicleAudio/Settings/Vehicle_Engine_Conc]
}
