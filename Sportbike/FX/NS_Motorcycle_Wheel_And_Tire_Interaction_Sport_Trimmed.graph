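# Niagara system for the sportbike's wheel, tire, exhaust and light effects, including peel-out dust and smoke, gear-shift flames and sparks, lens flares, neon trails and tire rings.
# User parameters for vehicle state (shifting, braking, boost jump, deep water, speed) gate spawning or scale the look, alongside system timelines.
# Emitters follow skeletal mesh sockets.
# The GPU flare scales by visibility, camera facing and time of day, and neon effects take the bike color.
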NiagaraSystem /Sportbike/FX/NS_Motorcycle_Wheel_And_Tire_Interaction_Sport_Trimmed
dependencies {
 usedDataInterfaces [
  &NiagaraDataInterfaceCurve
  &NiagaraDataInterfaceArrayInt32
  &NiagaraDataInterfaceArrayFloat3
  &NiagaraDataInterfaceCamera
  &NiagaraDataInterfaceVector2DCurve
  /Script/Niagara.NiagaraDataInterfaceOcclusion
  &NiagaraDataInterfaceArrayFloat
  &NiagaraDataInterfaceVectorCurve
 ]
 usedDynamicInputs [&FN_LOD_Distance_Remap, &InvertBool]
 usedModules [
  &self:SystemSpawnScript.SystemScriptSource.SystemScriptGraph.NiagaraNodeAssignment_1.SetVariables_#1
  &self:SystemSpawnScript.SystemScriptSource.SystemScriptGraph.NiagaraNodeAssignment_2.SetVariables_#2
  &self:SystemSpawnScript.SystemScriptSource.SystemScriptGraph.NiagaraNodeAssignment_3.SetVariables_#3
  &SystemState
  &self:SystemSpawnScript.SystemScriptSource.SystemScriptGraph.NiagaraNodeAssignment_0.SetVariables_#4
  &self:SetPeeloutFXArrayIndex
  &self:SparksSpawn
  &Timeline
 ]
}
emitters {
 Dust_Tires_Mesh {
  emitterUpdateScript [
   &EmitterState()
   &SpawnRate(
    Spawn Probability=&MakeCustomFloatFromBool(
     Bool=&Boolean_AndOperation(Bool1=$User.PeelOut_Wheel_Rear, Bool2=&Boolean_NotOperation(Bool=$User.TirePoppedSparksRear))
     True Float=1
     False Float=0
    )
    SpawnRate=10
    Use Spawn Probability=true
   )
  ]
  particleSpawnScript [
   &InitializeParticle(
    Color=$System.PeelOutColor
    Lifetime=0.5
    Mesh Scale=<1.2 0.6 1.2>
    Mesh Scale Mode=::NewEnumerator4
    UsePositionOffset=false
    Write Material Random=true
   )
   &InitialMeshOrientation(
    Mesh Orientation Mode=::NewEnumerator2
    Rotation=&MakeVector(X=0, Y=&RandomRangeFloat(Minimum=0, Maximum=1), Z=0)
    Use Rotation Vector=true
   )
   &CameraOffset(Camera Offset Amount=1)
  ]
  particleUpdateScript [
   &ParticleState()
   &SkeletalMeshLocation(
    Bone / Socket Index=3
    MaskSpawnGroup=false
    Mesh Sampling Type=::NewEnumerator5
    Sampled Position Offset=<0 1.5 0>
    Skeletal Mesh=$System.SkeletalMesh
    Socket Sampling Mode=::NewEnumerator5
   )
   &ScaleColor(
    Scale Alpha=&FloatFromCurve(FloatCurve=DataInterface() {propertyValues {Curve {keys [
         {interpMode=RCIM_Cubic, tangentMode=RCTM_User, time=0, value=0}
         {arriveTangent=1.25, interpMode=RCIM_Constant, time=0.1, value=1}
         {interpMode=RCIM_Cubic, tangentMode=RCTM_User, time=0.9, value=1}
         {tangentMode=RCTM_User, time=1, value=0}
        ]}}})
    Scale RGB=&VectorFromFloat(Value=1)
    ScaleA=true
    ScaleRGB=true
   )
  ]
  rendererClasses [&NiagaraMeshRendererProperties]
  renderers [{rendererClass=&NiagaraMeshRendererProperties, rendererIndex=0}]
 }
 ExhaustLight001 {
  emitterUpdateScript [
   &EmitterState(Life Cycle Mode=::NewEnumerator1, Loop Duration=0.1, UseLoopDelay=false)
   &SpawnBurst_Instantaneous(
    Spawn Count=&MakeCustomIntFromBool(Bool=$User.IsWaterDeep, True INT=0, False INT=1)
    Spawn Probability=&MakeCustomFloatFromBool(
     Bool=&Boolean_AndOperation(Bool1=$User.VehicleHasDriver, Bool2=&Boolean_AndOperation(Bool1=$User.VehicleHasFuel, Bool2=$User.IsShifting))
     True Float=1
     False Float=0
    )
    Use Spawn Probability=true
    UseLoopCountLimit=false
   )
   &SpawnRate(
    Spawn Probability=&MakeCustomFloatFromBool(
     Bool=&Boolean_AndOperation(Bool1=$User.VehicleHasDriver, Bool2=&Boolean_AndOperation(Bool1=$User.VehicleHasFuel, Bool2=$User.IsShifting))
     True Float=1
     False Float=0
    )
    SpawnGroup=1
    SpawnRate=&MakeFloatFromInt(Int=&MakeCustomIntFromBool(Bool=$User.IsWaterDeep, True INT=0, False INT=50))
    Use Spawn Probability=true
   )
   &SpawnRate(
    Spawn Probability=&MakeCustomFloatFromBool(
     Bool=&Boolean_AndOperation(Bool1=$User.VehicleHasDriver, Bool2=&Boolean_AndOperation(Bool1=$User.VehicleHasFuel, Bool2=$User.IsShifting))
     True Float=1
     False Float=0
    )
    SpawnGroup=2
    SpawnRate=&MakeFloatFromInt(Int=&MakeCustomIntFromBool(Bool=$User.IsWaterDeep, True INT=0, False INT=200))
    Use Spawn Probability=true
   ) {enabled=false, module=SpawnRate001}
  ]
  particleSpawnScript [
   /SetVariables(Particles.VisibilityTag=$Emitter.SpawnGroup) {isSetParameters=true}
   &InitializeParticle(
    Color=&SelectLinearColorFromArray(
     Array Sampling Mode=::NewEnumerator1
     Direct Array Index=$Emitter.SpawnGroup
     Color Selection Array=DataInterface() {propertyValues {ColorData [<rgba 2 0.4574 0 1>, <rgba 1 0.3093 0 1>, <rgba 2 0.4574 0 1>]}}
     Random Seed=?
    )
    Lifetime=&SelectFloatFromArray(Direct Array Index=$Emitter.SpawnGroup, Float Selection Array=DataInterface() {propertyValues {FloatData [0.15,0.15,1]}})
    Sprite Size Mode=::NewEnumerator3
    Uniform Sprite Size=12
    UsePositionOffset=false
    Write Material Random=true
   )
   &SkeletalMeshLocation(
    Bone / Socket Index=4
    Interpolate Mesh Sampling=false
    MaskSpawnGroup=false
    Mesh Sampling Type=::NewEnumerator5
    Sampled Position Offset Coordinate Space:ENiagaraSkelMeshTransforms=::NewEnumerator4
    Skeletal Mesh=$System.SkeletalMesh
    Socket Sampling Mode=::NewEnumerator5
   )
   &InitialMeshOrientation(Mesh Orientation Mode=::NewEnumerator2, Orientation Vector=<-1 0 0>, Use Rotation Vector=true)
   &AddVelocityInCone(
    Cone Angle=5
    Cone Axis=<1 0 0.2>
    Module.Use Velocity Falloff On Cone Axis=true
    Velocity Distribution Along Cone Axis=0.6
    Velocity Falloff Away From Cone Axis=0.333
    Velocity Strength=&Multiply_Float(
     A=&SelectFloatFromArray(Direct Array Index=$Emitter.SpawnGroup, Float Selection Array=DataInterface() {propertyValues {FloatData [0,0,1]}})
     B=&Lerp_Float(
      A=1200
      B=2200
      Alpha=&FloatFromCurve(
       FloatCurve=DataInterface() {propertyValues {Curve {keys [
           {interpMode=RCIM_Cubic, tangentMode=RCTM_User, time=0, value=0}
           {arriveTangent=0.9384, interpMode=RCIM_Cubic, leaveTangent=0.9384, tangentMode=RCTM_User, time=0.8, value=0.35}
           {arriveTangent=4.6474, interpMode=RCIM_Cubic, leaveTangent=4.6474, tangentMode=RCTM_User, time=1, value=1}
          ]}}}
       CurveIndex=&RandomRangeFloat(Minimum=0, Maximum=1)
      )
     )
    )
   )
   /Niagara/Modules/Spawn/Location/SphereLocation(
    MaskSpawnGroup=false
    Module.Override Randomness=false
    Sphere Radius=&SelectFloatFromArray(Direct Array Index=$Emitter.SpawnGroup, Float Selection Array=DataInterface() {propertyValues {FloatData [0,0,2]}})
    Surface Only=false
   )
  ]
  particleUpdateScript [
   &ParticleState()
   &FN_Light_Attributes(
    Color=<rgba 2 0.4574 0 1>
    Create Color Attribute=true
    Create Exponent Attribute=true
    Create Position Attribute=true
    Create Radius Attribute=true
    Exponent=6
    Position=&ConvertPositionToVector(Input Position=$Particles.Position, Convert To Absolute Large World Vector=false)
    Radius=120
    Scale Color=&Multiply_Float(
     A=&Multiply_Float(A=2, B=&Waveform("Waveform [1]"=::NewEnumerator0, "[1] Frequency"=2, "[1] Amplitude Min/Max"=<0.95 1>))
     B=&Lerp_Float(
      A=0.25
      B=1
      Alpha=&FloatFromCurve(
       FloatCurve=DataInterface() {propertyValues {Curve {keys [
           {interpMode=RCIM_Cubic, tangentMode=RCTM_User, time=0, value=1}
           {arriveTangentWeight=1.1078, interpMode=RCIM_Cubic, tangentMode=RCTM_User, time=1, value=0}
          ]}}}
       CurveIndex=$Emitter.NormalizedLoopAge
      )
     )
    )
   )
   &ScaleColor(
    Scale Alpha=&FloatFromCurve(
     FloatCurve=DataInterface() {propertyValues {Curve {keys [
         {interpMode=RCIM_Cubic, tangentMode=RCTM_User, time=0, value=0}
         {interpMode=RCIM_Cubic, tangentMode=RCTM_User, time=0.25, value=1}
         {interpMode=RCIM_Cubic, tangentMode=RCTM_User, time=1, value=0}
        ]}}}
     Scale Curve=0.15
    )
    Scale RGB=&VectorFromFloat(Value=&Lerp_Float(
      A=0.5
      B=2
      Alpha=&FloatFromCurve(FloatCurve=DataInterface() {propertyValues {Curve {keys [
           {interpMode=RCIM_Cubic, tangentMode=RCTM_User, time=0, value=0}
           {interpMode=RCIM_Cubic, tangentMode=RCTM_User, time=0.25, value=1}
           {interpMode=RCIM_Cubic, tangentMode=RCTM_User, time=1, value=0}
          ]}}})
     ))
    ScaleA=true
    ScaleRGB=true
   )
   /Niagara/Modules/Update/Forces/CurlNoiseForce(
    Noise Frequency=100
    Noise Strength=&Multiply_Float(
     A=&Lerp_Float(
      A=1500
      B=6500
      Alpha=&FloatFromCurve(
       FloatCurve=DataInterface() {propertyValues {Curve {keys [
           {interpMode=RCIM_Cubic, tangentMode=RCTM_User, time=0, value=0}
           {arriveTangent=0.9384, interpMode=RCIM_Cubic, leaveTangent=0.9384, tangentMode=RCTM_User, time=0.8, value=0.35}
           {arriveTangent=4.6474, interpMode=RCIM_Cubic, leaveTangent=4.6474, tangentMode=RCTM_User, time=1, value=1}
          ]}}}
       CurveIndex=&RandomRangeFloat(Minimum=0, Maximum=1)
      )
     )
     B=&SelectFloatFromArray(Direct Array Index=$Emitter.SpawnGroup, Float Selection Array=DataInterface() {propertyValues {FloatData [0,0,1]}})
    )
    Pan Noise=false
    Random Seed=1
   )
   &Drag(Drag=5, Use Linear Drag=true, Use Rotational Drag=false)
   &SolveForcesAndVelocity(Clamp Velocity=false, Limit Acceleration=false)
   /Game/Effects/Niagara/Modules/Materials/MaterialPanner(
    Jitter Pan by SpawnGroups=false
    Offset:NiagaraFloat=&Add_Float(A=$Emitter.Age, B=$Particles.Initial.MaterialRandom)
    Speed=-5
   )
   /SetVariables(
    Particles.Second Color:LinearColor=&MultiplyLinearColor(
     Linear Color=<rgba 0 0.1916 1 1>
     Scale Factor=&MakeFloat4FromVectorAndFloat(
      "Vector (RGB)"=&VectorFromFloat(Value=&Multiply_Float(
        A=&RandomRangeFloat(Minimum=3, Maximum=20)
        B=&FloatFromCurve(
         FloatCurve=DataInterface() {propertyValues {Curve {keys [
             {interpMode=RCIM_Cubic, tangentMode=RCTM_User, time=0.3, value=1}
             {arriveTangentWeight=0.6424, interpMode=RCIM_Cubic, tangentMode=RCTM_User, time=1, value=0.25}
            ]}}}
         CurveIndex=$Emitter.NormalizedLoopAge
         Scale Curve=3
        )
       ))
      "Float (Alpha)"=&FloatFromCurve(FloatCurve=DataInterface() {propertyValues {Curve {keys [
           {interpMode=RCIM_Cubic, leaveTangent=9.8521, tangentMode=RCTM_Break, time=0, value=0.1}
           {interpMode=RCIM_Cubic, tangentMode=RCTM_User, time=0.2, value=1}
           {interpMode=RCIM_Cubic, tangentMode=RCTM_Break, time=1, value=0}
          ]}}})
     )
    )
    Particles.Second Mesh Scale:Vector3f=&Multiply_Vector(
     A=&Lerp_Vector(
      A=<0.18 0.16 0.16>
      B=<0.22 0.28 0.28>
      Alpha=&FloatFromCurve(
       FloatCurve=DataInterface() {propertyValues {Curve {keys [
           {interpMode=RCIM_Cubic, tangentMode=RCTM_Break, time=0, value=0}
           {arriveTangent=1.393, interpMode=RCIM_Cubic, tangentMode=RCTM_Break, time=1, value=1}
          ]}}}
       CurveIndex=&RandomRangeFloat(Minimum=0, Maximum=1)
      )
     )
     B=&VectorFromCurve(
      VectorCurve=DataInterface() {propertyValues {
        bOptimizeLUT=false
        XCurve {keys [{arriveTangent=-0.4338, interpMode=RCIM_Cubic, leaveTangent=-0.4338, tangentMode=RCTM_User, time=0, value=0.75}, <at 1 2>]}
        YCurve {keys [{arriveTangent=0.1422, interpMode=RCIM_Cubic, leaveTangent=0.1422, tangentMode=RCTM_User, time=0, value=0.75}, <at 1 2>]}
        ZCurve {keys [
          <at -0.0011 1.0067 RCIM_Cubic>
          {arriveTangent=-0.105, interpMode=RCIM_Cubic, leaveTangent=-0.105, tangentMode=RCTM_User, time=1.0016, value=1}
         ]}
       }}
      Scale Curve=&VectorFromFloat(Value=&Lerp_Float(
        A=1.3
        B=2
        Alpha=&FloatFromCurve(
         FloatCurve=DataInterface() {propertyValues {Curve {keys [
             {interpMode=RCIM_Cubic, tangentMode=RCTM_User, time=0, value=1}
             {arriveTangentWeight=1.1078, interpMode=RCIM_Cubic, tangentMode=RCTM_User, time=1, value=0}
            ]}}}
         CurveIndex=$Emitter.NormalizedLoopAge
        )
       ))
     )
    )
   ) {isSetParameters=true}
   &DynamicMaterialParameters(
    Index 1 Param 1=150
    Module.Index1Float0Write=true
    Module.Index1Float1Write=true
    Module.Index1Float2Write=true
    Module.Index1Float3Write=true
    Write Parameter Index 0=::NewEnumerator0
    Write Parameter Index 1=::NewEnumerator1
   )
   &ScaleColor(
    Scale RGB=&VectorFromFloat(Value=&SelectFloatFromArray(Direct Array Index=$Emitter.SpawnGroup, Float Selection Array=DataInterface() {propertyValues {FloatData [1,1,10]}}))
    ScaleA=true
    ScaleRGB=true
   ) {module=ScaleColor001}
   &ScaleSpriteSize(Scale Factor=&MaskVector2dBySpawnGroup(
     Masked Vector2D=&Vector2DFromCurve(
      Vector2Curve=DataInterface() {propertyValues {
        XCurve {keys [<at -0.0017 0.4556 RCIM_Cubic>, {arriveTangent=-0.9208, interpMode=RCIM_Cubic, tangentMode=RCTM_Break, time=0.1932, value=0.1104}]}
        YCurve {keys [<at 0 1 RCIM_Cubic>,<at 0.6807 0.0224 RCIM_Cubic>]}
       }}
      ScaleCurve=&MakeVector2D(X=1, Y=1)
     )
     Passthrough Vector2D=<1 1>
     Spawn Group Mask=2
    )) {enabled=false}
  ]
  rendererClasses [
   &NiagaraLightRendererProperties
   &NiagaraSpriteRendererProperties
   &NiagaraMeshRendererProperties
  ]
  renderers [
   {rendererClass=&NiagaraLightRendererProperties, rendererIndex=0}
   {rendererClass=&NiagaraSpriteRendererProperties, rendererIndex=1}
   {rendererClass=&NiagaraMeshRendererProperties, rendererIndex=2}
   {rendererClass=&NiagaraMeshRendererProperties, rendererIndex=3}
  ]
 }
 Exhaust_Local_New {
  emitterUpdateScript [
   &EmitterState()
   &SpawnRate(
    Spawn Probability=&MakeCustomFloatFromBool(Bool=&Boolean_AndOperation(Bool1=$User.VehicleHasDriver, Bool2=$User.VehicleHasFuel), True Float=1, False Float=0)
    SpawnRate=&Multiply_Float(
     A=&Lerp_Float(A=10, B=90, Alpha=$User.ChargeProgress)
     B=&MakeCustomFloatFromBool(Bool=$User.IsWaterDeep, True Float=0, False Float=&Multiply_Float(A=$System.Timeline001.Float, B=1))
    )
    Use Spawn Probability=true
   )
  ]
  particleSpawnScript [
   /SetVariables(Particles.RandomFloat=&UniformAOrBFloat(A=-1)) {isSetParameters=true, module=SetVariables_#5}
   /SetVariables(Particles.InitSpriteRotRate=&RandomRangeFloat(Minimum=-2, Maximum=2)) {isSetParameters=true, module=SetVariables_#6}
   /Niagara/Modules/Spawn/Initialization/InitializeParticle(
    Color=<rgba 0.4236 0.4777 0.5469 0.025>
    Lifetime=&Lerp_Float(
     A=&Lerp_Float(A=&UniformRangedFloat(Minimum=0.225, Maximum=0.8), B=1, Alpha=$User.Vehicle_Speed)
     B=0.2
     Alpha=$User.ChargeProgress
    )
    Module.Write Color=true
    Module.Write Lifetime=true
    Module.Write Mass=false
    Module.Write Material Random=true
    Module.Write Position=true
    Module.Write Scale=false
    Module.Write SpriteRotation=true
    Module.Write SpriteSize=true
    Position=$Engine.Owner.Position
    Sprite Rotation=&UniformRangedFloat(Minimum=0, Maximum=360)
    Sprite Size=&Vector2DFromFloat(Value=&Lerp_Float(
      A=&UniformRangedFloat(Minimum=40, Maximum=60)
      B=&UniformRangedFloat(Minimum=50, Maximum=65)
      Alpha=$User.Vehicle_Speed
     ))
   ) {module=InitializeParticle001}
  ]
  particleUpdateScript [
   &ParticleState()
   &SkeletalMeshLocation(
    Bone / Socket Index=4
    MaskSpawnGroup=false
    Mesh Sampling Type=::NewEnumerator5
    Skeletal Mesh=$System.SkeletalMesh
    Socket Sampling Mode=::NewEnumerator5
   ) {module=SkeletalMeshLocation001}
   &ScaleColor(
    Scale Alpha=&FloatFromCurve(
     FloatCurve=DataInterface() {propertyValues {Curve {keys [{interpMode=RCIM_Cubic, tangentMode=RCTM_User, time=0, value=1}, <at 1 0>]}}}
     Scale Curve=&Lerp_Float(
      A=8
      B=15
      Alpha=&Add_Float(A=$User.Vehicle_Speed, B=&Multiply_Float(A=$User.ChargeProgress, B=0.5))
     )
    )
    Scale RGB=&VectorFromFloat(Value=&FloatFromCurve(FloatCurve=DataInterface() {propertyValues {Curve {keys [{arriveTangent=1.8145, interpMode=RCIM_Cubic, leaveTangent=4.5795, tangentMode=RCTM_Break, time=0, value=0.4}, <at 0.3387 0.9565>, <at 1 1>]}}}))
    ScaleA=true
    ScaleRGB=true
   )
   &ScaleSpriteSize(Scale Factor=&Vector2DFromFloat(Value=&FloatFromCurve(FloatCurve=DataInterface() {propertyValues {Curve {keys [{interpMode=RCIM_Cubic, tangentMode=RCTM_Break, time=0, value=0.179}, <at 0.8329 1>]}}})))
   &SpriteRotationRate(Rotation Rate=&Multiply_Float(A=$Particles.Initial.MaterialRandom, B=1))
   &DynamicMaterialParameters(
    Index 0 Param 1=&Add_Float(
     A=$Particles.RandomFloat
     B=&FloatFromCurve(FloatCurve=DataInterface() {propertyValues {Curve {keys [
          {arriveTangent=2.7533, interpMode=RCIM_Cubic, leaveTangent=2.847, tangentMode=RCTM_Break, time=0, value=0}
          {arriveTangent=1.3581, interpMode=RCIM_Cubic, leaveTangent=1.3581, tangentMode=RCTM_User, time=0.2558, value=0.5125}
          {arriveTangent=0.0529, interpMode=RCIM_Cubic, leaveTangent=0.2021, tangentMode=RCTM_User, time=1, value=1}
         ]}}})
    )
    Index 0 Param 3=&FloatFromCurve(FloatCurve=DataInterface() {propertyValues {Curve {keys [{arriveTangent=-53.9199, time=0, value=0}, {leaveTangent=0.2723, time=0.25, value=0}]}}})
    Index 0 Param 4=&FloatFromCurve(FloatCurve=DataInterface() {propertyValues {Curve {keys [<at 0 1>,<at 1 0>]}}})
    Index 1 Param 1=&FloatFromCurve(FloatCurve=DataInterface() {propertyValues {Curve {keys [<at 0 0>, {arriveTangent=4.2257, interpMode=RCIM_Cubic, leaveTangent=4.2257, tangentMode=RCTM_User, time=0.25, value=0}]}}})
    Index 1 Param 2=&FloatFromCurve(FloatCurve=DataInterface() {propertyValues {Curve {keys [<at 0 0>]}}})
    Index 1 Param 3=&FloatFromCurve(FloatCurve=DataInterface() {propertyValues {Curve {keys [{interpMode=RCIM_Cubic, leaveTangent=6.3917, tangentMode=RCTM_Break, time=0, value=0}]}}})
    Index 1 Param 4=&FloatFromCurve(FloatCurve=DataInterface() {propertyValues {Curve {keys [{arriveTangent=-0.136, leaveTangent=-0.136, time=0, value=1}]}}})
    Index 2 Param 1=&FloatFromCurve(FloatCurve=DataInterface() {propertyValues {Curve {keys [
         {arriveTangent=2.646, interpMode=RCIM_Cubic, leaveTangent=2.646, tangentMode=RCTM_User, time=0, value=0}
         {leaveTangent=-0.0719, time=1, value=1}
        ]}}})
    Index 2 Param 2=&FloatFromCurve(
     FloatCurve=DataInterface() {propertyValues {Curve {keys [
         {arriveTangent=1.9056, time=0, value=0.2}
         {arriveTangent=0.0259, interpMode=RCIM_Cubic, leaveTangent=0.0259, tangentMode=RCTM_User, time=1, value=1.4}
        ]}}}
     Scale Curve=&RandomRangeFloat(Minimum=0.5, Maximum=1)
    )
    Index 2 Param 3=&FloatFromCurve(FloatCurve=DataInterface() {propertyValues {Curve {keys [
         {interpMode=RCIM_Cubic, leaveTangent=2.3064, tangentMode=RCTM_Break, time=0, value=1}
         {interpMode=RCIM_Cubic, leaveTangent=11.1233, tangentMode=RCTM_Break, time=1, value=2}
        ]}}})
    Index 2 Param 4=&FloatFromCurve(
     FloatCurve=DataInterface() {propertyValues {Curve {keys [
         {arriveTangent=-0.6867, interpMode=RCIM_Cubic, leaveTangent=-0.6867, tangentMode=RCTM_User, time=0, value=1}
         {interpMode=RCIM_Cubic, leaveTangent=-1.2956, tangentMode=RCTM_Break, time=1, value=0.75}
        ]}}}
     Scale Curve=1.6
    )
    Module.Param0WriteEnabled=true
    Module.Param0WriteEnabled1=true
    Module.Param0WriteEnabled2=true
    Module.Param1WriteEnabled=false
    Module.Param1WriteEnabled1=true
    Module.Param1WriteEnabled2=true
    Module.Param2WriteEnabled=true
    Module.Param2WriteEnabled1=true
    Module.Param2WriteEnabled2=true
    Module.Param3WriteEnabled=true
    Module.Param3WriteEnabled1=true
    Module.Param3WriteEnabled2=true
    Write Parameter Index 1=true
    Write Parameter Index 2=true
   )
  ]
  rendererClasses [&NiagaraSpriteRendererProperties]
  renderers [
   {rendererClass=&NiagaraSpriteRendererProperties}
   {rendererClass=&NiagaraSpriteRendererProperties, rendererIndex=1}
  ]
 }
 Jet_BurstSparksHigh {
  emitterUpdateScript [
   &EmitterState(Life Cycle Mode=::NewEnumerator1, Loop Duration=0.1, UseLoopDelay=false)
   &SpawnRate(
    Spawn Probability=&MakeCustomFloatFromBool(
     Bool=&Boolean_AndOperation(Bool1=$User.VehicleHasDriver, Bool2=&Boolean_AndOperation(Bool1=$User.VehicleHasFuel, Bool2=$User.IsShifting))
     True Float=1
     False Float=0
    )
    SpawnRate=&MakeFloatFromInt(Int=&MakeCustomIntFromBool(Bool=$User.IsWaterDeep, True INT=0, False INT=200))
    Use Spawn Probability=true
   )
  ]
  particleSpawnScript [
   &InitializeParticle(
    Color=&RandomRangeLinearColor(Minimum=<rgba 2 0.4574 0 1>, Maximum=<rgba 5 1.5362 0 1>)
    Lifetime=&Lerp_Float(
     A=0.15
     B=0.4
     Alpha=&FloatFromCurve(
      FloatCurve=DataInterface() {propertyValues {Curve {keys [
          {interpMode=RCIM_Cubic, tangentMode=RCTM_User, time=0, value=0}
          {arriveTangent=0.9384, interpMode=RCIM_Cubic, leaveTangent=0.9384, tangentMode=RCTM_User, time=0.85, value=0.25}
          {arriveTangent=8.8331, interpMode=RCIM_Cubic, leaveTangent=8.8331, tangentMode=RCTM_User, time=1, value=1}
         ]}}}
      CurveIndex=&RandomRangeFloat(Minimum=0, Maximum=1)
     )
    )
    Module.Write Color=true
    Module.Write Lifetime=true
    Module.Write Mass=false
    Module.Write Material Random=true
    Module.Write Position=false
    Module.Write Scale=false
    Module.Write SpriteRotation=true
    Module.Write SpriteSize=true
    Position=$Engine.Owner.Position
    Sprite Size=&Multiply_Vector2D_ByFloat(
     Float=&Lerp_Float(
      A=1
      B=3
      Alpha=&FloatFromCurve(
       FloatCurve=DataInterface() {propertyValues {Curve {keys [
           {interpMode=RCIM_Cubic, tangentMode=RCTM_User, time=0, value=0}
           {arriveTangent=0.9384, interpMode=RCIM_Cubic, leaveTangent=0.9384, tangentMode=RCTM_User, time=0.85, value=0.25}
           {arriveTangent=8.8331, interpMode=RCIM_Cubic, leaveTangent=8.8331, tangentMode=RCTM_User, time=1, value=1}
          ]}}}
       CurveIndex=&RandomRangeFloat(Minimum=0, Maximum=1)
      )
     )
     Vector2D=&UniformRangedVector2D(Minimum=<2 7>, Maximum=<4 8>)
    )
   )
   &SkeletalMeshLocation(
    Bone / Socket Index=4
    Interpolate Mesh Sampling=false
    MaskSpawnGroup=false
    Mesh Sampling Type=::NewEnumerator5
    Sampled Position Offset Coordinate Space:ENiagaraSkelMeshTransforms=::NewEnumerator4
    Skeletal Mesh=$System.SkeletalMesh
    Socket Sampling Mode=::NewEnumerator5
   )
   &AddVelocityInCone(
    Cone Angle=5
    Cone Axis=<1 0 0.2>
    Module.Use Velocity Falloff On Cone Axis=true
    Velocity Distribution Along Cone Axis=0.6
    Velocity Falloff Away From Cone Axis=0.333
    Velocity Strength=&Lerp_Float(
     A=1200
     B=2200
     Alpha=&FloatFromCurve(
      FloatCurve=DataInterface() {propertyValues {Curve {keys [
          {interpMode=RCIM_Cubic, tangentMode=RCTM_User, time=0, value=0}
          {arriveTangent=0.9384, interpMode=RCIM_Cubic, leaveTangent=0.9384, tangentMode=RCTM_User, time=0.8, value=0.35}
          {arriveTangent=4.6474, interpMode=RCIM_Cubic, leaveTangent=4.6474, tangentMode=RCTM_User, time=1, value=1}
         ]}}}
      CurveIndex=&RandomRangeFloat(Minimum=0, Maximum=1)
     )
    )
   )
   &SphereLocation(MaskSpawnGroup=false, Module.Override Randomness=false, Sphere Radius=2, Surface Only=false)
  ]
  particleUpdateScript [
   &ParticleState()
   &CurlNoiseForce(
    Noise Frequency=100
    Noise Strength=&Lerp_Float(
     A=1500
     B=6500
     Alpha=&FloatFromCurve(
      FloatCurve=DataInterface() {propertyValues {Curve {keys [
          {interpMode=RCIM_Cubic, tangentMode=RCTM_User, time=0, value=0}
          {arriveTangent=0.9384, interpMode=RCIM_Cubic, leaveTangent=0.9384, tangentMode=RCTM_User, time=0.8, value=0.35}
          {arriveTangent=4.6474, interpMode=RCIM_Cubic, leaveTangent=4.6474, tangentMode=RCTM_User, time=1, value=1}
         ]}}}
      CurveIndex=&RandomRangeFloat(Minimum=0, Maximum=1)
     )
    )
    Pan Noise=false
    Random Seed=1
   )
   &Drag(Drag=5, Use Linear Drag=true, Use Rotational Drag=false)
   &SolveForcesAndVelocity(Clamp Velocity=false, Limit Acceleration=false)
   &ScaleSpriteSize(Scale Factor=&Vector2DFromCurve(
     Vector2Curve=DataInterface() {propertyValues {
       XCurve {keys [<at -0.0017 0.4556 RCIM_Cubic>, {arriveTangent=-0.9208, interpMode=RCIM_Cubic, tangentMode=RCTM_Break, time=0.1932, value=0.1104}]}
       YCurve {keys [<at 0 1 RCIM_Cubic>,<at 0.6807 0.0224 RCIM_Cubic>]}
      }}
     ScaleCurve=&MakeVector2D(X=1, Y=1)
    ))
   &ScaleColor(Scale RGB=&VectorFromFloat(Value=10), ScaleA=true, ScaleRGB=true)
  ]
  rendererClasses [&NiagaraSpriteRendererProperties]
  renderers [{rendererClass=&NiagaraSpriteRendererProperties}]
 }
 LensFlare {
  simTarget=GPUComputeSim
  emitterUpdateScript [
   &EmitterState(Life Cycle Mode=::NewEnumerator1, Loop Duration=1, UseLoopDelay=false)
   &SpawnBurst_Instantaneous(Spawn Count=3, Use Spawn Probability=false, UseLoopCountLimit=true)
  ]
  particleSpawnScript [
   &PartitionParticles(Number Of Partitions=3)
   /Niagara/Modules/Spawn/Initialization/V2/InitializeParticle(
    Sprite Rotation Mode=::NewEnumerator2
    Sprite Size Mode=::NewEnumerator3
    Uniform Sprite Size=75
    UsePositionOffset=false
    Write Material Random=true
   )
   /SetVariables(
    Particles.Streak_01_Offset:NiagaraPosition=&AddVectorToPosition(Position=$Particles.Position, Vector=<0 0 0>)
    Particles.Streak_01_Size:Vector2f=<325 10>
   ) {isSetParameters=true}
  ]
  particleUpdateScript [
   &ParticleState(Kill Particles When Lifetime Has Elapsed=false, Let Infinitely Lived Particles Die When Emitter Deactivates=true)
   &SkeletalMeshLocation(
    Bone / Socket Index=&Add_Integer(A=$Particles.PartitionParticles.Partition, B=5)
    MaskSpawnGroup=false
    Mesh Sampling Type=::NewEnumerator5
    Sampled Position Offset=&SelectVectorFromArray(
     Direct Array Index=$Particles.PartitionParticles.Partition
     Vector Selection Array=DataInterface() {propertyValues {FloatData [<0 0 0>, <0 0 0>, <-2 0 1>]}}
    )
    Skeletal Mesh=$System.SkeletalMesh
    Socket Sampling Mode=::NewEnumerator5
   )
   &self:ScratchModule_03(
    NumberOfRings=5
    OcclusionSize=50
    Position:NiagaraPosition=&AddVectorToPosition(Position=$Engine.Owner.Position, Vector=<0 0 100>)
    SamplesPerRing=5
   ) {module=ScratchModule_03}
   /SetVariables(Particles.CameraScale:NiagaraFloat=&Multiply_Float(
     A=1
     B=&MakeFloatFromVector(
      Vector=&ReturnCameraVector(Camera Vectors=::NewEnumerator4, Coordinate Space=::Local, Origin Position=$Particles.Position)
      Channel=::NewEnumerator0
     )
    )) {isSetParameters=true}
   &ScaleSpriteSize(
    Uniform Curve Index=$Particles.CameraScale
    Uniform Curve Scale=&Lerp_Float(A=0.35, B=1, Alpha=$User.IsBraking)
    Uniform Curve Sprite Scale=DataInterface() {propertyValues {Curve {keys [
        {tangentMode=RCTM_User, time=0.3, value=0}
        {arriveTangent=2.1429, interpMode=RCIM_Cubic, leaveTangent=2.1429, time=0.7, value=0.55}
        {tangentMode=RCTM_User, time=1, value=1.5}
       ]}}}
   )
   &ScaleColor(
    Scale Alpha=$Particles.Visibility
    Scale RGB=&MakeVectorFromLinearColorRGB(Color=&LerpLinearColors(StartColor=$User.BikeColor, EndColor=<rgba 1 0 0 1>, LerpFactor=$User.IsBraking))
    ScaleA=true
    ScaleRGB=true
   ) {module=ScaleColor002}
   &ScaleColor(
    Scale Alpha=&FloatFromCurve(
     FloatCurve=DataInterface() {propertyValues {Curve {keys [
         {tangentMode=RCTM_User, time=0, value=0}
         {arriveTangent=0.4636, interpMode=RCIM_Cubic, leaveTangent=0.4636, tangentMode=RCTM_User, time=0.3, value=0}
         {arriveTangent=4.5091, interpMode=RCIM_Cubic, leaveTangent=4.5091, tangentMode=RCTM_User, time=1, value=1}
        ]}}}
     CurveIndex=$Particles.CameraScale
    )
    Scale RGB=&Multiply_VectorByFloat(Vector=<1 1 1>, Float=$Particles.CameraScale)
    ScaleA=true
    ScaleRGB=true
   )
   /SetVariables(Particles.SpriteRotation:NiagaraFloat=&FloatFromCurve(
     FloatCurve=DataInterface() {propertyValues {Curve {keys [<at 0 -1>,<at 0.5 1>,<at 1 -1>]}}}
     CurveIndex=$Particles.CameraScale
     Scale Curve=30
    )) {isSetParameters=true}
   /SetVariables(Particles.Streak_01_Size:Vector2f=&Multiply_Vector2D_ByFloat(
     Float=&FloatFromCurve(
      FloatCurve=DataInterface() {propertyValues {Curve {keys [
          {tangentMode=RCTM_User, time=0.3, value=0}
          {arriveTangent=2.1429, interpMode=RCIM_Cubic, leaveTangent=2.1429, time=0.7, value=0.55}
          {tangentMode=RCTM_User, time=1, value=1.5}
         ]}}}
      CurveIndex=$Particles.CameraScale
      Scale Curve=&Lerp_Float(A=0.35, B=1, Alpha=$User.IsBraking)
     )
     Vector2D=$Particles.Initial.Streak_01_Size
    )) {isSetParameters=true}
   &ScaleColor(
    Scale Alpha=&FN_Select_Float_from_Time_Of_Day(Morning=0.05, Day=0.05, Evening=1, Night=1)
    ScaleA=true
    ScaleRGB=false
   ) {module=ScaleColor001}
  ]
  rendererClasses [&NiagaraSpriteRendererProperties]
  renderers [
   {rendererClass=&NiagaraSpriteRendererProperties}
   {rendererClass=&NiagaraSpriteRendererProperties, rendererIndex=1}
  ]
 }
 NeonAnimTrail_Thin {
  emitterUpdateScript [
   &EmitterState()
   /Game/Effects/Niagara/Modules/Spawn/FN_SpawnByEngineDistance(Multiplier=3, Step=20)
  ]
  particleSpawnScript [
   &InitializeParticle(
    Color=&MultiplyLinearColor(Linear Color=<rgba 1 1 1 1>, Scale Factor=&MakeVector4_FromLinearColor(LinearColor=<rgba 2 2 2 0.25>))
    Lifetime=&Lerp_Float(
     A=0.06
     B=0.08
     Alpha=&FNTimeOfDayScalar(Night Scale=1, Morning Scale=0, Day Scale=0, Evening Scale=1)
    )
    Ribbon Width=2.75
    Ribbon Width Mode=::NewEnumerator1
    UsePositionOffset=false
    Write Material Random=true
   )
   /SetVariables(Particles.Stream:NiagaraInt32=&SpawnGroup()) {isSetParameters=true}
   &FN_SetSpawnGroup(Number of Spawn Groups:NiagaraInt32=$Emitter.Spawn Multiplier)
   &AddVelocity(Velocity=<400 0 0>)
  ]
  particleUpdateScript [
   &ParticleState()
   &SkeletalMeshLocation(
    Bone / Socket Index=&Add_Integer(A=5, B=&SpawnGroup())
    MaskSpawnGroup=false
    Mesh Sampling Type=::NewEnumerator5
    Position Sampling=::NewEnumerator2
    Sampled Position Offset=&SelectVectorFromArray(
     Direct Array Index=$Particles.SpawnGroup
     Vector Selection Array=DataInterface() {propertyValues {FloatData [<0 0 0>, <0 0 0>, <-0.5 0 1.5>]}}
    )
    Skeletal Mesh=$System.SkeletalMesh
    Socket Sampling Mode=::NewEnumerator5
   )
   &InheritSourceMovement(
    Applied Position Scale=&VectorFromFloat(Value=&FloatFromCurve(FloatCurve=DataInterface() {propertyValues {Curve {keys [
          {arriveTangent=-3.0498, arriveTangentWeight=0.8563, interpMode=RCIM_Cubic, leaveTangentWeight=0.8563, time=0, value=1}
          {arriveTangent=-0.4165, interpMode=RCIM_Cubic, leaveTangent=-0.4165, tangentMode=RCTM_User, time=0.5, value=0}
         ]}}}))
    ApplyForce=false
    ApplyPosition=true
    ApplyPositionFalloff=false
   )
   &SolveForcesAndVelocity(Clamp Velocity=false, Limit Acceleration=false)
   &self:FN_Attached_Ribbon_v7_BETA.NiagaraScratchPadContainer_1.FN_Lock_Ribbon_Head_0(
    Emitter Count:NiagaraInt32=$Emitter.Count
    Locked Position:Vector3f=$Particles.SkeletalMeshLocation.SampledPosition
    Particle Stream:NiagaraInt32=$Particles.Stream
   ) {module=FN_Lock_Ribbon_Head}
   &ScaleColor(
    Scale Alpha=&FloatFromCurve(
     FloatCurve=DataInterface() {propertyValues {Curve {keys [{interpMode=RCIM_Cubic, tangentMode=RCTM_User, time=0, value=1}, {interpMode=RCIM_Cubic, tangentMode=RCTM_User, time=1, value=0}]}}}
     Scale Curve=&FloatFromCurve(
      FloatCurve=DataInterface() {propertyValues {Curve {keys [
          {interpMode=RCIM_Cubic, tangentMode=RCTM_User, time=0.1, value=0}
          {arriveTangent=2.7654, interpMode=RCIM_Cubic, leaveTangent=2.7654, tangentMode=RCTM_User, time=0.9, value=1}
         ]}}}
      CurveIndex=$User.Vehicle_Speed
      Scale Curve=&Multiply_Float(
       A=$System.Timeline005.InverseNormalizedTime
       B=&FNTimeOfDayScalar(Night Scale=1, Morning Scale=0.25, Day Scale=0.25, Evening Scale=1)
      )
     )
    )
    Scale RGB=&MakeVectorFromLinearColorRGB(Color=&MultiplyLinearColor(
      Linear Color=&LerpLinearColors(StartColor=$User.BikeColor, EndColor=<rgba 1 0 0 1>, LerpFactor=$User.IsBraking)
      Scale Factor=&MakeVector4_FromLinearColor(LinearColor=&LerpLinearColors(
        StartColor=<rgba 4 4 4 1>
        EndColor=<rgba 0.5156 0.5156 0.5156 1>
        LerpFactor=&FloatFromCurve(FloatCurve=DataInterface() {propertyValues {Curve {keys [
             {arriveTangent=6.5352, interpMode=RCIM_Cubic, leaveTangent=6.5352, tangentMode=RCTM_User, time=0, value=0}
             {arriveTangent=0.1546, interpMode=RCIM_Cubic, leaveTangent=0.1546, tangentMode=RCTM_User, time=0.15, value=0.9}
             {interpMode=RCIM_Cubic, tangentMode=RCTM_User, time=1, value=1}
            ]}}})
       ))
     ))
    ScaleA=true
    ScaleRGB=true
   )
  ]
  rendererClasses [&NiagaraRibbonRendererProperties]
  renderers [{rendererClass=&NiagaraRibbonRendererProperties}]
 }
 NeonAnimTrail_Tires {
  emitterUpdateScript [
   &EmitterState()
   &FN_SpawnAnimTrail(Enable:NiagaraBool=$User.IsBoostJump, Number of Anim Trails=2)
  ]
  particleSpawnScript [
   &FN_InitializeAnimTrailBoneIndecies(First Bone or Socket Index=9, Second Bone or Socket Index=8)
   &FN_InitializeAnimTrailBoneIndecies(AnimTrail Selection=1, First Bone or Socket Index=11, Second Bone or Socket Index=10) {module=FN_InitializeAnimTrailBoneIndecies001}
   &FN_InitializeAnimTrail(Color=$User.BikeColor, SkeletalMesh=$System.SkeletalMesh)
   &AddVelocity(Velocity=<200 0 0>)
  ]
  particleUpdateScript [
   &ParticleState()
   &FN_UpdateAnimTrail(SkeletalMesh=$System.SkeletalMesh)
   &InheritSourceMovement(
    Applied Position Scale=&VectorFromFloat(Value=&FloatFromCurve(FloatCurve=DataInterface() {propertyValues {Curve {keys [
          {interpMode=RCIM_Cubic, tangentMode=RCTM_User, time=0, value=1}
          {arriveTangent=-1.392, interpMode=RCIM_Cubic, leaveTangent=-1.392, tangentMode=RCTM_User, time=0.75, value=0}
         ]}}}))
    ApplyForce=false
    ApplyPosition=true
    ApplyPositionFalloff=false
   )
   &SolveForcesAndVelocity(Clamp Velocity=false, Limit Acceleration=false)
   &ScaleColor(
    Scale Alpha=&FloatFromCurve(
     FloatCurve=DataInterface() {propertyValues {Curve {keys [
         {arriveTangent=9.0725, interpMode=RCIM_Cubic, leaveTangent=9.0725, tangentMode=RCTM_User, time=0, value=0}
         {arriveTangent=-2.5249, interpMode=RCIM_Cubic, leaveTangent=-2.5249, tangentMode=RCTM_User, time=0.1408, value=0.9847}
         {arriveTangent=-0.0405, interpMode=RCIM_Cubic, leaveTangent=-0.0405, tangentMode=RCTM_User, time=1, value=0}
        ]}}}
     Scale Curve=$System.Timeline003.NormalizedTime
    )
    Scale RGB=<8 8 8>
    ScaleA=true
    ScaleRGB=true
   )
   &DynamicMaterialParameters(
    Index 0 Param 2=10
    Index 0 Param 3=0.05
    Module.Param0WriteEnabled=true
    Module.Param1WriteEnabled=true
    Module.Param2WriteEnabled=true
    Module.Param3WriteEnabled=true
   )
   /SetVariables(Particles.RibonCameraOffset:NiagaraPosition=&Add_Vector(
     A=Linked() {linkedVariable {name=Particles.Position, type=NiagaraPosition}}
     B=&Multiply_VectorByFloat(
      Vector=&ReturnCameraVector(Camera Vectors=::NewEnumerator4, Coordinate Space=::Simulation, Origin Position=$Particles.Position)
      Float=5
     )
    )) {enabled=false, isSetParameters=true}
  ]
  rendererClasses [&NiagaraRibbonRendererProperties]
  renderers [{rendererClass=&NiagaraRibbonRendererProperties}]
 }
 NeonTireRings {
  emitterUpdateScript [
   &EmitterState()
   &SpawnRate(
    Spawn Probability=&MakeCustomFloatFromBool(
     Bool=&Boolean_AndOperation(
      Bool1=&Boolean_OrOperation(Bool1=$User.IsOnWheelie, Bool2=&Boolean_OrOperation(Bool1=$User.IsBoostJump, Bool2=$User.IsSuperSliding))
      Bool2=&Boolean_NotOperation(Bool=$User.IsWaterDeep)
     )
     True Float=1
     False Float=0
    )
    SpawnRate=6
    Use Spawn Probability=true
   )
   &SpawnRate(
    Spawn Probability=&MakeCustomFloatFromBool(
     Bool=&Boolean_AndOperation(
      Bool1=&Boolean_OrOperation(Bool1=$User.IsOnWheelie, Bool2=&Boolean_OrOperation(Bool1=$User.IsBoostJump, Bool2=$User.IsSuperSliding))
      Bool2=&Boolean_NotOperation(Bool=$User.IsWaterDeep)
     )
     True Float=1
     False Float=0
    )
    SpawnGroup=1
    SpawnRate=6
    Use Spawn Probability=true
   ) {module=SpawnRate001}
  ]
  particleSpawnScript [
   &InitializeParticle(
    Color=&MultiplyLinearColor(Linear Color=$User.BikeColor, Scale Factor=&MakeVector4_FromLinearColor(LinearColor=<rgba 1 1 1 0.25>))
    Lifetime=0.5
    Mesh Scale Mode=::NewEnumerator3
    Mesh Uniform Scale=0.21
    UsePositionOffset=false
    Write Material Random=true
   )
   &InitialMeshOrientation(Mesh Orientation Mode=::NewEnumerator2, Orientation Vector=<0 0 1>, Rotation=<0.25 0 0>, Use Rotation Vector=true)
  ]
  particleUpdateScript [
   &ParticleState()
   &SkeletalMeshLocation(
    Bone / Socket Index=3
    MaskSpawnGroup=false
    Mesh Sampling Type=::NewEnumerator5
    Sampled Position Offset=&Lerp_Vector(
     A=&Add_Vector(A=<0 -7.5 0>, B=&Multiply_VectorByFloat(Vector=<0 -12 0>, Float=$Particles.NormalizedAge))
     B=&Add_Vector(A=<0 4.5 0>, B=&Multiply_VectorByFloat(Vector=<0 12 0>, Float=$Particles.NormalizedAge))
     Alpha=&MakeFloatFromInt(Int=&SpawnGroup())
    )
    Skeletal Mesh=$System.SkeletalMesh
    Socket Sampling Mode=::NewEnumerator5
   )
   &DynamicMaterialParameters(
    Index 0 Param 1=14
    Index 0 Param 3=1
    Module.Param0WriteEnabled=true
    Module.Param1WriteEnabled=true
    Module.Param2WriteEnabled=true
    Module.Param3WriteEnabled=true
   )
   &ScaleMeshSize(Scale Factor=&VectorFromFloat(Value=&FloatFromCurve(FloatCurve=DataInterface() {propertyValues {Curve {keys [{interpMode=RCIM_Cubic, tangentMode=RCTM_User, time=0, value=1}, {interpMode=RCIM_Cubic, tangentMode=RCTM_User, time=1, value=0}]}}})))
   &ScaleColor(
    Scale Alpha=&FloatFromCurve(
     FloatCurve=DataInterface() {propertyValues {Curve {keys [
         {arriveTangent=9.0725, interpMode=RCIM_Cubic, leaveTangent=9.0725, tangentMode=RCTM_User, time=0, value=0}
         {arriveTangent=-2.5249, interpMode=RCIM_Cubic, leaveTangent=-2.5249, tangentMode=RCTM_User, time=0.1408, value=0.9847}
         {arriveTangent=-0.0405, interpMode=RCIM_Cubic, leaveTangent=-0.0405, tangentMode=RCTM_User, time=1, value=0}
        ]}}}
     Scale Curve=$System.Timeline004.NormalizedTime
    )
    Scale RGB=&VectorFromFloat(Value=&Lerp_Float(
      A=10
      B=25
      Alpha=&FloatFromCurve(
       FloatCurve=DataInterface() {propertyValues {Curve {keys [
           {interpMode=RCIM_Cubic, tangentMode=RCTM_User, time=0.1, value=0}
           {arriveTangent=1.5933, interpMode=RCIM_Cubic, leaveTangent=1.5933, tangentMode=RCTM_User, time=0.5, value=0.2}
           {interpMode=RCIM_Cubic, tangentMode=RCTM_User, time=0.9, value=1}
          ]}}}
       CurveIndex=$User.Vehicle_Speed
      )
     ))
    ScaleA=true
    ScaleRGB=true
   )
  ]
  rendererClasses [&NiagaraMeshRendererProperties]
  renderers [{rendererClass=&NiagaraMeshRendererProperties, rendererIndex=0}]
 }
 Peel_Tires_Smoke001 {
  emitterUpdateScript [
   &EmitterState()
   &SpawnRate(
    Spawn Probability=&Multiply_Float(
     A=&MakeCustomFloatFromBool(
      Bool=&Boolean_OrOperation(Bool1=$User.IsWaterDeep, Bool2=$System.SpawnMud)
      True Float=0
      False Float=&MakeCustomFloatFromBool(Bool=$User.IsDoingStoppie, True Float=0, False Float=1)
     )
     B=&Lerp_Float(
      A=&MakeCustomFloatFromBool(Bool=$User.IsPeelOutOnRoad, True Float=0, False Float=1)
      B=1
      Alpha=$System.Timeline002.Float
     )
    )
    SpawnRate=&MakeFloatFromInt(Int=&MakeCustomIntFromBool(
      Bool=&Boolean_AndOperation(Bool1=$User.PeelOut_Wheel_Rear, Bool2=&Boolean_NotOperation(Bool=$System.SpawnSnow))
      True INT=85
      False INT=0
     ))
    Use Spawn Probability=true
   )
  ]
  particleSpawnScript [
   /SetVariables(Particles.RandomFloat=&UniformAOrBFloat(A=-1)) {isSetParameters=true, module=SetVariables_#5}
   /SetVariables(Particles.InitSpriteRotRate=&RandomRangeFloat(Minimum=-2, Maximum=2)) {isSetParameters=true, module=SetVariables_#6}
   /Niagara/Modules/Spawn/Initialization/InitializeParticle(
    Color=$System.PeelOutColor
    Lifetime=&Lerp_Float(
     A=&MakeCustomFloatFromBool(
      Bool=&SetBoolByFloatComparison(A=&MakeFloatFromInt(Int=$User.PeelOutColorIndex), B=4, Comparison Type=::NewEnumerator4)
      True Float=&RandomRangeFloat(Minimum=1.5, Maximum=2)
      False Float=&RandomRangeFloat(Minimum=0.6, Maximum=0.8)
     )
     B=1.3
     Alpha=$User.Vehicle_Speed
    )
    Module.Write Color=true
    Module.Write Lifetime=true
    Module.Write Mass=false
    Module.Write Material Random=true
    Module.Write Position=true
    Module.Write Scale=false
    Module.Write SpriteRotation=true
    Module.Write SpriteSize=true
    Position=$Engine.Owner.Position
    Sprite Rotation=&UniformRangedFloat(Minimum=0, Maximum=360)
    Sprite Size=&Vector2DFromFloat(Value=&Lerp_Float(
      A=&UniformRangedFloat(Minimum=40, Maximum=60)
      B=&UniformRangedFloat(Minimum=65, Maximum=80)
      Alpha=$User.Vehicle_Speed
     ))
   ) {module=InitializeParticle001}
   &SkeletalMeshLocation(
    Bone / Socket Index=1
    Inherit Velocity=false
    MaskSpawnGroup=false
    Mesh Sampling Type=::NewEnumerator5
    Sampled Position Offset=&MakeVector(X=0, Y=0, Z=0)
    Skeletal Mesh=$System.SkeletalMesh
    Socket Sampling Mode=::NewEnumerator5
    Write Sampled Mesh Attrributes to Particle Attributes=true
   )
   &AddVelocity(
    Velocity=&RandomRangeVector(
     Minimum=<500 -150 100>
     Maximum=&MakeVector(
      X=600
      Y=150
      Z=&Lerp_Float(
       A=&MakeCustomFloatFromBool(Bool=$User.IsPeelOutOnRoad, True Float=0, False Float=500)
       B=1400
       Alpha=$System.Timeline002.Float
      )
     )
    )
    Velocity Speed Scale=1.1
   )
  ]
  particleUpdateScript [
   &ParticleState()
   &ScaleColor(
    Scale Alpha=&FloatFromCurve(FloatCurve=DataInterface() {propertyValues {Curve {keys [
         {interpMode=RCIM_Cubic, tangentMode=RCTM_User, time=0, value=1}
         {arriveTangent=-1, interpMode=RCIM_Cubic, leaveTangent=-1, time=0.3521, value=0.3455}
         <at 1 0>
        ]}}})
    Scale RGB=&VectorFromFloat(Value=&FloatFromCurve(FloatCurve=DataInterface() {propertyValues {Curve {keys [{arriveTangent=1.8145, interpMode=RCIM_Cubic, leaveTangent=0.8876, tangentMode=RCTM_Break, time=0, value=0.75}, <at 0.3387 0.9565>, <at 1 1>]}}}))
    ScaleA=true
    ScaleRGB=true
   )
   &ScaleSpriteSize(Scale Factor=&Vector2DFromFloat(Value=&FloatFromCurve(
      FloatCurve=DataInterface() {propertyValues {Curve {keys [
          {interpMode=RCIM_Cubic, leaveTangent=1.201, tangentMode=RCTM_Break, time=0, value=0.179}
          {arriveTangent=0.6691, interpMode=RCIM_Cubic, leaveTangent=0.6691, tangentMode=RCTM_User, time=0.8329, value=1}
         ]}}}
      Scale Curve=&Multiply_Float(
       A=&Lerp_Float(A=3.5, B=5.5, Alpha=$System.Timeline002.Time)
       B=&MakeCustomFloatFromBool(Bool=$User.IsPeelOutOnRoad, True Float=0.6, False Float=1)
      )
     )))
   &SpriteRotationRate(Rotation Rate=&Multiply_Float(A=$Particles.Initial.MaterialRandom, B=1))
   &GravityForce(Gravity=&MakeVector(X=0, Y=0, Z=&Multiply_Float(A=-980, B=3)))
   &Drag(Use Linear Drag=true, Use Rotational Drag=true)
   &InheritSourceMovement(Applied Position Scale=<0.9 0.9 0.9>, ApplyForce=false, ApplyPosition=true, ApplyPositionFalloff=false)
   &SolveForcesAndVelocity(Clamp Velocity=false, Limit Acceleration=false)
   &DynamicMaterialParameters(
    Index 0 Param 1=&Add_Float(
     A=$Particles.RandomFloat
     B=&FloatFromCurve(FloatCurve=DataInterface() {propertyValues {Curve {keys [
          {arriveTangent=2.7533, interpMode=RCIM_Cubic, leaveTangent=2.847, tangentMode=RCTM_Break, time=0, value=0}
          {arriveTangent=1.3581, interpMode=RCIM_Cubic, leaveTangent=1.3581, tangentMode=RCTM_User, time=0.2558, value=0.5125}
          {arriveTangent=0.0529, interpMode=RCIM_Cubic, leaveTangent=0.2021, tangentMode=RCTM_User, time=1, value=1}
         ]}}})
    )
    Index 0 Param 3=&FloatFromCurve(FloatCurve=DataInterface() {propertyValues {Curve {keys [{arriveTangent=-53.9199, time=0, value=0}, {leaveTangent=0.2723, time=0.25, value=0}]}}})
    Index 0 Param 4=&FloatFromCurve(FloatCurve=DataInterface() {propertyValues {Curve {keys [<at 0 1>,<at 1 0>]}}})
    Index 1 Param 1=&FloatFromCurve(FloatCurve=DataInterface() {propertyValues {Curve {keys [<at 0 0>, {arriveTangent=4.2257, interpMode=RCIM_Cubic, leaveTangent=4.2257, tangentMode=RCTM_User, time=0.25, value=0}]}}})
    Index 1 Param 2=&FloatFromCurve(FloatCurve=DataInterface() {propertyValues {Curve {keys [<at 0 0>]}}})
    Index 1 Param 3=&FloatFromCurve(FloatCurve=DataInterface() {propertyValues {Curve {keys [{interpMode=RCIM_Cubic, leaveTangent=6.3917, tangentMode=RCTM_Break, time=0, value=0}]}}})
    Index 1 Param 4=&FloatFromCurve(FloatCurve=DataInterface() {propertyValues {Curve {keys [{arriveTangent=-0.136, leaveTangent=-0.136, time=0, value=1}]}}})
    Index 2 Param 1=&FloatFromCurve(FloatCurve=DataInterface() {propertyValues {Curve {keys [
         {arriveTangent=2.646, interpMode=RCIM_Cubic, leaveTangent=2.646, tangentMode=RCTM_User, time=0, value=0}
         {leaveTangent=-0.0719, time=1, value=1}
        ]}}})
    Index 2 Param 2=&FloatFromCurve(
     FloatCurve=DataInterface() {propertyValues {Curve {keys [
         {arriveTangent=1.9056, time=0, value=0.2}
         {arriveTangent=0.0259, interpMode=RCIM_Cubic, leaveTangent=0.0259, tangentMode=RCTM_User, time=1, value=1.4}
        ]}}}
     Scale Curve=&RandomRangeFloat(Minimum=0.5, Maximum=1)
    )
    Index 2 Param 3=&FloatFromCurve(FloatCurve=DataInterface() {propertyValues {Curve {keys [
         {interpMode=RCIM_Cubic, leaveTangent=2.3064, tangentMode=RCTM_Break, time=0, value=1}
         {interpMode=RCIM_Cubic, leaveTangent=11.1233, tangentMode=RCTM_Break, time=1, value=2}
        ]}}})
    Index 2 Param 4=&FloatFromCurve(
     FloatCurve=DataInterface() {propertyValues {Curve {keys [
         {arriveTangent=-0.6867, interpMode=RCIM_Cubic, leaveTangent=-0.6867, tangentMode=RCTM_User, time=0, value=1}
         {interpMode=RCIM_Cubic, leaveTangent=-1.2956, tangentMode=RCTM_Break, time=1, value=0.75}
        ]}}}
     Scale Curve=1.6
    )
    Module.Param0WriteEnabled=true
    Module.Param0WriteEnabled1=true
    Module.Param0WriteEnabled2=true
    Module.Param1WriteEnabled=false
    Module.Param1WriteEnabled1=true
    Module.Param1WriteEnabled2=true
    Module.Param2WriteEnabled=true
    Module.Param2WriteEnabled1=true
    Module.Param2WriteEnabled2=true
    Module.Param3WriteEnabled=true
    Module.Param3WriteEnabled1=true
    Module.Param3WriteEnabled2=true
    Write Parameter Index 1=true
    Write Parameter Index 2=true
   )
  ]
  rendererClasses [&NiagaraSpriteRendererProperties]
  renderers [
   {rendererClass=&NiagaraSpriteRendererProperties}
   {rendererClass=&NiagaraSpriteRendererProperties, rendererIndex=1}
  ]
 }
 SkidMarks {
  emitterUpdateScript [
   &EmitterState()
   &SpawnPerUnit(
    Module.Use Max Movement Threshold=false
    Module.Use Movement Tolerance=false
    Spawn Per Unit=50
    Spawn Probability:NiagaraFloat=&MakeCustomFloatFromBool(
     Bool=&Boolean_AndOperation(
      Bool1=&Boolean_AndOperation(
       Bool1=&Boolean_OrOperation(Bool1=$User.PeelOut_Wheel_Rear, Bool2=$User.IsSuperSliding)
       Bool2=&Boolean_NotOperation(Bool=$User.IsDoingStoppie)
      )
      Bool2=&Boolean_NotOperation(Bool=&Boolean_OrOperation(
        Bool1=$System.SpawnSand
        Bool2=&Boolean_OrOperation(Bool1=$System.SpawnSnow, Bool2=&Boolean_OrOperation(Bool1=$System.SpawnIce, Bool2=$System.SpawnMud))
       ))
     )
     True Float=1
     False Float=0
    )
    Use Spawn Probability=true
   )
  ]
  particleSpawnScript [
   &InitializeParticle(
    Color=<rgba 0 0 0 0.6>
    Lifetime=5
    Module.Write Color=true
    Module.Write Lifetime=true
    Module.Write Mass=true
    Module.Write Material Random=true
    Module.Write Position=false
    Module.Write Scale=false
    Module.Write SpriteRotation=false
    Module.Write SpriteSize=true
    Position=&SimulationPosition()
    Sprite Size=<25 25>
   )
   /SetVariables() {isSetParameters=true}
   /SetVariables(Particles.RibbonWidth=20) {isSetParameters=true}
   &FN_SetSpawnGroup(Number of Spawn Groups:NiagaraInt32=&MakeCustomIntFromBool(Bool=$User.IsSuperSliding, True INT=2, False INT=1))
   &SkeletalMeshLocation(
    Bone / Socket Index=&SelectIntFromArray(Direct Array Index=&SpawnGroup(), Int Selection Array=DataInterface() {propertyValues {IntData [1,0]}})
    Inherit Velocity=false
    Interpolate Mesh Sampling=false
    MaskSpawnGroup=false
    Mesh Sampling Type=::NewEnumerator5
    Sampled Position Offset Coordinate Space:ENiagaraMeshTransforms=::NewEnumerator3
    Skeletal Mesh=$System.SkeletalMesh
    Socket Sampling Mode=::NewEnumerator5
   )
  ]
  particleUpdateScript [
   &ParticleState()
   &CameraOffset(Camera Offset Amount=32)
   &ScaleColor(
    Scale Alpha=&FloatFromCurve(
     FloatCurve=DataInterface() {propertyValues {Curve {keys [{interpMode=RCIM_Cubic, tangentMode=RCTM_Break, time=0, value=1}, <at 1 0>]}}}
     Scale Curve=&FloatFromCurve(
      FloatCurve=DataInterface() {propertyValues {Curve {keys [{interpMode=RCIM_Cubic, tangentMode=RCTM_Break, time=0, value=0}, {interpMode=RCIM_Cubic, tangentMode=RCTM_Break, time=0.1153, value=1}]}}}
      CurveIndex=$Particles.Age
     )
    )
    ScaleA=true
    ScaleRGB=true
   )
  ]
  rendererClasses [&NiagaraRibbonRendererProperties]
  renderers [{rendererClass=&NiagaraRibbonRendererProperties}]
 }
 Sparks_Tires_Blown {
  emitterUpdateScript [
   &EmitterState()
   &SpawnRate(
    SpawnRate=&Multiply_Float(
     A=&Lerp_Float(
      A=&RandomRangeFloat(Minimum=20, Maximum=30)
      B=&RandomRangeFloat(Minimum=40, Maximum=50)
      Alpha=&Power(a=$User.Vehicle_Speed, b=1.25)
     )
     B=&Multiply_Float(
      A=&MakeFloatFromInt(Int=$System.Sparks_FrontTire)
      B=&FloatFromCurve(FloatCurve=$System.Sparks_SpawnCurve, CurveIndex=&Abs_Float(Float:NiagaraFloat=$User.Vehicle_Speed))
     )
    )
    Use Spawn Probability=false
   )
   &SpawnRate(
    SpawnGroup=1
    SpawnRate=&Multiply_Float(
     A=&Lerp_Float(
      A=&RandomRangeFloat(Minimum=20, Maximum=30)
      B=&RandomRangeFloat(Minimum=40, Maximum=50)
      Alpha=&Power(a=$User.Vehicle_Speed, b=1.25)
     )
     B=&Multiply_Float(
      A=&MakeFloatFromInt(Int=$System.Sparks_RearTire)
      B=&FloatFromCurve(FloatCurve=$System.Sparks_SpawnCurve, CurveIndex=&Abs_Float(Float:NiagaraFloat=$User.Vehicle_Speed))
     )
    )
    Use Spawn Probability=false
   ) {module=SpawnRate001}
  ]
  particleSpawnScript [
   /SetVariables(Particles.VisibilityTag=&SelectIntFromArray(Direct Array Index=&SpawnGroup(), Int Selection Array=DataInterface() {propertyValues {IntData [0,0,1,1]}})) {isSetParameters=true}
   /Niagara/Modules/Spawn/Initialization/V2/InitializeParticle(
    Color Maximum=&MakeCustomLinearColorFromBool(
     Bool=&SetBoolByIntComparison(A=&SpawnGroup(), B=2, Comparison Type=::NewEnumerator2)
     True Color=<rgba 150 41.4953 0 1>
     False Color=<rgba 40 0.5 0 1>
    )
    Color Minimum=&MakeCustomLinearColorFromBool(
     Bool=&SetBoolByIntComparison(A=&SpawnGroup(), B=2, Comparison Type=::NewEnumerator2)
     True Color=<rgba 50 9.9924 0 1>
     False Color=<rgba 25 15.7812 0 1>
    )
    Color Mode=::NewEnumerator2
    Lifetime Max=&MakeCustomFloatFromBool(
     Bool=&SetBoolByIntComparison(A=&SpawnGroup(), B=2, Comparison Type=::NewEnumerator2)
     True Float=0.1
     False Float=0.4
    )
    Lifetime Min=0.05
    Lifetime Mode=::NewEnumerator1
    Mesh Scale=&Multiply_VectorByFloat(
     Vector=&MaskVectorBySpawnGroup(Masked Vector=<0.205 0.15 0.205>, Passthrough Vector=<0.26 0.11 0.26>, Spawn Group Mask=3)
     Float=2
    )
    Mesh Scale Mode=::NewEnumerator4
    Sprite Size Max=<2.5 30>
    Sprite Size Min=<1.25 20>
    Sprite Size Mode=::NewEnumerator2
    UsePositionOffset=false
    Write Material Random=true
   )
   &SkeletalMeshLocation(
    Bone / Socket Index=&SpawnGroup()
    MaskSpawnGroup=false
    Mesh Sampling Type=::NewEnumerator5
    Sampled Position Offset=&SelectVectorFromArray(
     Direct Array Index=&SpawnGroup()
     Vector Selection Array=DataInterface() {propertyValues {FloatData [<0 0 0>, <0 0 6>, <0 0 0>, <0 0 0>]}}
    )
    Skeletal Mesh=$System.SkeletalMesh
    Socket Sampling Mode=::NewEnumerator5
   )
   /SetVariables(Particles.ParticleVelocityAdd:Vector3f=&RotateVectorByQuaternion(
     Vector To Rotate=&MakeVector(
      X=&RandomRangeFloat(Minimum=-100, Maximum=-400)
      Y=0
      Z=&Multiply_Float(
       A=&RandomRangeFloat(Minimum=10, Maximum=200)
       B=&Abs_Float(Float:NiagaraFloat=$User.Vehicle_Speed)
      )
     )
     Quaternion=$StackContext.SkeletalMeshLocation.SampledRotation
    )) {isSetParameters=true}
   /SetVariables(Particles.Velocity=&MakeCustomVectorFromBool(
     Bool=&SetBoolByIntComparison(A=&SpawnGroup(), B=2, Comparison Type=::NewEnumerator6)
     True Vector=&Add_Vector(A=$Particles.Velocity, B=$Particles.ParticleVelocityAdd)
     False Vector=<0 0 0>
    )) {isSetParameters=true}
   /SetVariables(Particles.SpriteAlignment=&NormalizeVector(Vector To Normalize=$Particles.ParticleVelocityAdd, Safe Normalize=true, Safe Fallback Threshold=0, Safe Fallback Vector=<0 0 0.001>)) {isSetParameters=true}
   &InitialMeshOrientation(
    Mesh Orientation Mode=::NewEnumerator2
    Orientation Vector=<0 1 0>
    Rotation=&RandomRangeVector(Minimum=<0 0 0>, Maximum=<0 1 0>)
    Use Rotation Vector=true
   )
  ]
  particleUpdateScript [
   &ParticleState()
   &SolveForcesAndVelocity(Clamp Velocity=false, Limit Acceleration=false)
   &ScaleColor(
    Scale Alpha=&FloatFromCurve(
     FloatCurve=DataInterface() {propertyValues {Curve {keys [<at -0.0046 1.0113 RCIM_Cubic>,<at 1.0011 0.002 RCIM_Cubic>]}}}
     Scale Curve=&MakeCustomFloatFromBool(
      Bool=&SetBoolByIntComparison(A=&SpawnGroup(), B=2, Comparison Type=::NewEnumerator2)
      True Float=&ClampFloat(Float=&Multiply_Float(A=2, B=&Abs_Float(Float:NiagaraFloat=&Power(a=$User.Vehicle_Speed, b=2.5))))
      False Float=1
     )
    )
    Scale RGB=&MakeCustomVectorFromBool(
     Bool=&SetBoolByIntComparison(A=&SpawnGroup(), B=2, Comparison Type=::NewEnumerator2)
     True Vector=<100 100 100>
     False Vector=<1 1 1>
    )
    ScaleA=true
    ScaleRGB=true
   )
   &ScaleSpriteSize(
    Uniform Curve Scale=&Lerp_Float(A=0.2, B=1, Alpha=$User.Vehicle_Speed)
    Uniform Curve Sprite Scale=DataInterface() {propertyValues {Curve {keys [
        {arriveTangent=56.0592, interpMode=RCIM_Cubic, leaveTangent=56.0592, tangentMode=RCTM_User, time=0, value=0}
        {arriveTangent=-1.0335, interpMode=RCIM_Cubic, leaveTangent=-1.0335, tangentMode=RCTM_User, time=0.1674, value=2.5}
        <at 1 0>
       ]}}}
   )
   &ScaleSpriteSizeBySpeed(Max Scale Factor=<1 1>, Min Scale Factor=<0.5 0.6>)
   &DynamicMaterialParameters(
    Index 0 Param 1=5
    Index 0 Param 2=2
    Module.Index0Param0Write=true
    Module.Index0Param1Write=true
    Module.Index0Param2Write=false
    Module.Index0Param3Write=false
   )
   &SkeletalMeshLocation(
    Bone / Socket Index=2
    MaskSpawnGroup=true
    Mesh Sampling Type=::NewEnumerator5
    Skeletal Mesh=$System.SkeletalMesh
    Socket Sampling Mode=::NewEnumerator5
    Spawn Group Mask=2
   ) {module=SkeletalMeshLocation001}
   &SkeletalMeshLocation(
    Bone / Socket Index=3
    MaskSpawnGroup=true
    Mesh Sampling Type=::NewEnumerator5
    Skeletal Mesh=$System.SkeletalMesh
    Socket Sampling Mode=::NewEnumerator5
    Spawn Group Mask=3
   ) {module=SkeletalMeshLocation002}
   &ScaleMeshSize(Scale Factor=&VectorFromFloat(Value=2)) {enabled=false}
   /SetVariables(Particles.MeshOrientation=&MultiplyQuaternion(
     Quaternion A=/Niagara/DynamicInputs/TypeConversions/MakeCustomQuaternionFromBool(
      Bool:NiagaraBool=&SetBoolByIntComparison(A=&SpawnGroup(), B=3, Comparison Type=::NewEnumerator6)
      True Quat:Quat4f=$StackContext.SkeletalMeshLocation001.SampledRotation
      False Quat:Quat4f=$StackContext.SkeletalMeshLocation002.SampledRotation
     )
     Quaternion B=&OrientMeshToVector(Look At Direction=&RandomRangeVector(Minimum=<0 0 0>, Maximum=<0 1 0>), Reference Axis Vector=<1 0 0>)
    )) {isSetParameters=true}
  ]
  rendererClasses [&NiagaraSpriteRendererProperties]
  renderers [{rendererClass=&NiagaraSpriteRendererProperties}]
 }
}
summary {
 systemName=NS_Motorcycle_Wheel_And_Tire_Interaction_Sport_Trimmed
 emitters [
  {emitterName=Dust_Tires_Mesh, rendererClasses [&NiagaraMeshRendererProperties]}
  {emitterName=Exhaust_Local_New, rendererClasses [&NiagaraSpriteRendererProperties]}
  {emitterName=SkidMarks, rendererClasses [&NiagaraRibbonRendererProperties]}
  {emitterName=Peel_Tires_Smoke001, rendererClasses [&NiagaraSpriteRendererProperties]}
  {emitterName=NeonAnimTrail_Thin, rendererClasses [&NiagaraRibbonRendererProperties]}
  {emitterName=NeonTireRings, rendererClasses [&NiagaraMeshRendererProperties]}
  {emitterName=NeonAnimTrail_Tires, rendererClasses [&NiagaraRibbonRendererProperties]}
  {emitterName=Jet_BurstSparksHigh, rendererClasses [&NiagaraSpriteRendererProperties]}
  {emitterName=LensFlare, simTarget=GPUComputeSim, rendererClasses [&NiagaraSpriteRendererProperties]}
  {emitterName=Sparks_Tires_Blown, rendererClasses [&NiagaraSpriteRendererProperties]}
  {
   emitterName=ExhaustLight001
   rendererClasses [
    &NiagaraLightRendererProperties
    &NiagaraSpriteRendererProperties
    &NiagaraMeshRendererProperties
   ]
  }
 ]
 userVariables [
  User.BikeColor {type=LinearColor, description="", defaultValue {struct=LinearColor, value=<rgba 1 1 1 1>}}
  User.ChargeProgress {type=NiagaraFloat, defaultValue=0, description=""}
  User.IsBoostJump {type=NiagaraBool}
  User.IsBraking {type=NiagaraFloat, defaultValue=0, description=""}
  User.IsDoingStoppie {type=NiagaraBool}
  User.IsInAir {type=NiagaraBool}
  User.IsOnWheelie {type=NiagaraBool}
  User.IsPeelOutOnRoad {type=NiagaraBool}
  User.IsPeelOutOnRoadRewind {type=NiagaraBool}
  User.IsShifting {type=NiagaraBool}
  User.IsSuperSliding {type=NiagaraBool}
  User.IsWaterDeep {type=NiagaraBool}
  User.PeelOut_Wheel_Rear {type=NiagaraBool}
  User.PeelOutColorArray {
   type=NiagaraDataInterfaceArrayColor
   description=""
   defaultValue {
    struct=NiagaraExt_VariableValue_DataInterface
    value {
     dataInterface=&self:NiagaraDataInterfaceArrayColor_0
     dataInterfaceClass=&NiagaraDataInterfaceArrayColor
    }
   }
  }
  User.PeelOutColorIndex {type=NiagaraInt32}
  User.RibbonIDOffset {type=NiagaraInt32}
  User.SkeletalMesh {
   type=NiagaraDataInterfaceSkeletalMesh
   defaultValue {
    struct=NiagaraExt_VariableValue_DataInterface
    value {dataInterface=&self:NiagaraDataInterfaceSkeletalMesh_3}
   }
  }
  User.SpawnExhaustFlames {type=NiagaraBool}
  User.TirePoppedSparksFront {type=NiagaraBool}
  User.TirePoppedSparksRear {type=NiagaraBool}
  User.Vehicle_Speed {type=NiagaraFloat, defaultValue=0, description=""}
  User.VehicleHasDriver {type=NiagaraBool}
  User.VehicleHasFuel {type=NiagaraBool}
 ]
}
